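# Material function in the MaterialX category that declares an anisotropic volume distribution function (VDF) for participating media, described as Henyey-Greenstein scattering.
# It exposes absorption and scattering per color component, plus a scalar anisotropy input for forward, backward or uniform scattering.
# The body is a placeholder: none of the inputs are connected, and the output is a MakeMaterialAttributes node with nothing wired in, so no phase function is computed.
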MaterialFunction /InterchangeAssets/Functions/MX_AnisotropicVDF
graph {
 properties {
  description="Constructs a VDF scattering light for a participating medium, based onthe Henyey-Greenstein phase function. Forward, backward and uniform scattering is supportedand controlled by the anisotropy input."
  libraryCategoriesText [MaterialX]
 }
 nodes [
  FunctionInput() {
   description="Absorption rate for the medium (rate per distance traveled in themedium, given in m−1 ). Set for each color component/wavelength separately."
   inputName=absorption
  }
  FunctionInput() {
   description="Scattering rate for the medium (rate per distance traveled in themedium, given in m−1 ). Set for each color component/wavelength separately."
   inputName=scattering
   sortPriority=1
  }
  FunctionInput() {
   description="Anisotropy factor, controlling the scattering direction, range [-1.0,1.0]. Negative values give backwards scattering, positive values give forward scattering, and a value of 0.0 (the default) gives uniform scattering."
   inputName=anisotropy
   inputType=FunctionInput_Scalar
   sortPriority=2
  }
  FunctionOutput(None=MakeMaterialAttributes!(
    BaseColor
    Metallic
    Specular
    Roughness
    Anisotropy
    EmissiveColor
    Opacity
    OpacityMask
    Normal
    Tangent
    WorldPositionOffset
    SubsurfaceColor
    ClearCoat
    ClearCoatRoughness
    AmbientOcclusion
    Refraction
    CustomizedUVs_0
    CustomizedUVs_1
    CustomizedUVs_2
    CustomizedUVs_3
    CustomizedUVs_4
    CustomizedUVs_5
    CustomizedUVs_6
    CustomizedUVs_7
    PixelDepthOffset
    ShadingModel
    Displacement
   )) {outputName=Output}
 ]
}
