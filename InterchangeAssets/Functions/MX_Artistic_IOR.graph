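# MaterialX library material function that converts artist-friendly reflectivity and edge_color into complex IOR (ior and extinction) for conductor BSDFs.
# Reflectivity is clamped below one, then yields maximum and minimum index bounds, which edge_color interpolates per channel to give ior.
# Extinction is solved from that ior and the clamped reflectivity, flooring the squared term at zero before the square root.
# Named reroutes carry the intermediate terms between stages.
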MaterialFunction /InterchangeAssets/Functions/MX_Artistic_IOR
graph {
 properties {
  bExposeToLibrary=true
  description="Converts the artistic parameterization reflectivity and edge_color to complex IOR values. To be used with the <conductor_bsdf> node."
  libraryCategoriesText [MaterialX]
 }
 nodes [
  FunctionOutput(None=NamedRerouteUsage()) {description=Computed index of refraction., outputName=ior}
  FunctionOutput(None=NamedRerouteUsage()) {description=Computed extinction coefficient., outputName=extinction, sortPriority=1}
  n1 = FunctionInput() {
   bUsePreviewValueAsDefault=true
   description=Reflectivity per color component at facing angles.
   inputName=reflectivity
   previewValue=<0.944 0.776 0.373 1>
  }
  n2 = Clamp(None=@n1) {maxDefault=0.99}
  NamedRerouteDeclaration(None=SquareRoot(None=NamedRerouteDeclaration(None=@n2) {name=r, nodeColor=<rgba 1 0.2109 0 1>})) {name=r_sqrt, nodeColor=<rgba 0 0.2031 1 1>}
  n3 = NamedRerouteUsage()
  n4 = Divide(A=Add(B=@n3) {constA=1}, B=OneMinus(None=@n3))
  n5 = NamedRerouteUsage()
  n6 = Divide(A=OneMinus(None=@n5), B=Add(B=@n5) {constA=1})
  n7 = FunctionInput() {
   bUsePreviewValueAsDefault=true
   description=Reflectivity per color component at grazing angles.
   inputName=edge_color
   sortPriority=1
   previewValue=<0.998 0.981 0.751 1>
  }
  NamedRerouteDeclaration(None=LinearInterpolate(
    A=NamedRerouteDeclaration(None=@n4) {name=n_max, nodeColor=<rgba 0 0.8359 1 1>}
    B=NamedRerouteDeclaration(None=@n6) {name=n_min, nodeColor=<rgba 0.0078 0 1 1>}
    Alpha=@n7
   )) {name=ior, nodeColor=<rgba 0 1 0.0625 1>}
  n8 = Add(A=NamedRerouteUsage())
  n9 = Multiply(A=@n8, B=@n8)
  n10 = NamedRerouteUsage()
  n11 = Subtract(A=NamedRerouteUsage())
  n12 = Divide(A=Subtract(A=Multiply(A=@n9, B=@n10), B=Multiply(A=@n11, B=@n11)), B=OneMinus(None=@n10))
  NamedRerouteDeclaration(None=SquareRoot(None=Max(A=@n12) {constB=0})) {name=extinction, nodeColor=<rgba 0.5234 1 0 1>}
 ]
}
