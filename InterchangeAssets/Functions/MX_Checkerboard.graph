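# MX_Checkerboard is a MaterialX library material function that produces a 2D checkerboard of two input colors.
# It multiplies the coordinates (first UV set by default) by uvtiling, subtracts uvoffset, then floors the result to get cell indices.
# A dot product with (1,1) sums the indices, and a MaterialX modulo by 2 turns that parity into the lerp alpha choosing color2 or color1.
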MaterialFunction /InterchangeAssets/Functions/MX_Checkerboard
graph {
 properties {bExposeToLibrary=true, description=a 2D checkerboard pattern., libraryCategoriesText [MaterialX]}
 nodes [
  n1 = FunctionInput() {bUsePreviewValueAsDefault=true, description=The second color used in the checkerboard pattern., inputName=color2, sortPriority=1}
  n2 = FunctionInput() {bUsePreviewValueAsDefault=true, description=The first color used in the checkerboard pattern., inputName=color1, previewValue=<1 1 1 1>}
  n3 = FunctionInput() {
   bUsePreviewValueAsDefault=true
   description="The tiling of the checkerboard pattern along each axis, with higher values producing smaller squares."
   inputName=uvtiling
   inputType=FunctionInput_Vector2
   sortPriority=2
   previewValue=<8 8 0 1>
  }
  n4 = FunctionInput(Preview=TextureCoordinate()) {
   bUsePreviewValueAsDefault=true
   description=The input 2d space. Default is the first texture coordinates.
   inputName=Coordinates
   inputType=FunctionInput_Vector2
   sortPriority=3
  }
  n5 = Multiply(A=@n3, B=@n4)
  n6 = FunctionInput() {
   bUsePreviewValueAsDefault=true
   description=The offset of the checkerboard pattern along each axis.
   inputName=uvoffset
   inputType=FunctionInput_Vector2
   sortPriority=2
  }
  n7 = DotProduct(A=Floor(None=Subtract(A=@n5, B=@n6)), B=Constant2Vector() {g=1, r=1})
  n8 = Constant() {r=2}
  FunctionOutput(None=LinearInterpolate(A=@n1, B=@n2, Alpha=MaterialXMod!(A=@n7, B=@n8)))
 ]
}
