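# MaterialX library material function that converts a hair scattering color to an absorption coefficient using the Chiang 2016 mapping.
# It clamps the color to a small minimum, takes its logarithm, divides by a roughness-dependent factor and squares the result.
# That factor is a polynomial in azimuthal roughness with fixed fitted coefficients, assembled from powers of the roughness through named reroutes.
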MaterialFunction /InterchangeAssets/Functions/MX_ChiangHairAbsorptionFromColor
graph {
 properties {
  bExposeToLibrary=true
  description="Converts the hair scattering color to absorption coefficient using the mapping method described in [^Chiang2016]."
  libraryCategoriesText [MaterialX]
 }
 nodes [
  n1 = Clamp(None=FunctionInput() {bUsePreviewValueAsDefault=true, inputName=color, previewValue=<1 1 1 1>}) {minDefault=0.001}
  n2 = NamedRerouteDeclaration(None=Divide(A=Logarithm!(None=@n1), B=NamedRerouteUsage())) {name=sigma, nodeColor=<rgba 1 0.4688 0 1>}
  FunctionOutput(None=Multiply(A=@n2, B=@n2)) {outputName=absorption}
  n3 = FunctionInput() {bUsePreviewValueAsDefault=true, inputName=azimuthal_roughness, inputType=FunctionInput_Scalar, sortPriority=1, previewValue=<0.2 1 1 1>}
  n4 = NamedRerouteDeclaration(None=@n3) {name=betaN, nodeColor=<rgba 1 0 0.8906 1>}
  n5 = NamedRerouteDeclaration(None=Multiply(A=@n4, B=@n4)) {name=b2, nodeColor=<rgba 0 0.25 1 1>}
  NamedRerouteDeclaration(None=Multiply(A=@n5, B=@n5)) {name=b4, nodeColor=<rgba 0.5703 1 0 1>}
  n6 = Constant() {r=5.969}
  n7 = Multiply(A=Constant() {r=0.215}, B=NamedRerouteUsage())
  n8 = Constant() {r=2.532}
  n9 = NamedRerouteUsage()
  n10 = NamedRerouteUsage()
  n11 = NamedRerouteUsage()
  n12 = Subtract(A=Add(A=Subtract(A=@n6, B=@n7), B=Multiply(A=@n8, B=@n9)), B=Multiply(A=Constant() {r=10.73}, B=Multiply(A=@n10, B=@n11)))
  n13 = Multiply(A=Constant() {r=5.574}, B=NamedRerouteUsage())
  n14 = Constant() {r=0.245}
  n15 = Multiply(A=NamedRerouteUsage(), B=NamedRerouteUsage())
  NamedRerouteDeclaration(None=Add(A=Add(A=@n12, B=@n13), B=Multiply(A=@n14, B=@n15))) {name=b_fac, nodeColor=<rgba 0 1 0.7891 1>}
 ]
}
