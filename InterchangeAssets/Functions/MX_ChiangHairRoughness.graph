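# MaterialX library material function that converts artistic hair roughness into roughness for the R, TT and TRT lobes, following Chiang 2016.
# Clamped longitudinal and azimuthal inputs feed fitted polynomials with high-power terms, producing variance v (squared) and scale s, shared through named reroutes.
# Each output is a two-component vector, with the TT and TRT first components multiplied by their squared scale inputs, whose defaults follow Marschner 2003.
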MaterialFunction /InterchangeAssets/Functions/MX_ChiangHairRoughness
graph {
 properties {
  bExposeToLibrary=true
  description="Converts the artistic parameterization hair roughness to roughness for R, TT and TRT lobes, as described in [^Chiang2016]."
  libraryCategoriesText [MaterialX]
 }
 nodes [
  FunctionOutput(None=AppendVector(A=NamedRerouteUsage(), B=NamedRerouteUsage())) {outputName=roughness_R}
  n1 = FunctionInput() {
   bUsePreviewValueAsDefault=true
   description="Roughness scale for TT lobe. Empirical roughness scale from Marschner et al. (2003)."
   inputName=scale_TT
   inputType=FunctionInput_Scalar
   sortPriority=2
   previewValue=<0.5 0 0 1>
  }
  n2 = Multiply(A=@n1, B=@n1)
  n3 = NamedRerouteUsage()
  FunctionOutput(None=AppendVector(A=Multiply(A=@n2, B=@n3), B=NamedRerouteUsage())) {outputName=roughness_TT, sortPriority=1}
  n4 = FunctionInput() {
   bUsePreviewValueAsDefault=true
   description="Roughness scale for TRT lobe. Empirical roughness scale from Marschner et al. (2003)."
   inputName=scale_TRT
   inputType=FunctionInput_Scalar
   sortPriority=2
   previewValue=<2 0 0 1>
  }
  n5 = Multiply(A=@n4, B=@n4)
  n6 = NamedRerouteUsage()
  FunctionOutput(None=AppendVector(A=Multiply(A=@n5, B=@n6), B=NamedRerouteUsage())) {outputName=roughness_TRT, sortPriority=2}
  n7 = FunctionInput() {
   bUsePreviewValueAsDefault=true
   description="Longitudinal roughness, range [0.0, 1.0]"
   inputName=longitudinal
   inputType=FunctionInput_Scalar
   previewValue=<0.1 0 0 1>
  }
  NamedRerouteDeclaration(None=Clamp(None=@n7) {minDefault=0.001}) {name=lr, nodeColor=<rgba 0.9453 0 1 1>}
  n8 = FunctionInput() {
   bUsePreviewValueAsDefault=true
   description="Azimuthal roughness, range [0.0, 1.0]."
   inputName=azimuthal
   inputType=FunctionInput_Scalar
   sortPriority=1
   previewValue=<0.2 0 0 1>
  }
  NamedRerouteDeclaration(None=Clamp(None=@n8) {minDefault=0.001}) {name=ar, nodeColor=<rgba 0 1 0.625 1>}
  n9 = Constant() {r=0.726}
  n10 = NamedRerouteUsage()
  n11 = Multiply(A=@n10, B=@n10)
  n12 = Constant() {r=0.812}
  n13 = NamedRerouteUsage()
  n14 = Constant() {r=20}
  n15 = Add(A=Add(A=Multiply(A=@n9, B=@n10), B=Multiply(A=@n11, B=@n12)), B=Multiply(A=Power(Base=@n13, Exp=@n14), B=Constant() {r=3.7}))
  NamedRerouteDeclaration(None=Multiply(A=@n15, B=@n15)) {name=v, nodeColor=<rgba 0.3359 1 0 1>}
  n16 = NamedRerouteUsage()
  n17 = Multiply(A=Constant() {r=0.265}, B=@n16)
  n18 = Multiply(A=Multiply(A=@n16, B=@n16), B=Constant() {r=1.194})
  n19 = Power(Base=NamedRerouteUsage(), Exp=Constant() {r=22})
  n20 = Constant() {r=5.372}
  NamedRerouteDeclaration(None=Add(A=Add(A=@n17, B=@n18), B=Multiply(A=@n19, B=@n20))) {name=s, nodeColor=<rgba 0.3828 0 1 1>}
 ]
}
