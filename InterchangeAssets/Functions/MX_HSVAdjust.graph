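# MaterialX library material function that adjusts the hue, saturation and value of an input RGB color.
# It converts Input to HSV, offsets hue by amount.x, scales saturation and value by amount.y and amount.z, then converts back to RGB.
# Constant channel masks split amount into additive and multiplicative vectors, linked through named reroutes.
# The amount input defaults to a neutral adjustment.
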MaterialFunction /InterchangeAssets/Functions/MX_HSVAdjust
graph {
 properties {
  bExposeToLibrary=true
  description="Adjust the hue, saturation and value of an RGB color by converting the input color    to HSV, adding amount.x to the hue, multiplying the saturation by amount.y,    multiplying the value by amount.z, then converting back to RGB."
  libraryCategoriesText [MaterialX]
 }
 nodes [
  n1 = NamedRerouteUsage()
  n2 = NamedRerouteUsage()
  FunctionOutput(None=HsvToRgb!(None=Multiply(A=@n1, B=@n2)))
  n3 = FunctionInput() {bUsePreviewValueAsDefault=true, inputName=amount, sortPriority=1, previewValue=<0 1 1 1>}
  n4 = Constant3Vector() {constant=<rgba 1 0 0 0>}
  NamedRerouteDeclaration(None=Multiply(A=@n3, B=@n4)) {name=hchans, nodeColor=<rgba 0 0.8359 1 1>}
  NamedRerouteDeclaration(None=Multiply(A=Constant3Vector() {constant=<rgba 0 1 1 0>}, B=@n3)) {name=tmp1, nodeColor=<rgba 1 0.8438 0 1>}
  NamedRerouteDeclaration(None=Add(A=@n4, B=NamedRerouteUsage())) {name=svchans, nodeColor=<rgba 0.5234 0 1 1>}
  n5 = FunctionInput() {description=The input color to be adjusted., inputName=Input}
  NamedRerouteDeclaration(None=Add(A=RgbToHsv!(None=@n5), B=NamedRerouteUsage())) {name=tmp2, nodeColor=<rgba 0.7578 1 0 1>}
 ]
}
