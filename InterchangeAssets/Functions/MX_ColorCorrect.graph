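# MaterialX color correction material function, exposed to the library, combining hue, saturation, gamma, lift, gain, contrast and exposure adjustments into one node.
# Hue shifts through MX_HSVAdjust, then saturation interpolates between luminance and the shifted color.
# A MaterialX Range node applies gamma, lift and gain raise darks and scale lights, and contrast works around an input pivot.
# Named reroutes link the stages, with exposure computed as a power-of-two multiplier.
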MaterialFunction /InterchangeAssets/Functions/MX_ColorCorrect
graph {
 properties {
  bExposeToLibrary=true
  description="Combines various adjustment nodes into one, artist-friendly color correction node."
  libraryCategoriesText [MaterialX]
 }
 nodes [
  FunctionOutput(None=Multiply(A=NamedRerouteUsage(), B=NamedRerouteUsage()))
  n1 = FunctionInput() {
   bUsePreviewValueAsDefault=true
   description="Rotates the color hue, with values wrapping at 0-1 boundaries."
   inputName=hue
   inputType=FunctionInput_Scalar
   sortPriority=1
  }
  n2 = Constant() {r=1}
  n3 = MaterialFunctionCall!(Input=FunctionInput() {description=The input color to be adjusted., inputName=Input}, amount=MaterialXAppend3Vector!(A=@n1, B=@n2, C=@n2)) {materialFunction=&/MX_HSVAdjust}
  n4 = FunctionInput() {
   bUsePreviewValueAsDefault=true
   description=Adjusts the input color saturation level.
   inputName=saturation
   inputType=FunctionInput_Scalar
   sortPriority=2
   previewValue=<1 0 0 1>
  }
  NamedRerouteDeclaration(None=LinearInterpolate(A=MaterialXLuminance!(None=@n3.Result), B=@n3.Result, Alpha=@n4)) {name=Saturate, nodeColor=<rgba 0.3594 0 1 1>}
  n5 = FunctionInput() {
   bUsePreviewValueAsDefault=true
   description=Applies a gamma correction to the color.
   inputName=gamma
   inputType=FunctionInput_Scalar
   sortPriority=3
   previewValue=<1 0 0 1>
  }
  NamedRerouteDeclaration(None=MaterialXRange!(None=NamedRerouteUsage(), InputLow, InputHigh, TargetLow, TargetHigh, Gamma=@n5, Clamp)) {name=gamma, nodeColor=<rgba 0 1 0.0391 1>}
  n6 = FunctionInput() {
   bUsePreviewValueAsDefault=true
   description="Multplier which increases or decreases color brightness by 2^value."
   inputName=exposure
   inputType=FunctionInput_Scalar
   sortPriority=8
  }
  NamedRerouteDeclaration(None=Power(Base=Constant() {r=2}, Exp=@n6)) {name=exposure, nodeColor=<rgba 0 0.2031 1 1>}
  n7 = FunctionInput() {
   bUsePreviewValueAsDefault=true
   description="Raise the dark color values, leaving the white values unchanged."
   inputName=lift
   inputType=FunctionInput_Scalar
   sortPriority=4
  }
  n8 = NamedRerouteDeclaration(None=@n7) {name=lift, nodeColor=<rgba 0.9219 1 0 1>}
  n9 = Add(A=Multiply(A=OneMinus(None=@n8), B=NamedRerouteUsage()), B=NamedRerouteUsage())
  n10 = FunctionInput() {
   bUsePreviewValueAsDefault=true
   description="Multiplier increases lighter color values, leaving black values unchanged."
   inputName=gain
   inputType=FunctionInput_Scalar
   sortPriority=5
   previewValue=<1 0 0 1>
  }
  n11 = FunctionInput() {
   bUsePreviewValueAsDefault=true
   description=Linearly increase or decrease the color contrast.
   inputName=contrast
   inputType=FunctionInput_Scalar
   sortPriority=6
   previewValue=<1 0 0 1>
  }
  n12 = FunctionInput() {
   bUsePreviewValueAsDefault=true
   description=Pivot value around which contrast applies. This value will not change as contrast is adjusted.
   inputName=contrastpivot
   inputType=FunctionInput_Scalar
   sortPriority=7
   previewValue=<0.5 0 0 1>
  }
  NamedRerouteDeclaration(None=MaterialXContrast!(None=Multiply(A=@n9, B=@n10), Amount=@n11, Pivot=@n12)) {name=contrast, nodeColor=<rgba 0.7344 1 0 1>}
 ]
}
