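# MaterialX library material function that converts hair melanin concentration and redness into absorption coefficients, using the d'Eon 2011 pigment mapping.
# Concentration becomes a melanin amount via -ln(1 - c), floored to avoid log of zero, and redness splits it into eumelanin and pheomelanin.
# Each amount scales the negative log of its pigment color input, and the clamped sum is the absorption output.
# Default pigment colors are linear Rec.709.
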MaterialFunction /InterchangeAssets/Functions/MX_DeonHairAbsorptionFromMelanin
graph {
 properties {
  bExposeToLibrary=true
  description="Converts the hair melanin parameterization to absorption coefficient based on pigments eumelanin and pheomelanin using the mapping method described in [^d'Eon2011]. The default of eumelanin_color and pheomelanin_color are lin_rec709 color converted from the constants[^d'Eon2011] via exp(-c). They may be transformed to scene-linear rendering color space."
  libraryCategoriesText [MaterialX]
 }
 nodes [
  n1 = FunctionInput() {
   bUsePreviewValueAsDefault=true
   description="constant from d'Eon et al. 2011, converted to color via exp(-c)"
   inputName=eumelanin_color
   sortPriority=2
   previewValue=<0.6577 0.4981 0.2541 1>
  }
  n2 = Multiply(A=Subtract(A=Constant(), B=Logarithm!(None=@n1)) {constA=0}, B=NamedRerouteUsage())
  n3 = FunctionInput() {
   bUsePreviewValueAsDefault=true
   description="constant from d'Eon et al. 2011, converted to color via exp(-c)"
   inputName=pheomelanin_color
   sortPriority=3
   previewValue=<0.8294 0.6703 0.3499 1>
  }
  n4 = Multiply(A=Subtract(A=Constant(), B=Logarithm(None=@n3)), B=NamedRerouteUsage())
  FunctionOutput(None=Max(A=Add(A=@n2, B=@n4), B=Constant())) {outputName=absorption}
  n5 = FunctionInput() {
   bUsePreviewValueAsDefault=true
   description="Amount of melanin affected to the output, range [0.0, 1.0]"
   inputName=melanin_concentration
   inputType=FunctionInput_Scalar
   previewValue=<0.25 0 0 1>
  }
  n6 = Max(A=OneMinus(None=@n5)) {constB=0.0001}
  NamedRerouteDeclaration(None=Subtract(A=Constant(), B=Logarithm(None=@n6)) {constA=0}) {name=melanin, nodeColor=<rgba 0 0.9062 1 1>}
  n7 = FunctionInput() {
   bUsePreviewValueAsDefault=true
   description="Amount of redness affected to the output, range [0.0, 1.0]."
   inputName=melanin_redness
   inputType=FunctionInput_Scalar
   sortPriority=1
   previewValue=<0.5 0 0 1>
  }
  NamedRerouteDeclaration(None=Multiply(A=OneMinus(None=@n7), B=NamedRerouteUsage())) {name=eumelanin, nodeColor=<rgba 0.4531 0 1 1>}
  NamedRerouteDeclaration(None=Multiply(A=@n7, B=NamedRerouteUsage())) {name=pheomelanin, nodeColor=<rgba 0 1 0.1328 1>}
 ]
}
