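# MaterialX library function that outputs scalar displacement (input times scale, plus bias) and a matching normal.
# HeightToNormalSmooth derives a normal from the scaled height and position input, then blends it with the input normal.
# MATERIAL_TANGENTSPACENORMAL selects between tangent-space BlendAngleCorrectedNormals and a world-space shortest-arc quaternion rotation.
# Static switch bRecomputeNormals, on by default, can pass the input normal through instead.
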MaterialFunction /InterchangeAssets/Functions/MX_Displacement
graph {
 properties {bExposeToLibrary=true, libraryCategoriesText [MaterialX]}
 nodes [
  n1 = FunctionInput() {
   bUsePreviewValueAsDefault=true
   description=Scalar displacement amount along the surface normal direction.
   inputName=displacement
   inputType=FunctionInput_Scalar
  }
  n2 = FunctionInput() {
   bUsePreviewValueAsDefault=true
   description=Scale factor for the displacement vector
   inputName=scale
   inputType=FunctionInput_Scalar
   sortPriority=1
   previewValue=<1 0 0 1>
  }
  n3 = Multiply(A=@n1, B=@n2)
  n4 = FunctionInput() {
   bUsePreviewValueAsDefault=true
   description=Bias factor for the displacement.
   inputName=bias
   inputType=FunctionInput_Scalar
   sortPriority=2
  }
  FunctionOutput(None=Add(A=@n3, B=@n4)) {outputName=Displacement}
  MaterialExpressionVertexNormalWS_1 = VertexNormalWS()
  MaterialExpressionConstant3Vector_0 = Constant3Vector() {constant=<rgba 0 0 1 0>}
  n5 = Custom!(SurfaceNormal=@MaterialExpressionVertexNormalWS_1, TSNormal=@MaterialExpressionConstant3Vector_0) {
   code="{\r\n##if MATERIAL_TANGENTSPACENORMAL\r\n\tconst float3 N = TSNormal;\r\n##else\r\n\tconst float3 N = SurfaceNormal;\r\n##endif\r\n\treturn N;\r\n}"
   description=GetMtlDefNormalFrame
   showCode=true
   inputs [
    {inputName=SurfaceNormal, input {expression=&self:MaterialExpressionVertexNormalWS_1}}
    {
     inputName=TSNormal
     input {expression=&self:MaterialExpressionConstant3Vector_0, mask=1, maskB=1, maskG=1, maskR=1}
    }
   ]
  }
  MaterialExpressionFunctionInput_3 = FunctionInput(Preview=@n5) {
   bUsePreviewValueAsDefault=true
   description=Normal that will be blended with the computed normal from the displacement
   inputName=normal
   sortPriority=4
   previewValue=<0 0 1 1>
  }
  n6 = FunctionInput(Preview=Multiply(A=LocalPosition().XYZ) {constB=0.01}) {bUsePreviewValueAsDefault=true, description=Position to compute the normals based on the displacement, inputName=position, sortPriority=3}
  MaterialExpressionMaterialFunctionCall_3 = MaterialFunctionCall!(Height=@n3, Absolute World Position=@n6, World Space Vertex Normals) {materialFunction=/Engine/Functions/Engine_MaterialFunctions03/Procedurals/HeightToNormalSmooth}
  MaterialExpressionReroute_0 = Reroute(None=@MaterialExpressionVertexNormalWS_1)
  MaterialExpressionCustom_4 = Custom!(
   BaseNormal=@MaterialExpressionFunctionInput_3
   AdditionalNormal=@MaterialExpressionMaterialFunctionCall_3.WorldSpaceNormal
   WorldSpaceNormal=@MaterialExpressionReroute_0
  ) {
   code="//https://blog.selfshadow.com/publications/blending-in-detail/ (equation 4 in nitty-gritty)\r\n// Build the shortest-arc quaternion\r\nfloat dot_WS_Base_1 = dot(WorldSpaceNormal, BaseNormal) + 1;\r\nfloat4 q = float4(cross(WorldSpaceNormal, BaseNormal), dot_WS_Base_1) / sqrt(2 * dot_WS_Base_1);\r\n// Rotate the normal\r\nreturn AdditionalNormal * (q.w * q.w - dot(q.xyz, q.xyz)) + 2 * q.xyz * dot(q.xyz, AdditionalNormal) + 2 * q.w * cross(q.xyz, AdditionalNormal);\r\n"
   description=BlendAngleCorrectedWorld
   inputs [
    {inputName=BaseNormal, input {expression=&self:MaterialExpressionFunctionInput_3}}
    {inputName=AdditionalNormal, input {expression=&self:MaterialExpressionMaterialFunctionCall_3}}
    {inputName=WorldSpaceNormal, input {expression=&self:MaterialExpressionReroute_0}}
   ]
  }
  MaterialExpressionMaterialFunctionCall_0 = MaterialFunctionCall(
   BaseNormal=@MaterialExpressionFunctionInput_3
   AdditionalNormal=Transform(None=@MaterialExpressionMaterialFunctionCall_3.WorldSpaceNormal) {transformSourceType=TRANSFORMSOURCE_World, transformType=TRANSFORM_Tangent}
  ) {materialFunction=&BlendAngleCorrectedNormals}
  n7 = Custom!(SurfaceNormal=@MaterialExpressionCustom_4, TSNormal=@MaterialExpressionMaterialFunctionCall_0.Result) {
   code="{\r\n##if MATERIAL_TANGENTSPACENORMAL\r\n\tconst float3 N = TSNormal;\r\n##else\r\n\tconst float3 N = SurfaceNormal;\r\n##endif\r\n\treturn N;\r\n}"
   description=GetMtlDefNormalFrame
   showCode=true
   inputs [
    {inputName=SurfaceNormal, input {expression=&self:MaterialExpressionCustom_4}}
    {inputName=TSNormal, input {expression=&self:MaterialExpressionMaterialFunctionCall_0}}
   ]
  }
  n8 = FunctionInput(Preview=StaticBool() {value=true}) {
   bUsePreviewValueAsDefault=true
   description=Recompute the normals after a displacement. Usefull for offline-renderers when there are procedural normals. Disable if you already have a normal map.
   inputName=bRecomputeNormals
   inputType=FunctionInput_StaticBool
   sortPriority=5
  }
  FunctionOutput(None=StaticSwitch(True=@n7, False=@MaterialExpressionFunctionInput_3, Value=@n8)) {outputName=Normal, sortPriority=1}
 ]
}
