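# MaterialX library material function that computes Gooch shading, a non-photorealistic warm-to-cool tone with a specular highlight.
# The pixel normal and normalized light direction input give a cool intensity that blends the warm and cool colors into diffuse.
# A Custom HLSL reflect node feeds a clamped view-reflection dot product, raised to shininess and scaled by specular intensity.
# Named reroutes link the stages, and the output sums two terms, evidently diffuse and specular.
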MaterialFunction /InterchangeAssets/Functions/MX_GoochShade
graph {
 properties {bExposeToLibrary=true, libraryCategoriesText [MaterialX]}
 nodes [
  FunctionOutput(None=Add(A=NamedRerouteUsage(), B=NamedRerouteUsage()))
  n1 = FunctionInput() {bUsePreviewValueAsDefault=true, inputName=light_direction, sortPriority=4, previewValue=<1 -0.5 -0.5 1>}
  NamedRerouteDeclaration(None=Normalize(VectorInput=@n1)) {name=LightDir, nodeColor=<rgba 1 0 0.5156 1>}
  NamedRerouteDeclaration(None=Subtract(B=CameraVectorWS()) {constA=0}) {name=ViewDir, nodeColor=<rgba 0 0.8359 1 1>}
  n2 = PixelNormalWS()
  NamedRerouteDeclaration(None=DotProduct(A=NamedRerouteDeclaration(None=@n2) {name=Normal, nodeColor=<rgba 0 1 0.2031 1>}, B=NamedRerouteUsage())) {name=NdotL, nodeColor=<rgba 1 0.8438 0 1>}
  n3 = NamedRerouteUsage()
  NamedRerouteDeclaration(None=Multiply(A=Add(A=@n3)) {constB=0.5}) {name=CoolIntensity, nodeColor=<rgba 0.5234 0 1 1>}
  MaterialExpressionNamedRerouteUsage_3 = NamedRerouteUsage()
  MaterialExpressionNamedRerouteUsage_4 = NamedRerouteUsage()
  n4 = Custom!(incident=@MaterialExpressionNamedRerouteUsage_3, normal=@MaterialExpressionNamedRerouteUsage_4) {
   code="return reflect(incident, normal);"
   description=reflect
   inputs [
    {inputName=incident, input {expression=&self:MaterialExpressionNamedRerouteUsage_3}}
    {inputName=normal, input {expression=&self:MaterialExpressionNamedRerouteUsage_4}}
   ]
  }
  NamedRerouteDeclaration(None=@n4) {name=ViewReflect, nodeColor=<rgba 1 0 0.1172 1>}
  n5 = Multiply(A=NamedRerouteUsage()) {constB=-1}
  n6 = NamedRerouteUsage()
  NamedRerouteDeclaration(None=Max(A=DotProduct(A=@n5, B=@n6)) {constB=0}) {name=VdotR, nodeColor=<rgba 0 0.4375 1 1>}
  n7 = FunctionInput() {bUsePreviewValueAsDefault=true, inputName=cool_color, sortPriority=1, previewValue=<0.3 0.3 0.8 1>}
  NamedRerouteDeclaration(None=LinearInterpolate(A=FunctionInput() {bUsePreviewValueAsDefault=true, inputName=warm_color, previewValue=<0.8 0.8 0.7 1>}, B=@n7, Alpha=NamedRerouteUsage())) {name=Diffuse, nodeColor=<rgba 0.7578 1 0 1>}
  n8 = NamedRerouteUsage()
  n9 = FunctionInput() {bUsePreviewValueAsDefault=true, inputName=shininess, inputType=FunctionInput_Scalar, sortPriority=3, previewValue=<64 0 0 1>}
  n10 = FunctionInput() {bUsePreviewValueAsDefault=true, inputName=specular_intensity, inputType=FunctionInput_Scalar, sortPriority=2, previewValue=<1 0 0 1>}
  NamedRerouteDeclaration(None=Multiply(A=Power(Base=@n8, Exp=@n9), B=@n10)) {name=Specular, nodeColor=<rgba 0.9219 0 1 1>}
 ]
}
