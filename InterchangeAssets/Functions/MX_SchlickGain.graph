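# MaterialX-category material function that computes Christophe Schlick's fast approximation of Perlin's gain function for scalar inputs x and r.
# It clamps r just inside the zero-to-one range, then derives a shared term a = (1/r - 2)(1 - 2x) through named reroutes.
# An If node outputs x/(a+1) below the 0.5 midpoint and (a-x)/(a-1) at or above it.
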MaterialFunction /InterchangeAssets/Functions/MX_SchlickGain
graph {
 properties {
  description="Christophe Schlick. “Fast Alternatives to Perlin’s Bias and Gain Functions”.// In Graphics Gems IV, Morgan Kaufmann, 1994, pages 401–403.// https://dept-info.labri.fr/~schlick/DOC/gem2.html"
  libraryCategoriesText [MaterialX]
 }
 nodes [
  n1 = NamedRerouteUsage()
  n2 = NamedRerouteUsage()
  n3 = Divide(A=Subtract(A=@n2, B=@n1), B=Subtract(A=@n2))
  n4 = Add(A=@n2)
  FunctionOutput(None=If(A=@n1, "A > B"=@n3, "A == B"=@n3, "A < B"=Divide(A=@n1, B=@n4)) {constB=0.5})
  n5 = Multiply(A=NamedRerouteDeclaration(None=FunctionInput() {inputName=x, inputType=FunctionInput_Scalar}) {name=x, nodeColor=<rgba 0 1 0.9766 1>}) {constB=2}
  n6 = Divide(B=Clamp(None=FunctionInput() {inputName=r, inputType=FunctionInput_Scalar, sortPriority=1}) {maxDefault=0.999, minDefault=0.001})
  NamedRerouteDeclaration(None=Multiply(A=OneMinus(None=@n5), B=Subtract(A=@n6) {constB=2})) {name=a, nodeColor=<rgba 1 0 0.7031 1>}
 ]
}
