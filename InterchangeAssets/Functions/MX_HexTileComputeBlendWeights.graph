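# MaterialX material function that computes three normalized blend weights for hex-tile texturing from luminance weights, tile weights and a falloff.
# Tile weights are raised to the seventh power, multiplied by luminance weights and divided by their component sum to form w.
# The components of w pass through MX_SchlickGain with falloff and are renormalized.
# An If node outputs plain w when falloff equals 0.5, otherwise the gain-adjusted weights.
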MaterialFunction /InterchangeAssets/Functions/MX_HexTileComputeBlendWeights
graph {
 expressions {
  MaterialExpressionConstant3Vector_0 {class=MaterialExpressionConstant3Vector, properties {constant=<rgba 7 7 7 0>}}
  MaterialExpressionConvert_0 {
   class=MaterialExpressionConvert
   inputs [None {from=MaterialExpressionNamedRerouteUsage_0}]
   properties {
    nodeName=Break Float3
    convertInputs [{
      type=Vector3
      expressionInput {
       expression=&self:MaterialExpressionNamedRerouteUsage_0
       inputName=None
       mask=0
       maskA=0
       maskB=0
       maskG=0
       maskR=0
       outputIndex=0
      }
     }]
    convertOutputs [{type=Scalar}, {type=Scalar}, {type=Scalar}]
   }
  }
  MaterialExpressionConvert_1 {
   class=MaterialExpressionConvert
   inputs [
    None {from=MaterialExpressionMaterialFunctionCall_0, output=Result}
    None {from=MaterialExpressionMaterialFunctionCall_2, output=Result}
    None {from=MaterialExpressionMaterialFunctionCall_1, output=Result}
   ]
   properties {
    nodeName=Make Vector3
    convertInputs [
     {
      type=Scalar
      expressionInput {
       expression=&self:MaterialExpressionMaterialFunctionCall_0
       inputName=None
       mask=0
       maskA=0
       maskB=0
       maskG=0
       maskR=0
       outputIndex=0
      }
     }
     {
      type=Scalar
      expressionInput {
       expression=&self:MaterialExpressionMaterialFunctionCall_2
       inputName=None
       mask=0
       maskA=0
       maskB=0
       maskG=0
       maskR=0
       outputIndex=0
      }
     }
     {
      type=Scalar
      expressionInput {
       expression=&self:MaterialExpressionMaterialFunctionCall_1
       inputName=None
       mask=0
       maskA=0
       maskB=0
       maskG=0
       maskR=0
       outputIndex=0
      }
     }
    ]
    convertOutputs [{type=Vector3}]
   }
  }
  MaterialExpressionDivide_0 {
   class=MaterialExpressionDivide
   inputs [
    A {from=MaterialExpressionMultiply_0}
    B {from=MaterialExpressionMaterialFunctionCall_4, output=f3}
   ]
  }
  MaterialExpressionDivide_1 {
   class=MaterialExpressionDivide
   inputs [
    A {from=MaterialExpressionConvert_1}
    B {from=MaterialExpressionMaterialFunctionCall_6, output=f3}
   ]
  }
  MaterialExpressionFunctionInput_0 {class=MaterialExpressionFunctionInput, inputs [Preview {from="", output=""}], properties {inputName=luminance_weights}}
  MaterialExpressionFunctionInput_1 {class=MaterialExpressionFunctionInput, inputs [Preview {from="", output=""}], properties {inputName=tile_weights, sortPriority=1}}
  MaterialExpressionFunctionInput_2 {
   class=MaterialExpressionFunctionInput
   inputs [Preview {from="", output=""}]
   properties {inputName=falloff, inputType=FunctionInput_Scalar, sortPriority=2}
  }
  MaterialExpressionFunctionOutput_0 {class=MaterialExpressionFunctionOutput, inputs [None {from=MaterialExpressionIf_0, output=""}]}
  MaterialExpressionIf_0 {
   class=MaterialExpressionIf
   inputs [
    A {from=MaterialExpressionFunctionInput_2, output=""}
    B {from="", output=""}
    "A > B" {from=MaterialExpressionDivide_1, output=""}
    "A == B" {from=MaterialExpressionNamedRerouteDeclaration_0, output=""}
    "A < B" {from=MaterialExpressionDivide_1, output=""}
   ]
   properties {constB=0.5}
  }
  MaterialExpressionMaterialFunctionCall_0 {
   class=MaterialExpressionMaterialFunctionCall
   inputs [
    x {from=MaterialExpressionConvert_0}
    r {from=MaterialExpressionFunctionInput_2}
   ]
   properties {materialFunction=&/MX_SchlickGain}
  }
  MaterialExpressionMaterialFunctionCall_1 {
   class=MaterialExpressionMaterialFunctionCall
   inputs [
    x {from=MaterialExpressionConvert_0}
    r {from=MaterialExpressionFunctionInput_2}
   ]
   properties {materialFunction=&MX_SchlickGain}
  }
  MaterialExpressionMaterialFunctionCall_2 {
   class=MaterialExpressionMaterialFunctionCall
   inputs [
    x {from=MaterialExpressionConvert_0}
    r {from=MaterialExpressionFunctionInput_2}
   ]
   properties {materialFunction=&MX_SchlickGain}
  }
  MaterialExpressionMaterialFunctionCall_4 {
   class=MaterialExpressionMaterialFunctionCall
   inputs [
    f2 {from=""}
    f3 {from=MaterialExpressionMultiply_0}
    f4 {from=""}
   ]
   properties {materialFunction=&AddComponents}
  }
  MaterialExpressionMaterialFunctionCall_6 {
   class=MaterialExpressionMaterialFunctionCall
   inputs [
    f2 {from=""}
    f3 {from=MaterialExpressionConvert_1}
    f4 {from=""}
   ]
   properties {materialFunction=&AddComponents}
  }
  MaterialExpressionMultiply_0 {
   class=MaterialExpressionMultiply
   inputs [
    A {from=MaterialExpressionFunctionInput_0}
    B {from=MaterialExpressionPower_0}
   ]
  }
  MaterialExpressionNamedRerouteDeclaration_0 {
   class=MaterialExpressionNamedRerouteDeclaration
   inputs [None {from=MaterialExpressionDivide_0}]
   properties {name=w, nodeColor=<rgba 1 0.5859 0 1>}
  }
  MaterialExpressionNamedRerouteUsage_0 {class=MaterialExpressionNamedRerouteUsage}
  MaterialExpressionPower_0 {
   class=MaterialExpressionPower
   inputs [
    Base {from=MaterialExpressionFunctionInput_1}
    Exp {from=MaterialExpressionConstant3Vector_0}
   ]
   properties {constExponent=7}
  }
 }
 properties {libraryCategoriesText [MaterialX]}
}
