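# MX_HextileHash is a MaterialX-category material function that hashes a Vector2 input p into a pseudo-random 2D value, by its name for hex tiling.
# Native nodes expand p to a Vector3, scale it by small constants, add an offset dot product, then combine Vector2 conversions through add, multiply and a final Frac.
# An unwired Custom node holds the equivalent vec-syntax code, though its first fract has no counterpart in the node chain.
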MaterialFunction /InterchangeAssets/Functions/MX_HextileHash
graph {
 properties {libraryCategoriesText [MaterialX]}
 nodes [
  Custom!(None) {
   code="    vec3 p3 = fract(vec3(p.x, p.y, p.x) * vec3(0.1031, 0.1030, 0.0973));\r\n    p3 += dot(p3, vec3(p3.y, p3.z, p3.x) + 33.33);\r\n    return fract((vec2(p3.x, p3.x) + vec2(p3.y, p3.z)) * vec2(p3.z, p3.y));"
   showCode=true
  }
  MaterialExpressionFunctionInput_0 = FunctionInput() {inputName=p, inputType=FunctionInput_Vector2}
  n1 = Convert(None=@MaterialExpressionFunctionInput_0) {
   nodeName=Make Vector3
   convertInputs [{
     type=Vector2
     expressionInput {
      expression=&self:MaterialExpressionFunctionInput_0
      inputName=None
      mask=0
      maskA=0
      maskB=0
      maskG=0
      maskR=0
      outputIndex=0
     }
    }]
   convertOutputs [{type=Vector3}]
  }
  MaterialExpressionMultiply_0 = Multiply(A=@n1, B=Constant3Vector() {constant=<rgba 0.1031 0.103 0.0973 0>})
  n2 = Convert(None=@MaterialExpressionMultiply_0) {
   nodeName=Make Vector3
   convertInputs [{
     type=Vector3
     expressionInput {
      expression=&self:MaterialExpressionMultiply_0
      inputName=None
      mask=0
      maskA=0
      maskB=0
      maskG=0
      maskR=0
      outputIndex=0
     }
    }]
   convertOutputs [{type=Vector3}]
  }
  MaterialExpressionAdd_1 = Add(A=@MaterialExpressionMultiply_0, B=DotProduct(A=@MaterialExpressionMultiply_0, B=Add(A=@n2) {constB=33.33}))
  n3 = Convert(None=@MaterialExpressionAdd_1) {
   nodeName=Make Vector2
   convertInputs [{
     type=Vector3
     expressionInput {
      expression=&self:MaterialExpressionAdd_1
      inputName=None
      mask=0
      maskA=0
      maskB=0
      maskG=0
      maskR=0
      outputIndex=0
     }
    }]
   convertOutputs [{type=Vector2}]
  }
  n4 = Convert(None=@MaterialExpressionAdd_1) {
   nodeName=Make Vector2
   convertInputs [{
     type=Vector3
     expressionInput {
      expression=&self:MaterialExpressionAdd_1
      inputName=None
      mask=0
      maskA=0
      maskB=0
      maskG=0
      maskR=0
      outputIndex=0
     }
    }]
   convertOutputs [{type=Vector2}]
  }
  n5 = Convert(None=@MaterialExpressionAdd_1) {
   nodeName=Make Vector2
   convertInputs [{
     type=Vector3
     expressionInput {
      expression=&self:MaterialExpressionAdd_1
      inputName=None
      mask=0
      maskA=0
      maskB=0
      maskG=0
      maskR=0
      outputIndex=0
     }
    }]
   convertOutputs [{type=Vector2}]
  }
  n6 = Add(A=@n4, B=@n5)
  FunctionOutput(None=Frac(None=Multiply(A=@n3, B=@n6)))
 ]
}
