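# MaterialX-category material function that, by its name and inputs, multiplies a 2D vector by a 2x2 matrix packed in a Vector4.
# A Convert node takes the Matrix2x2 input and declares two Vector2 outputs.
# Two dot products pair the Vec input with the Convert node, and their results are appended into the two-component output.
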MaterialFunction /InterchangeAssets/Functions/MX_Matrix2Mul
graph {
 properties {libraryCategoriesText [MaterialX]}
 nodes [
  n1 = FunctionInput() {inputName=Vec, inputType=FunctionInput_Vector2, sortPriority=1}
  MaterialExpressionFunctionInput_0 = FunctionInput() {inputName=Matrix2x2, inputType=FunctionInput_Vector4}
  n2 = Convert(None=@MaterialExpressionFunctionInput_0) {
   convertInputs [{
     type=Vector4
     defaultValue=<rgba 0 0 0 1>
     expressionInput {
      expression=&self:MaterialExpressionFunctionInput_0
      inputName=None
      mask=0
      maskA=0
      maskB=0
      maskG=0
      maskR=0
      outputIndex=0
     }
    }]
   convertOutputs [{type=Vector2}, {type=Vector2}]
  }
  FunctionOutput(None=AppendVector(A=DotProduct(A=@n1, B=@n2), B=DotProduct(A=@n1, B=@n2)))
 ]
}
