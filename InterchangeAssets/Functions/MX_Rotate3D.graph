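# MX_Rotate3D is a MaterialX library material function that rotates a 3D vector about a unit axis by an angle in degrees.
# It feeds the input, axis and amount into RotateAboutAxis with a zero pivot and a 360 period, so the angle reads as degrees.
# Because that node yields an offset, the original input is added back to output the rotated vector.
# The axis defaults to Y when unconnected.
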MaterialFunction /InterchangeAssets/Functions/MX_Rotate3D
graph {
 properties {bExposeToLibrary=true, description=Rotate the incoming 3D vector about the specified unit axis vector., libraryCategoriesText [MaterialX]}
 nodes [
  n1 = FunctionInput() {
   bUsePreviewValueAsDefault=true
   description=The unit axis vector to rotate the input around.
   inputName=axis
   sortPriority=2
   previewValue=<0 1 0 1>
  }
  n2 = FunctionInput() {
   bUsePreviewValueAsDefault=true
   description="The angle to rotate, specified in degrees. Positive values rotate counter-clockwise"
   inputName=amount
   sortPriority=1
  }
  n3 = Constant3Vector()
  n4 = FunctionInput() {description=The input vector to rotate, inputName=Input}
  FunctionOutput(None=Add(A=RotateAboutAxis(NormalizedRotationAxis=@n1, RotationAngle=@n2, PivotPoint=@n3, Position=@n4) {period=360}, B=@n4))
 ]
}
