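# MaterialX library material function that samples a lat-long (equirectangular) environment image along a view direction and outputs linear-color RGB.
# MX_Rotate3D rotates the direction, which defaults to local position, by a longitudinal offset in degrees around an axis input defaulting to negative Z.
# The rotated vector indexes the texture as a TextureCube, so the graph contains no explicit lat-long UV math.
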MaterialFunction /InterchangeAssets/Functions/MX_LatLongImage
graph {
 properties {bExposeToLibrary=true, libraryCategoriesText [MaterialX]}
 nodes [
  n1 = FunctionInput(Preview=LocalPosition().XYZ) {
   bUsePreviewValueAsDefault=true
   description=The view direction determining the value sampled from the projected equiangular map. Default to the local position.
   inputName=viewdir
   sortPriority=1
  }
  n2 = FunctionInput() {
   bUsePreviewValueAsDefault=true
   description="The longitudinal sampling offset, in degrees"
   inputName=rotation
   inputType=FunctionInput_Scalar
   sortPriority=2
  }
  n3 = FunctionInput() {bUsePreviewValueAsDefault=true, inputName=axis, sortPriority=3, previewValue=<0 0 -1 1>}
  n4 = FunctionInput(Preview=TextureObject() {texture=/Engine/EditorMaterials/AssetViewer/EpicQuadPanorama_CC+EV1.EpicQuadPanorama_CC+EV1}) {
   bUsePreviewValueAsDefault=true
   description=Equirectangular texture sampled as a TextureCube.
   inputName=Texture
   inputType=FunctionInput_TextureCube
  }
  FunctionOutput(None=TextureSample(UVs=MaterialFunctionCall!(Input=@n1, amount=@n2, axis=@n3) {materialFunction=&/MX_Rotate3D}.Result, Tex=@n4) {samplerType=SAMPLERTYPE_LinearColor}.RGB)
 ]
}
