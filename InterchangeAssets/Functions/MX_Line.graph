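# Material function MX_Line, which returns 1 where 2D coordinates lie within a radius of a line segment between point1 and point2, otherwise 0.
# It subtracts a center offset from the coordinates, which default to the first UVs, and projects them onto the segment using a saturated dot-product ratio.
# The distance to that clamped closest point is compared with the radius in an If, giving a hard-edged, round-capped line mask.
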MaterialFunction /InterchangeAssets/Functions/MX_Line
graph {
 properties {description="Returns 1 if texcoord is at less than radius distance from a line segment defined by point1 and point2; otherwise returns 0."}
 nodes [
  n1 = FunctionInput() {
   bUsePreviewValueAsDefault=true
   description="The radius or \"half thickness\" of the line"
   inputName=radius
   inputType=FunctionInput_Scalar
   sortPriority=2
   previewValue=<0.1 0 0 1>
  }
  n2 = Constant() {r=1}
  FunctionOutput(None=If(A=NamedRerouteUsage(), B=@n1, "A > B"=Constant(), "A == B"=@n2, "A < B"=@n2))
  n3 = FunctionInput(Preview=TextureCoordinate()) {
   bUsePreviewValueAsDefault=true
   description=The input 2d space. Default is the first texture coordinates.
   inputName=Coordinates
   inputType=FunctionInput_Vector2
  }
  n4 = FunctionInput() {
   bUsePreviewValueAsDefault=true
   description=An offset value added to both the point1 and point2 coordinates
   inputName=center
   inputType=FunctionInput_Vector2
   sortPriority=1
  }
  n5 = FunctionInput() {
   bUsePreviewValueAsDefault=true
   description=The UV coordinate of the first endpoint
   inputName=point1
   inputType=FunctionInput_Vector2
   sortPriority=3
   previewValue=<0.25 0.25 0 1>
  }
  NamedRerouteDeclaration(None=Subtract(A=Subtract(A=@n3, B=@n4), B=@n5)) {name=p_a, nodeColor=<rgba 0 0.3437 1 1>}
  n6 = NamedRerouteUsage()
  n7 = NamedRerouteUsage()
  n8 = FunctionInput() {
   bUsePreviewValueAsDefault=true
   description=The UV coordinate of the second endpoint
   inputName=point2
   inputType=FunctionInput_Vector2
   sortPriority=4
   previewValue=<0.75 0.75 0 1>
  }
  n9 = NamedRerouteDeclaration(None=Subtract(A=@n8, B=@n5)) {name=b_a, nodeColor=<rgba 0.6641 1 0 1>}
  n10 = Saturate(None=Divide(A=DotProduct(A=@n6, B=@n7), B=DotProduct(A=@n9, B=@n9)))
  n11 = NamedRerouteUsage()
  NamedRerouteDeclaration(None=Distance(A=NamedRerouteUsage(), B=Multiply(A=@n10, B=@n11))) {name=distance, nodeColor=<rgba 1 0.375 0 1>}
 ]
}
