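# Material function that converts a color from linear Display P3 to linear Rec709, exposed to the material library under MaterialX.
# The single input passes through a Transform3x3Matrix call with three constant basis vectors that form the gamut conversion matrix and a zero offset, and the transformed vector is the output.
# The math is purely linear, with no transfer curve or clamping, so out-of-gamut colors can produce negative values.
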MaterialFunction /InterchangeAssets/Functions/MX_LinearDisplayP3ToRec709
graph {
 properties {
  bExposeToLibrary=true
  description=Color Transform from Linear Display P3 to Linear Rec709
  userExposedCaption=LinearDisplayP3ToRec709
  libraryCategoriesText [MaterialX]
 }
 nodes [FunctionOutput(None=MaterialFunctionCall(
    VectorToTransform=FunctionInput() {bUsePreviewValueAsDefault=true, inputName=Input}
    BasisX=Constant3Vector() {constant=<rgba 1.2249 -0.0421 -0.0196 0>}
    BasisY=Constant3Vector() {constant=<rgba -0.2249 1.0421 -0.0786 0>}
    BasisZ=Constant3Vector() {constant=<rgba 0 0 1.0983 0>}
    BasisPosition=Constant3Vector()
   ) {materialFunction=&Transform3x3Matrix}.Transformed Vector)]
}
