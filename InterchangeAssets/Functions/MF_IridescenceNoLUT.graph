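# Material function that computes view-dependent thin-film iridescence analytically in custom HLSL, outputting SpecularColor, EdgeSpecularColor and Fresnel.
# Normal-dot-camera stands in for NoV, NoL and VoH, feeding an Airy-style interference sum converted from XYZ to RGB, then back-solved into Schlick F0/F90.
# Static switch UseComplexIOR selects a costlier polarized conductor model, calling an F0/F90-to-eta/kappa conversion, or a simplified unpolarized version using F0 as base reflectance.
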MaterialFunction /InterchangeAssets/Functions/MF_IridescenceNoLUT
graph {
 properties {bExposeToLibrary=true}
 nodes [
  MaterialExpressionDotProduct_2 = DotProduct(A=FunctionInput(Preview=PixelNormalWS()) {bUsePreviewValueAsDefault=true, inputName=NormalWS}, B=CameraVectorWS())
  MaterialExpressionFunctionInput_5 = FunctionInput() {bUsePreviewValueAsDefault=true, inputName=BaseF0, sortPriority=10, previewValue=<0.04 0.04 0.04 1>}
  MaterialExpressionFunctionInput_6 = FunctionInput() {bUsePreviewValueAsDefault=true, inputName=BaseF90, sortPriority=20, previewValue=<1 1 1 1>}
  MaterialExpressionFunctionInput_3 = FunctionInput() {
   bUsePreviewValueAsDefault=true
   description=Index of refraction of the thin film
   inputName=n2
   inputType=FunctionInput_Scalar
   sortPriority=40
   previewValue=<1 0 0 1>
  }
  MaterialExpressionFunctionInput_2 = FunctionInput() {bUsePreviewValueAsDefault=true, inputName=Thickness, inputType=FunctionInput_Scalar, sortPriority=30}
  MaterialExpressionFunctionInput_0 = FunctionInput() {
   bUsePreviewValueAsDefault=true
   description="Complex IOR (Eta) - only used if UseComplexIOR is true"
   inputName=n3
   sortPriority=50
   previewValue=<1.6 1.6 1.6 1>
  }
  MaterialExpressionFunctionInput_1 = FunctionInput() {
   bUsePreviewValueAsDefault=true
   description="Complex Extinction Coefficient (Kappa) - only used if UseComplexIOR is true"
   inputName=k
   sortPriority=60
  }
  n1 = Custom!(
   NoV=@MaterialExpressionDotProduct_2
   NoL=@MaterialExpressionDotProduct_2
   VoH=@MaterialExpressionDotProduct_2
   F0=@MaterialExpressionFunctionInput_5
   F90=@MaterialExpressionFunctionInput_6
   ThinFilmIOR=@MaterialExpressionFunctionInput_3
   ThinFilmThickness=@MaterialExpressionFunctionInput_2
   n3=@MaterialExpressionFunctionInput_0
   k3=@MaterialExpressionFunctionInput_1
  ) {
   code="//Main\r\n// Translate F0/F90 to eta/kappa form\r\n//float3 n3 = 1;\r\n//float3 k3 = 0;\r\nComputeComplexIORFromF0AndEdgeTintA(F0, F90, n3, k3);\r\n\r\nfloat3 Avg = 1.f / 3.f;\r\nfloat eta2\t = clamp(ThinFilmIOR, 1.f, 3.f);\r\nfloat eta3\t = clamp(dot(Avg,n3), 1.f, 3.f);\r\nfloat kappa3 = clamp(dot(Avg,k3), 1.f, 5.f);\r\nfloat Dinc   = ThinFilmThickness * 10; // ThinFilmThickness is in micrometer. Range is 0 - 10um i.e, 0 - 10 000nm\r\n\r\n// Force eta_2 -> 1.0 when Dinc -> 0.0\r\nconst float eta_2 = lerp(1.0, eta2, smoothstep(0.0, 0.03, Dinc));\r\n\r\n// Compute dot products\r\nconst bool bIsValid = NoL > 0 &&&& NoV > 0;\r\n\r\nfloat3 I = 0;\r\nif (bIsValid)\r\n{\r\n\tfloat cosTheta1 = VoH;\r\n\tfloat cosTheta2 = sqrt(1.0 - Pow2(1.0 / eta_2) * (1 - Pow2(cosTheta1)));\r\n\r\n\t// First interface\r\n\tfloat2 R12, phi12;\r\n\tThinFilmFresnelDielectricA(cosTheta1, 1.0, eta_2, R12, phi12);\r\n\tfloat2 R21 = R12;\r\n\tfloat2 T121 = float2(1, 1) - R12;\r\n\tfloat2 phi21 = float2(PI, PI) - phi12;\r\n\r\n\t// Second interface\r\n\tfloat2 R23, phi23;\r\n\tThinFilmFresnelConductorA(cosTheta2, eta_2, eta3, kappa3, R23, phi23);\r\n\r\n\t// Phase shift\r\n\tfloat OPD = Dinc * cosTheta2;\r\n\tfloat2 phi2 = phi21 + phi23;\r\n\r\n\t// Compound terms\r\n\tfloat2 R123 = R12 * R23;\r\n\tfloat2 r123 = sqrt(R123);\r\n\tfloat2 Rs = Pow2(T121) * R23 / (1 - R123);\r\n\r\n\t// Reflectance term for m=0 (DC term amplitude)\r\n\tfloat2 C0 = R12 + Rs;\r\n\tfloat3 S0 = ThinFilmEvalSensitivityA(0.0, 0.0);\r\n\tI += ThinFilmDepolA(C0) * S0;\r\n\r\n\t// Reflectance term for m>0 (pairs of diracs)\r\n\tfloat2 Cm = Rs - T121;\r\n\tfor (int m = 1; m <= 3; ++m)\r\n\t{\r\n\t\tCm *= r123;\r\n\t\tfloat3 SmS = 2.0 * ThinFilmEvalSensitivityA(m * OPD, m * phi2.x);\r\n\t\tfloat3 SmP = 2.0 * ThinFilmEvalSensitivityA(m * OPD, m * phi2.y);\r\n\t\tI += ThinFilmDepolColorA(Cm.x * SmS, Cm.y * SmP);\r\n\t}\r\n\r\n\t// Convert back to RGB reflectance\r\n\t// XYZ to CIE 1931 RGB color space (using neutral E illuminant)\r\n\tconst float3x3 ThinFilm_XYZ_TO_RGB = float3x3(2.3706743, -0.5138850, 0.0052982, -0.9000405, 1.4253036, -0.0146949, -0.4706338, 0.0885814, 1.0093968);\r\n\tI = clamp(mul(I, ThinFilm_XYZ_TO_RGB), float3(0, 0, 0), float3(1, 1, 1));\r\n}\r\n\r\nfloat3 RThinFilm = I;\r\n\r\n// Compute a F0 which match evaluation Thin-Film reflectance for the current view angle using Schlick's Fresnel\r\nconst float FReference = min(0.99f, SchlickB(0, 1.0f, NoV).x);\r\nRThinFilm = max(RThinFilm, FReference);\r\n\r\nfloat3 F0_Thin = (RThinFilm - F90 * FReference) / (1 - FReference);\r\n\r\n// Ensure the generated F0 doesn't cause micro-oclusion\r\nF0_Thin = max(GetF0MicroOcclusionThreshold(), F0_Thin);\r\n\r\n// If micro-occlusion was set, re-apply it.\t\t\r\nconst float3 MicroOcclusion = F0RGBToMicroOcclusion(F0);\r\nF0_Thin *= MicroOcclusion;\r\n\r\nOutF0 = F0_Thin;\r\nOutF90 = max(F0_Thin, F90);\r\n\r\nreturn lerp( OutF0, OutF90, FReference );\r\n"
   description=ThinFilmComplex
   showCode=true
   additionalOutputs [{outputName=OutF0, outputType=CMOT_Float3}, {outputName=OutF90, outputType=CMOT_Float3}]
   includeFilePaths [/Plugin/InterchangeAssets/ThinFilmFunctions.usf]
   inputs [
    {inputName=NoV, input {expression=&self:MaterialExpressionDotProduct_2}}
    {inputName=NoL, input {expression=&self:MaterialExpressionDotProduct_2}}
    {inputName=VoH, input {expression=&self:MaterialExpressionDotProduct_2}}
    {inputName=F0, input {expression=&self:MaterialExpressionFunctionInput_5}}
    {inputName=F90, input {expression=&self:MaterialExpressionFunctionInput_6}}
    {inputName=ThinFilmIOR, input {expression=&self:MaterialExpressionFunctionInput_3}}
    {inputName=ThinFilmThickness, input {expression=&self:MaterialExpressionFunctionInput_2}}
    {inputName=n3, input {expression=&self:MaterialExpressionFunctionInput_0}}
    {inputName=k3, input {expression=&self:MaterialExpressionFunctionInput_1}}
   ]
  }
  n2 = Custom!(
   NoV=@MaterialExpressionDotProduct_2
   NoL=@MaterialExpressionDotProduct_2
   VoH=@MaterialExpressionDotProduct_2
   F0=@MaterialExpressionFunctionInput_5
   F90=@MaterialExpressionFunctionInput_6
   ThinFilmIOR=@MaterialExpressionFunctionInput_3
   ThinFilmThickness=@MaterialExpressionFunctionInput_2
  ) {
   code="float Dinc = ThinFilmThickness * 10; // ThinFilmThickness is in micrometer. Range is 0 - 10um i.e, 0 - 10 000nm\r\n\r\n// Force eta2 -> 1.0 when Dinc -> 0.0\r\nconst float eta2 = lerp(1.0, ThinFilmIOR, smoothstep(0.0, 0.03, Dinc));\r\n\r\n// Compute dot products\r\nconst bool bIsValid = NoL > 0 &&&& NoV > 0;\r\n\r\nfloat3 I = 0;\r\nif (bIsValid)\r\n{\r\n\tfloat cosTheta1 = VoH;\r\n\tfloat cosTheta2 = sqrt(1.0 - Pow2(1.0 / eta2) * (1 - Pow2(cosTheta1)));\r\n\r\n\t// First interface is dieletric so it is achromatic\r\n\tfloat R12 = SchlickA(DielectricToF0A(ThinFilmIOR), VoH).x;\r\n\tfloat phi12 = 0;\r\n\tfloat R21 = R12;\r\n\tfloat T121 = 1 - R12;\r\n\tfloat phi21 = PI - phi12;\r\n\r\n\t// Second interface is either conductor or dieletric, so can be chromatic. \r\n\t// Ideally we should recompute F0 (R23) by taking into account the thin layer IOR\r\n\tfloat3 R23 = F0;\r\n\tfloat phi23 = 0;\r\n\r\n\t// Phase shift\r\n\tfloat OPD = Dinc * cosTheta2;\r\n\tfloat phi2 = phi21 + phi23;\r\n\r\n\t// Compound terms\r\n\tfloat3 R123 = R12 * R23;\r\n\tfloat3 r123 = sqrt(R123);\r\n\tfloat3 Rs = Pow2(T121) * R23 / (1 - R123);\r\n\r\n\t// Reflectance term for m=0 (DC term amplitude)\r\n\tfloat3 C0 = R12 + Rs;\r\n\tfloat3 S0 = ThinFilmEvalSensitivityA(0.0, 0.0);\r\n\tI += C0 * S0;\r\n\r\n\t// Reflectance term for m>0 (pairs of diracs)\r\n\t// Simplified based as we don't take into account polarization\r\n\tfloat3 Cm = Rs - T121;\r\n\tfor (int m = 1; m <= 3; ++m)\r\n\t{\r\n\t\tCm *= r123;\r\n\t\tI += Cm * 2.0 * ThinFilmEvalSensitivityA(m * OPD, m * phi2);\r\n\t}\r\n\r\n\t// Convert back to RGB reflectance\r\n\t// XYZ to CIE 1931 RGB color space (using neutral E illuminant)\r\n\tconst float3x3 ThinFilm_XYZ_TO_RGB = float3x3(2.3706743, -0.5138850, 0.0052982, -0.9000405, 1.4253036, -0.0146949, -0.4706338, 0.0885814, 1.0093968);\r\n\tI = saturate(mul(I, ThinFilm_XYZ_TO_RGB));\r\n}\r\n\r\nfloat3 RThinFilm = I;\r\n\r\n// Compute a F0 which match evaluation Thin-Film reflectance for the current view angle using Schlick's Fresnel\r\nconst float FReference = min(0.99f, SchlickB(0, 1.0f, NoV).x);\r\nRThinFilm = max(RThinFilm, FReference);\r\n\r\nfloat3 F0_Thin = (RThinFilm - F90 * FReference) / (1 - FReference);\r\n\r\n// Ensure the generated F0 doesn't cause micro-oclusion\r\nF0_Thin = max(GetF0MicroOcclusionThreshold(), F0_Thin);\r\n\r\n// If micro-occlusion was set, re-apply it.\t\t\r\nconst float3 MicroOcclusion = F0RGBToMicroOcclusion(F0);\r\nF0_Thin *= MicroOcclusion;\r\n\r\nOutF0 = F0_Thin;\r\nOutF90 = max(F0_Thin, F90);\r\n\r\nreturn lerp( OutF0, OutF90, FReference );\r\n\r\n\r\n"
   description=ThinFilm
   showCode=true
   additionalOutputs [{outputName=OutF0, outputType=CMOT_Float3}, {outputName=OutF90, outputType=CMOT_Float3}]
   includeFilePaths [&usf]
   inputs [
    {inputName=NoV, input {expression=&self:MaterialExpressionDotProduct_2}}
    {inputName=NoL, input {expression=&self:MaterialExpressionDotProduct_2}}
    {inputName=VoH, input {expression=&self:MaterialExpressionDotProduct_2}}
    {inputName=F0, input {expression=&self:MaterialExpressionFunctionInput_5}}
    {inputName=F90, input {expression=&self:MaterialExpressionFunctionInput_6}}
    {inputName=ThinFilmIOR, input {expression=&self:MaterialExpressionFunctionInput_3}}
    {inputName=ThinFilmThickness, input {expression=&self:MaterialExpressionFunctionInput_2}}
   ]
  }
  n3 = FunctionInput(Preview=StaticBool()) {
   bUsePreviewValueAsDefault=true
   description="Should be used if the film is atop a conductor and accuracy is needed.  Otherwise, save the 50+ instructions and use the simple IOR."
   inputName=UseComplexIOR
   inputType=FunctionInput_StaticBool
   sortPriority=40
  }
  FunctionOutput(None=StaticSwitch(True=@n1.OutF0, False=@n2.OutF0, Value=@n3)) {outputName=SpecularColor}
  FunctionOutput(None=StaticSwitch(True=@n1.OutF90, False=@n2.OutF90, Value=@n3)) {outputName=EdgeSpecularColor, sortPriority=10}
  FunctionOutput(None=StaticSwitch(True=@n1.return, False=@n2.return, Value=@n3)) {outputName=Fresnel, sortPriority=20}
 ]
}
