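# Library material function computing an Oren-Nayar-style rough diffuse term from Light, View and Roughness inputs and the world-space pixel normal.
# Squared roughness yields coefficients a and b, combining as a + b·max(0, ga).
# That sum is multiplied by clamped NoL and a sine-product over max(NoL, NoV) factor, which scales both terms.
# The ga dot product uses NoL·N−N instead of projected Light, so it cancels for unit normals.
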MaterialFunction /InterchangeAssets/Functions/MF_OrenNayerView
graph {
 properties {bExposeToLibrary=true}
 nodes [
  n1 = PixelNormalWS() {desc=N}
  n2 = DotProduct(A=FunctionInput(Preview=ReflectionVectorWS()) {inputName=Light}, B=@n1) {desc=NoL}
  n3 = Max(A=Constant(), B=@n2)
  n4 = FunctionInput() {bUsePreviewValueAsDefault=true, inputName=Roughness, inputType=FunctionInput_Scalar}
  n5 = Multiply(A=@n4, B=@n4)
  n6 = Divide(A=@n5, B=Add(A=@n5, B=Constant() {r=0.57}))
  n7 = Constant() {r=0.5}
  n8 = Divide(A=@n5, B=Add(A=@n5, B=Constant() {r=0.09}))
  n9 = Constant() {r=0.45}
  n10 = Constant()
  n11 = FunctionInput(Preview=CameraVectorWS()) {inputName=View}
  n12 = DotProduct(A=@n1, B=@n11) {desc=NoV}
  n13 = DotProduct(A=Subtract(A=Multiply(A=@n2, B=@n1), B=@n1), B=Subtract(A=@n11, B=Multiply(A=@n1, B=@n12))) {desc=ga}
  n14 = Add(A=OneMinus(None=Multiply(A=@n6, B=@n7)) {desc=a}, B=Multiply(A=Multiply(A=@n8, B=@n9) {desc=b}, B=Max(A=@n10, B=@n13)))
  n15 = Multiply(A=@n2, B=@n2)
  n16 = Multiply(A=@n12, B=@n12)
  n17 = SquareRoot(None=Multiply(A=OneMinus(None=@n15), B=OneMinus(None=@n16)))
  n18 = Max(A=@n2, B=@n12)
  FunctionOutput(None=Multiply(A=Multiply(A=@n3, B=@n14), B=Divide(A=@n17, B=@n18)))
 ]
}
