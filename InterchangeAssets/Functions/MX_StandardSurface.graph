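# MaterialX Standard Surface material function, exposed to the library, converting that model's inputs into Unreal shading outputs.
# Base color uses an Oren-Nayar approximation with a placeholder camera-position light vector; specular derives from IOR via Schlick, with thin-film iridescence computed through MF_IridescenceNoLUT.
# Sheen, subsurface and coat share outputs, with non-zero sheen taking priority; coat is suppressed by metalness.
# Named reroutes carry rotated anisotropy tangents and thin-film terms; thin_walled is unconnected.
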MaterialFunction /InterchangeAssets/Functions/MX_StandardSurface
graph {
 properties {bExposeToLibrary=true, libraryCategoriesText [MaterialX]}
 nodes [
  FunctionInput(Preview=StaticBool()) {
   bUsePreviewValueAsDefault=true
   description=If true the surface is double-sided and represents an infinitely thin shell. Suitable for thin objects such as tree leaves or paper.
   inputName=thin_walled
   inputType=FunctionInput_StaticBool
   sortPriority=39
  }
  n1 = FunctionInput(Preview=Constant()) {
   bUsePreviewValueAsDefault=true
   description=The blend between diffuse reflection and subsurface scattering. A value of 1.0 indicates full subsurface scattering and a value 0 for diffuse reflection only.
   inputName=subsurface
   inputType=FunctionInput_Scalar
   sortPriority=17
   previewValue=<0 0 0 0>
  }
  n2 = FunctionInput(Preview=Constant3Vector() {constant=<rgba 1 1 1 0>}) {
   bUsePreviewValueAsDefault=true
   description=The mean free path. The distance which light can travel before being scattered inside the surface.
   inputName=subsurface_radius
   sortPriority=19
   previewValue=<1 1 1 1>
  }
  n3 = FunctionInput(Preview=Constant() {r=1}) {
   bUsePreviewValueAsDefault=true
   description=Scalar weight for the subsurface radius value.
   inputName=subsurface_scale
   inputType=FunctionInput_Scalar
   sortPriority=20
   previewValue=<1 1 1 1>
  }
  n4 = Multiply(A=@n2, B=@n3)
  FunctionOutput(None=Multiply(A=@n1, B=@n4)) {outputName=Opacity, sortPriority=6}
  n5 = FunctionInput(Preview=Constant3Vector() {constant=<rgba 0.8 0.8 0.8 0>}) {bUsePreviewValueAsDefault=true, description=Color of the diffuse reflection., inputName=base_color, sortPriority=1, previewValue=<1 1 1 1>}
  n6 = FunctionInput(Preview=Constant()) {
   bUsePreviewValueAsDefault=true
   description=Roughness of the diffuse reflection. Higher values cause the surface to appear flatter and darker.
   inputName=diffuse_roughness
   inputType=FunctionInput_Scalar
   sortPriority=2
   previewValue=<0 0 0 0>
  }
  n7 = MaterialFunctionCall!(Light=Normalize(VectorInput=CameraPositionWS()), View=CameraVectorWS(), Roughness=@n6) {desc=Mega Hacky - needs proper light vector, materialFunction=&/MF_OrenNayerView}
  n8 = Saturate(None=Add(A=Multiply(A=@n5, B=@n7.Result), B=@n5))
  n9 = FunctionInput(Preview=Constant()) {
   bUsePreviewValueAsDefault=true
   description="Specifies how metallic the material appears. At its maximum, the surface behaves like a metal, using fully specular reflection and complex fresnel."
   inputName=metalness
   inputType=FunctionInput_Scalar
   sortPriority=3
   previewValue=<0 0 0 0>
  }
  n10 = Multiply(A=NamedRerouteUsage(), B=Multiply(A=NamedRerouteUsage(), B=LinearInterpolate(A=@n8, B=@n5, Alpha=@n9))) {desc=Thinfilm doesnt add up if base color is black}
  n11 = Multiply(
   A=FunctionInput(Preview=Constant3Vector() {constant=<rgba 1 1 1 0>}) {bUsePreviewValueAsDefault=true, inputName=coat_color, sortPriority=26}
   B=@n10
  )
  n12 = FunctionInput(Preview=Constant()) {
   bUsePreviewValueAsDefault=true
   description=The weight of a reflective clear-coat layer on top of the material. Use for materials such as car paint or an oily layer.
   inputName=coat
   inputType=FunctionInput_Scalar
   sortPriority=25
   previewValue=<0 0 0 0>
  }
  n13 = NamedRerouteDeclaration(None=@n12) {name=Coat, nodeColor=<rgba 0.1953 0 1 1>}
  FunctionOutput(None=NamedRerouteDeclaration(None=LinearInterpolate(A=@n10, B=@n11, Alpha=@n13)) {name=BaseColor, nodeColor=<rgba 1 0.2578 0 1>}) {outputName=Base Color}
  n14 = FunctionInput(Preview=Constant()) {
   bUsePreviewValueAsDefault=true
   description=The weight of a sheen layer that can be used to approximate microfibers or fabrics such as velvet and satin.
   inputName=sheen
   inputType=FunctionInput_Scalar
   sortPriority=22
   previewValue=<0 0 0 0>
  }
  n15 = NamedRerouteUsage()
  n16 = FunctionInput(Preview=Constant3Vector() {constant=<rgba 1 1 1 0>}) {bUsePreviewValueAsDefault=true, description=The color of the sheen layer., inputName=sheen_color, sortPriority=23, previewValue=<1 1 1 1>}
  n17 = FunctionInput(Preview=Constant() {r=0.3}) {
   bUsePreviewValueAsDefault=true
   description=The roughness of the sheen layer.
   inputName=sheen_roughness
   inputType=FunctionInput_Scalar
   sortPriority=24
   previewValue=<0.3 0.3 0.3 0.3>
  }
  n18 = Multiply(A=OneMinus(None=@n9) {desc=cannot have a coat with metallic}, B=@n13)
  n19 = Constant3Vector() {constant=<rgba 0 0 1 0>}
  n20 = NamedRerouteDeclaration(None=FunctionInput(Preview=Constant3Vector() {constant=<rgba 0 0 1 0>}) {bUsePreviewValueAsDefault=true, description=Input geometric normal, inputName=normal, sortPriority=40}) {name=GeometryNormal, nodeColor=<rgba 1 0 0.8437 1>}
  n21 = If(
   A=@n18
   "A > B"=NamedRerouteDeclaration(None=FunctionInput(Preview=@n19) {bUsePreviewValueAsDefault=true, description=Input normal for clear-coat layer, inputName=coat_normal, sortPriority=31}) {name=GeometryCoatNormal, nodeColor=<rgba 0 1 0.8359 1>}
   "A == B"=@n20
   "A < B"=@n20
  )
  n22 = Fresnel(ExponentIn=LinearInterpolate(Alpha=@n17) {constA=5, constB=0.8, desc=SheenRoughness}, Normal=Transform(None=@n21))
  n23 = If(A=@n14, B=Constant(), "A > B"=@n14, "A == B"=@n18, "A < B"=@n18)
  n24 = FunctionInput(Preview=Constant3Vector() {constant=<rgba 1 1 1 0>}) {
   bUsePreviewValueAsDefault=true
   description=The color of the subsurface scattering effect.
   inputName=subsurface_color
   sortPriority=18
   previewValue=<1 1 1 1>
  }
  n25 = Multiply(A=Multiply(A=@n24, B=@n1), B=@n4)
  n26 = If(
   A=@n14
   "A > B"=LinearInterpolate(A=@n15, B=LinearInterpolate(A=@n15, B=@n16, Alpha=@n22), Alpha=@n23) {desc=Sheen}
   "A == B"=@n25
   "A < B"=@n25
  )
  FunctionOutput(None=@n26) {outputName=SheenColor, sortPriority=11}
  FunctionOutput(None=@n26) {outputName=SubsurfaceColor, sortPriority=13}
  FunctionOutput(None=@n23) {outputName=ClearCoat, sortPriority=14}
  n27 = FunctionInput() {bUsePreviewValueAsDefault=true, description=The opacity of the entire material., inputName=opacity, sortPriority=38, previewValue=<1 1 1 1>}
  FunctionOutput(None=MaterialXLuminance!(None=@n27)) {outputName=SurfaceCoverage, sortPriority=7}
  n28 = Constant3Vector() {constant=<rgba 0 0 1 0>}
  FunctionOutput(None=If(A=@n18, "A > B"=@n20, "A == B"=@n28, "A < B"=@n28)) {desc=Bottom Normal, outputName=ClearCoatNormal, sortPriority=23}
  FunctionOutput(None=@n23) {outputName=SheenRoughness, sortPriority=10}
  FunctionOutput(None=Multiply(A=NamedRerouteUsage(), B=@n9)) {outputName=Metallic, sortPriority=1}
  n29 = FunctionInput(Preview=Constant() {r=1}) {
   bUsePreviewValueAsDefault=true
   description=Multiplier on the intensity of the specular reflection.
   inputName=specular
   inputType=FunctionInput_Scalar
   sortPriority=4
   previewValue=<1 1 1 1>
  }
  n30 = FunctionInput(Preview=Constant3Vector() {constant=<rgba 1 1 1 0>}) {
   bUsePreviewValueAsDefault=true
   description=Color tint on the specular reflection.
   inputName=specular_color
   sortPriority=5
   previewValue=<1 1 1 1>
  }
  n31 = NamedRerouteDeclaration(None=Multiply(A=@n29, B=@n30)) {name=SpecF90, nodeColor=<rgba 1 0.7969 0 1>}
  n32 = FunctionInput(Preview=Constant() {r=1.5}) {
   bUsePreviewValueAsDefault=true
   description=Index of refraction for specular reflection.
   inputName=specular_IOR
   inputType=FunctionInput_Scalar
   sortPriority=7
   previewValue=<1.5 1.5 1.5 1.5>
  }
  n33 = MaterialFunctionCall!(In_IOR=NamedRerouteDeclaration(None=@n32) {name=SpecularIOR, nodeColor=<rgba 0 1 0.1797 1>}, Out_IOR=Constant() {r=1}) {materialFunction=/Engine/Functions/Substrate/MF_SchlickInternal}
  n34 = Saturate(None=InverseLinearInterpolate(A=Constant(), B=Constant() {r=0.08}, Value=@n33.Result))
  FunctionOutput(None=NamedRerouteDeclaration(None=Multiply(A=@n31, B=@n34)) {name=Specular, nodeColor=<rgba 0 1 0.2734 1>}) {outputName=Specular, sortPriority=2}
  n35 = NamedRerouteUsage()
  FunctionOutput(None=LinearInterpolate(A=NamedRerouteUsage(), B=NamedRerouteUsage(), Alpha=@n35)) {outputName=Anisotropy, sortPriority=4}
  n36 = FunctionInput(Preview=Constant3Vector() {constant=<rgba 1 1 1 0>}) {
   bUsePreviewValueAsDefault=true
   description=The color of the emitted light.
   inputName=emission_color
   sortPriority=37
   previewValue=<1 1 1 1>
  }
  n37 = FunctionInput(Preview=Constant()) {
   bUsePreviewValueAsDefault=true
   description=The amount of emitted incandescent light.
   inputName=emission
   inputType=FunctionInput_Scalar
   sortPriority=36
   previewValue=<0 0 0 0>
  }
  n38 = Multiply(A=@n36, B=@n37)
  n39 = NamedRerouteUsage()
  FunctionOutput(None=Add(
    A=NamedRerouteDeclaration(None=@n38) {name=EmissiveColor, nodeColor=<rgba 1 0.1172 0 1>}
    B=EyeAdaptationInverse(LightValueInput=@n39).EyeAdaptationInverse
   )) {outputName=EmissiveColor, sortPriority=5}
  FunctionOutput(None=LinearInterpolate(A=NamedRerouteUsage(), B=NamedRerouteUsage(), Alpha=@n35)) {outputName=Tangent, sortPriority=9}
  n40 = FunctionInput(Preview=Constant() {r=0.2}) {
   bUsePreviewValueAsDefault=true
   description="The roughness of the specular reflection. Lower numbers produce sharper reflections, higher numbers produce blurrier reflections."
   inputName=specular_roughness
   inputType=FunctionInput_Scalar
   sortPriority=6
   previewValue=<0.2 0.2 0.2 0.2>
  }
  FunctionOutput(None=NamedRerouteDeclaration(None=@n40) {name=Roughness, nodeColor=<rgba 1 0.6328 0 1>}) {outputName=Roughness, sortPriority=3}
  n41 = FunctionInput(Preview=Constant() {r=0.1}) {
   bUsePreviewValueAsDefault=true
   description="The roughness of the clear-coat reflections. The lower the value, the sharper the reflection."
   inputName=coat_roughness
   inputType=FunctionInput_Scalar
   sortPriority=27
   previewValue=<0.1 0.1 0.1 0.1>
  }
  FunctionOutput(None=@n41) {outputName=ClearCoatRoughness, sortPriority=16}
  FunctionOutput(None=NamedRerouteDeclaration(None=@n21) {name=Normal, nodeColor=<rgba 1 0.2109 0 1>}) {desc=Top normal, outputName=Normal, sortPriority=8}
  MaterialFunctionCall(In_IOR=NamedRerouteUsage(), Out_IOR=Constant() {r=1}) {materialFunction=&MF_SchlickInternal}
  n42 = Constant() {r=1}
  NamedRerouteDeclaration(None=If(A=NamedRerouteUsage(), "A > B"=NamedRerouteUsage(), "A == B"=@n42, "A < B"=@n42)) {name=ThinFilmIridescence, nodeColor=<rgba 0.9252 0.5 0.6863 1>}
  n43 = FunctionInput() {
   bUsePreviewValueAsDefault=true
   description=The directional bias of reflected and transmitted light resulting in materials appearing rougher or glossier in certain directions.
   inputName=specular_anisotropy
   inputType=FunctionInput_Scalar
   sortPriority=8
   previewValue=<0 0 0 0>
  }
  n44 = FunctionInput(Preview=Constant()) {
   bUsePreviewValueAsDefault=true
   description=Rotation of the axis of specular anisotropy around the surface normal.
   inputName=specular_rotation
   inputType=FunctionInput_Scalar
   sortPriority=9
   previewValue=<0 0 0 0>
  }
  n45 = Multiply(A=@n44) {constB=6.2832}
  n46 = FunctionInput(Preview=Constant3Vector() {constant=<rgba 1 0 0 0>}) {bUsePreviewValueAsDefault=true, description=Input geometric tangent, inputName=tangent, sortPriority=41, previewValue=<1 1 1 1>}
  n47 = NamedRerouteDeclaration(None=@n46) {name=Tangent, nodeColor=<rgba 1 0.1172 0 1>}
  n48 = Multiply(A=Cosine(None=@n45), B=@n47)
  n49 = Multiply(A=Sine(None=@n45), B=CrossProduct(A=NamedRerouteUsage(), B=@n47))
  n50 = NamedRerouteUsage()
  NamedRerouteDeclaration(None=If(A=@n43, B=Constant(), "A > B"=Add(A=@n48, B=@n49), "A == B"=@n50, "A < B"=@n50)) {name=SpecularTangent, nodeColor=<rgba 1 0.7266 0 1>}
  n51 = FunctionInput() {
   bUsePreviewValueAsDefault=true
   description="The amount of directional bias, or anisotropy, of the clear-coat layer."
   inputName=coat_anisotropy
   inputType=FunctionInput_Scalar
   sortPriority=28
  }
  n52 = FunctionInput(Preview=Constant()) {
   bUsePreviewValueAsDefault=true
   description=The rotation of the anisotropic effect of the clear-coat layer.
   inputName=coat_rotation
   inputType=FunctionInput_Scalar
   sortPriority=29
  }
  n53 = Multiply(A=@n52) {constB=6.2832}
  n54 = NamedRerouteUsage()
  n55 = Multiply(A=Cosine(None=@n53), B=@n54)
  n56 = Multiply(A=Sine(None=@n53), B=CrossProduct(A=NamedRerouteUsage(), B=@n54))
  n57 = NamedRerouteUsage()
  NamedRerouteDeclaration(None=If(A=@n51, B=Constant(), "A > B"=Add(A=@n55, B=@n56), "A == B"=@n57, "A < B"=@n57)) {name=CoatTangent, nodeColor=<rgba 0.4062 0 1 1>}
  NamedRerouteDeclaration(None=Multiply(A=@n33.Result, B=NamedRerouteUsage())) {name=SpecF0, nodeColor=<rgba 0 0.8828 1 1>}
  n58 = FunctionInput() {bUsePreviewValueAsDefault=true, inputName=coat_IOR, inputType=FunctionInput_Scalar, sortPriority=30, previewValue=<1.5 0 0 1>}
  n59 = MaterialFunctionCall!(
   NormalWS
   BaseF0=NamedRerouteUsage()
   BaseF90=NamedRerouteUsage()
   Thickness=NamedRerouteUsage()
   n2=NamedRerouteUsage()
   UseComplexIOR
   n3
   k=@n58
  ) {materialFunction=&/MF_IridescenceNoLUT}
  NamedRerouteDeclaration(None=@n59.SpecularColor) {name=ThinFilm, nodeColor=<rgba 0.4766 0 1 1>}
  NamedRerouteDeclaration(None=@n43) {name=SpecularAnisotropy, nodeColor=<rgba 1 0.6094 0 1>}
  NamedRerouteDeclaration(None=@n51) {name=CoatAnisotropy, nodeColor=<rgba 0.2891 0 1 1>}
  NamedRerouteDeclaration(None=@n59.Fresnel) {name=ThinFilmFresnel, nodeColor=<rgba 0 1 0.7891 1>}
  n60 = NamedRerouteUsage()
  n61 = NamedRerouteUsage()
  n62 = OneMinus(None=NamedRerouteUsage())
  n63 = Multiply(A=NamedRerouteUsage(), B=NamedRerouteUsage())
  NamedRerouteDeclaration(None=Multiply(A=Step(Y=@n60, X=@n61) {constX=0.1, constY=0.001}, B=Multiply(A=@n62, B=@n63))) {name=DielectricIridescence, nodeColor=<rgba 0 0.2031 1 1>}
  n64 = FunctionInput() {
   bUsePreviewValueAsDefault=true
   desc=nanometers
   description=The thickness of the thin film layer on a surface. Use for materials such as multitone car paint or soap bubbles.
   inputName=thin_film_thickness
   inputType=FunctionInput_Scalar
   sortPriority=34
   previewValue=<0 0 0 0>
  }
  n65 = Constant() {r=0.0003}
  NamedRerouteDeclaration(None=Multiply(A=@n64, B=NamedRerouteDeclaration(None=@n65) {name=ThinFilmUnits, nodeColor=<rgba 0.0312 0 1 1>}) {constB=0.001}) {name=ThinFilmThickness, nodeColor=<rgba 1 0.6797 0 1>}
  n66 = FunctionInput(Preview=Constant() {r=1}) {
   bUsePreviewValueAsDefault=true
   description=Multiplier on the intensity of the diffuse reflection.
   inputName=base
   inputType=FunctionInput_Scalar
   previewValue=<1 1 1 1>
  }
  NamedRerouteDeclaration(None=Saturate(None=@n66)) {name=Base, nodeColor=<rgba 0 1 0.25 1>}
  n67 = FunctionInput() {bUsePreviewValueAsDefault=true, inputName=thin_film_IOR, inputType=FunctionInput_Scalar, sortPriority=35, previewValue=<1.5 1.5 1.5 1>}
  NamedRerouteDeclaration(None=@n67) {name=ThinFilmIOR, nodeColor=<rgba 0.3594 0 1 1>}
 ]
}
