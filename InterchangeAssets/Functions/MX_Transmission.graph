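# MaterialX transmission material function that fakes refractive, absorbing transmission in raster shading through emissive, base colour and opacity outputs.
# MF_Temporal_Blur supplies refracted, blurred scene colour and depth, with extra roughness remapped to blur intensity.
# Custom HLSL normalises traversal length against transmission depth, feeding BeersLaw to drive colour tinting and fog.
# Inverse Fresnel scales transmission weight; anisotropy and dispersion inputs are unconnected.
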MaterialFunction /InterchangeAssets/Functions/MX_Transmission
graph {
 properties {bExposeToLibrary=true, previewBlendMode=BLEND_Translucent, libraryCategoriesText [MaterialX]}
 nodes [
  FunctionInput() {bUsePreviewValueAsDefault=true, inputName=transmission_scatter_anisotropy, inputType=FunctionInput_Scalar, sortPriority=54}
  FunctionInput() {bUsePreviewValueAsDefault=true, inputName=transmission_dispersion, inputType=FunctionInput_Scalar, sortPriority=55}
  n1 = OneMinus(None=NamedRerouteUsage())
  n2 = Max(A=@n1)
  n3 = Subtract(A=NamedRerouteUsage())
  n4 = NamedRerouteUsage()
  n5 = Saturate(None=Max(A=Ceil(None=@n3), B=Ceil(None=@n4)))
  FunctionOutput(None=Max(
    A=LinearInterpolate(A=@n1, B=@n2, Alpha=@n5) {constB=0.95, desc=Toggle opacity}
    B=FunctionInput() {inputName=Opacity, inputType=FunctionInput_Scalar, sortPriority=4}
   )) {outputName=OpacityRaster, sortPriority=6}
  n6 = FunctionInput(Preview=Constant() {r=-1}) {
   bUsePreviewValueAsDefault=true
   description=Additional roughness on top of specular roughness. Positive values blur refractions more and negative values blur refractions less
   inputName=transmission_extra_roughness
   inputType=FunctionInput_Scalar
   sortPriority=56
  }
  n7 = Add(A=Multiply(A=@n6) {constB=0.5}) {constB=0.5, desc="to 0,1 range"}
  FunctionOutput(None=NamedRerouteDeclaration(None=Saturate(None=@n7)) {name=BlurIntensity, nodeColor=<rgba 1 0 0.2578 1>}) {outputName=MicroRoughness}
  n8 = FunctionInput(Preview=Constant3Vector() {constant=<rgba 1 1 1 0>}) {bUsePreviewValueAsDefault=true, inputName=transmission_color, sortPriority=51, previewValue=<1 1 1 1>}
  n9 = FunctionInput(Preview=Constant()) {bUsePreviewValueAsDefault=true, inputName=transmission, inputType=FunctionInput_Scalar, sortPriority=50, previewValue=<0 0 0 0>}
  n10 = Saturate(None=Multiply(A=@n9, B=NamedRerouteUsage()) {desc="inject inv fresnel here to \r\nreduce complexity later"})
  FunctionOutput(None=Multiply(A=NamedRerouteDeclaration(None=@n8) {name=TransmissionColor, nodeColor=<rgba 0 1 0.7891 1>}, B=@n10)) {outputName=TransmissionColor, sortPriority=10}
  FunctionOutput(None=LinearInterpolate(A=FunctionInput() {inputName=BaseColor}, B=Constant() {desc=10% srgb, r=0.006}, Alpha=NamedRerouteUsage()) {desc=Black when we fully transmit}) {outputName=BaseColorRaster}
  n11 = NamedRerouteUsage()
  n12 = Power(Base=@n11) {constExponent=2.2}
  n13 = MaterialFunctionCall!(ior=NamedRerouteUsage(), BlurIntensity=NamedRerouteUsage(), Normals=NamedRerouteUsage(), DisableRefraction=StaticBool()) {materialFunction=/InterchangeAssets/gltf/MaterialFunctions/MF_Temporal_Blur}
  MaterialExpressionDotProduct_2 = DotProduct(
   A=Multiply(A=CameraVectorWS()) {constB=-1}
   B=MaterialFunctionCall() {materialFunction=&CameraDirectionVector}.Result
  )
  MaterialExpressionPixelDepth_1 = PixelDepth()
  MaterialExpressionReroute_3 = Reroute(None=@n13.BluredDepth)
  MaterialExpressionMax_6 = Max(A=FunctionInput(Preview=Constant() {r=1000}) {bUsePreviewValueAsDefault=true, inputName=transmission_depth, inputType=FunctionInput_Scalar, sortPriority=52})
  n14 = Custom!(
   FOVcos=@MaterialExpressionDotProduct_2
   PxDepth=@MaterialExpressionPixelDepth_1.R
   ScDepth=@MaterialExpressionReroute_3
   maxDepth=@MaterialExpressionMax_6
  ) {
   code="float entryDepth = PxDepth/FOVcos;\r\nfloat endDepth = min( ScDepth , entryDepth + maxDepth );\r\n\r\n// Density == 1 at max depth\r\nfloat TraversalLength = endDepth - entryDepth;\r\nfloat density = saturate( TraversalLength / maxDepth ) ;\r\n\r\n// OUT\r\nreturn density;"
   description=NormalizedTraversal
   outputType=CMOT_Float1
   inputs [
    {inputName=FOVcos, input {expression=&self:MaterialExpressionDotProduct_2}}
    {inputName=PxDepth, input {expression=&self:MaterialExpressionPixelDepth_1, mask=1, maskR=1}}
    {inputName=ScDepth, input {expression=&self:MaterialExpressionReroute_3}}
    {inputName=maxDepth, input {expression=&self:MaterialExpressionMax_6}}
   ]
  }
  n15 = Add(A=NamedRerouteUsage(), B=Multiply(A=NamedRerouteUsage()) {constB=0.05})
  n16 = Constant() {r=100}
  n17 = Saturate(None=MaterialFunctionCall!(Thickness=@n14, Depth Scale=Multiply(A=@n15, B=@n16)) {materialFunction=/Engine/Functions/Engine_MaterialFunctions01/Volumetrics/BeersLaw}.Result)
  n18 = LinearInterpolate(
   A=@n12
   B=LinearInterpolate(A=Multiply(A=@n13.Blured Scene, B=@n12) {constA=1, desc=Darker}, B=Multiply(A=@n11, B=@n13.Blured Scene) {desc=Default}, Alpha=@n17)
   Alpha=@n17
  ) {constA=1, desc=fog}
  n19 = NamedRerouteUsage()
  FunctionOutput(None=Max(A=Multiply(A=@n18, B=@n19), B=FunctionInput() {inputName=EmissiveColor, sortPriority=1})) {outputName=EmissiveColorRaster, sortPriority=5}
  n20 = Constant() {r=1.5}
  NamedRerouteDeclaration(None=Clamp(None=FunctionInput(Preview=@n20) {bUsePreviewValueAsDefault=true, inputName=Ior, inputType=FunctionInput_Scalar}) {maxDefault=2, minDefault=1}) {name=ior, nodeColor=<rgba 0 0.3906 1 1>}
  NamedRerouteDeclaration(None=FunctionInput(Preview=Constant3Vector() {constant=<rgba 0 0 1 0>}) {bUsePreviewValueAsDefault=true, description=Input geometric normal, inputName=Normal, sortPriority=3}) {name=Normal, nodeColor=<rgba 1 0.2109 0 1>}
  NamedRerouteDeclaration(None=FunctionInput(Preview=Constant() {r=0.5}) {bUsePreviewValueAsDefault=true, inputName=Specular, inputType=FunctionInput_Scalar, sortPriority=2}) {name=Specular, nodeColor=<rgba 0 1 0.2734 1>}
  n21 = NamedRerouteUsage()
  n22 = NamedRerouteUsage()
  n23 = Constant() {r=5}
  NamedRerouteDeclaration(None=OneMinus(None=MaterialFunctionCall!(Normals TS=@n21, Specular=@n22, FresnelExponent=@n23) {materialFunction=/InterchangeAssets/gltf/MaterialFunctions/MF_Fresnel_DS}.Fresnel)) {name=InverseFresnel, nodeColor=<rgba 0.4766 0 1 1>}
  NamedRerouteDeclaration(None=@n10) {name=Transmission, nodeColor=<rgba 1 0 0.4453 1>}
  n24 = MaterialFunctionCall(Float3=FunctionInput(Preview=Constant()) {bUsePreviewValueAsDefault=true, inputName=transmission_scatter, sortPriority=53}) {materialFunction=&BreakOutFloat3Components}
  n25 = Max(A=@n24.R, B=@n24.R)
  NamedRerouteDeclaration(None=Saturate(None=Max(A=@n25, B=@n24.B))) {name=ScaterringAmount, nodeColor=<rgba 0 0.3203 1 1>}
 ]
}
