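# MaterialX library material function for Interchange that builds a standard surface with transmission, exposing its parameters as inputs with defaults.
# Inputs feed MX_StandardSurface, whose results feed MX_Transmission along with specular IOR and the transmission inputs.
# PathTracingQualitySwitch nodes select raster-specific base color, opacity and emissive, or path-traced values with IOR refraction and roughness raised to at least MicroRoughness.
# Transmission scatter anisotropy and dispersion are marked not implemented.
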MaterialFunction /InterchangeAssets/Functions/MX_TransmissionSurface
graph {
 properties {bExposeToLibrary=true, libraryCategoriesText [MaterialX]}
 nodes [
  n1 = FunctionInput(Preview=Constant() {r=1}) {
   bUsePreviewValueAsDefault=true
   description=Multiplier on the intensity of the diffuse reflection.
   inputName=base
   inputType=FunctionInput_Scalar
   previewValue=<1 1 1 1>
  }
  n2 = FunctionInput(Preview=Constant3Vector() {constant=<rgba 0.8 0.8 0.8 0>}) {bUsePreviewValueAsDefault=true, description=Color of the diffuse reflection., inputName=base_color, sortPriority=1, previewValue=<1 1 1 1>}
  n3 = FunctionInput(Preview=Constant()) {
   bUsePreviewValueAsDefault=true
   description=Roughness of the diffuse reflection. Higher values cause the surface to appear flatter and darker.
   inputName=diffuse_roughness
   inputType=FunctionInput_Scalar
   sortPriority=2
   previewValue=<0 0 0 0>
  }
  n4 = FunctionInput(Preview=Constant()) {
   bUsePreviewValueAsDefault=true
   description="Specifies how metallic the material appears. At its maximum, the surface behaves like a metal, using fully specular reflection and complex fresnel."
   inputName=metalness
   inputType=FunctionInput_Scalar
   sortPriority=3
   previewValue=<0 0 0 0>
  }
  n5 = FunctionInput(Preview=Constant() {r=1}) {
   bUsePreviewValueAsDefault=true
   description=Multiplier on the intensity of the specular reflection.
   inputName=specular
   inputType=FunctionInput_Scalar
   sortPriority=4
   previewValue=<1 1 1 1>
  }
  n6 = FunctionInput(Preview=Constant3Vector() {constant=<rgba 1 1 1 0>}) {
   bUsePreviewValueAsDefault=true
   description=Color tint on the specular reflection.
   inputName=specular_color
   sortPriority=5
   previewValue=<1 1 1 1>
  }
  n7 = FunctionInput(Preview=Constant() {r=0.2}) {
   bUsePreviewValueAsDefault=true
   description="The roughness of the specular reflection. Lower numbers produce sharper reflections, higher numbers produce blurrier reflections."
   inputName=specular_roughness
   inputType=FunctionInput_Scalar
   sortPriority=6
   previewValue=<0.2 0.2 0.2 0.2>
  }
  n8 = FunctionInput(Preview=Constant() {r=1.5}) {
   bUsePreviewValueAsDefault=true
   description=Index of refraction for specular reflection.
   inputName=specular_IOR
   inputType=FunctionInput_Scalar
   sortPriority=7
   previewValue=<1.5 1.5 1.5 1.5>
  }
  n9 = FunctionInput(Preview=Constant()) {
   bUsePreviewValueAsDefault=true
   description=The directional bias of reflected and transmitted light resulting in materials appearing rougher or glossier in certain directions.
   inputName=specular_anisotropy
   inputType=FunctionInput_Scalar
   sortPriority=8
   previewValue=<0 0 0 0>
  }
  n10 = FunctionInput(Preview=Constant()) {
   bUsePreviewValueAsDefault=true
   description=Rotation of the axis of specular anisotropy around the surface normal.
   inputName=specular_rotation
   inputType=FunctionInput_Scalar
   sortPriority=9
   previewValue=<0 0 0 0>
  }
  n11 = FunctionInput(Preview=Constant()) {
   bUsePreviewValueAsDefault=true
   description=The blend between diffuse reflection and subsurface scattering. A value of 1.0 indicates full subsurface scattering and a value 0 for diffuse reflection only.
   inputName=subsurface
   inputType=FunctionInput_Scalar
   sortPriority=17
   previewValue=<0 0 0 0>
  }
  n12 = Constant3Vector() {constant=<rgba 1 1 1 0>}
  n13 = FunctionInput(Preview=@n12) {
   bUsePreviewValueAsDefault=true
   description=The color of the subsurface scattering effect.
   inputName=subsurface_color
   sortPriority=18
   previewValue=<1 1 1 1>
  }
  n14 = FunctionInput(Preview=@n12) {
   bUsePreviewValueAsDefault=true
   description=The mean free path. The distance which light can travel before being scattered inside the surface.
   inputName=subsurface_radius
   sortPriority=19
   previewValue=<1 1 1 1>
  }
  n15 = FunctionInput(Preview=Constant() {r=1}) {
   bUsePreviewValueAsDefault=true
   description=Scalar weight for the subsurface radius value.
   inputName=subsurface_scale
   inputType=FunctionInput_Scalar
   sortPriority=20
   previewValue=<1 1 1 1>
  }
  n16 = FunctionInput(Preview=Constant()) {
   bUsePreviewValueAsDefault=true
   description=The weight of a sheen layer that can be used to approximate microfibers or fabrics such as velvet and satin.
   inputName=sheen
   inputType=FunctionInput_Scalar
   sortPriority=22
   previewValue=<0 0 0 0>
  }
  n17 = FunctionInput(Preview=Constant3Vector() {constant=<rgba 1 1 1 0>}) {bUsePreviewValueAsDefault=true, description=The color of the sheen layer., inputName=sheen_color, sortPriority=23, previewValue=<1 1 1 1>}
  n18 = FunctionInput(Preview=Constant() {r=0.3}) {
   bUsePreviewValueAsDefault=true
   description=The roughness of the sheen layer.
   inputName=sheen_roughness
   inputType=FunctionInput_Scalar
   sortPriority=24
   previewValue=<0.3 0.3 0.3 0.3>
  }
  n19 = FunctionInput(Preview=Constant()) {
   bUsePreviewValueAsDefault=true
   description=The weight of a reflective clear-coat layer on top of the material. Use for materials such as car paint or an oily layer.
   inputName=coat
   inputType=FunctionInput_Scalar
   sortPriority=25
   previewValue=<0 0 0 0>
  }
  n20 = FunctionInput(Preview=Constant() {r=0.1}) {
   bUsePreviewValueAsDefault=true
   description="The roughness of the clear-coat reflections. The lower the value, the sharper the reflection."
   inputName=coat_roughness
   inputType=FunctionInput_Scalar
   sortPriority=27
   previewValue=<0.1 0.1 0.1 0.1>
  }
  n21 = FunctionInput(Preview=Constant()) {
   bUsePreviewValueAsDefault=true
   description="The amount of directional bias, or anisotropy, of the clear-coat layer."
   inputName=coat_anisotropy
   inputType=FunctionInput_Scalar
   sortPriority=28
  }
  n22 = FunctionInput(Preview=Constant()) {
   bUsePreviewValueAsDefault=true
   description=The rotation of the anisotropic effect of the clear-coat layer.
   inputName=coat_rotation
   inputType=FunctionInput_Scalar
   sortPriority=29
  }
  n23 = FunctionInput(Preview=Constant()) {
   bUsePreviewValueAsDefault=true
   description=The thickness of the thin film layer on a surface. Use for materials such as multitone car paint or soap bubbles.
   inputName=thin_film_thickness
   inputType=FunctionInput_Scalar
   sortPriority=34
   previewValue=<0 0 0 0>
  }
  n24 = FunctionInput(Preview=Constant()) {
   bUsePreviewValueAsDefault=true
   description=The amount of emitted incandescent light.
   inputName=emission
   inputType=FunctionInput_Scalar
   sortPriority=36
   previewValue=<0 0 0 0>
  }
  n25 = FunctionInput(Preview=Constant3Vector() {constant=<rgba 1 1 1 0>}) {
   bUsePreviewValueAsDefault=true
   description=The color of the emitted light.
   inputName=emission_color
   sortPriority=37
   previewValue=<1 1 1 1>
  }
  n26 = FunctionInput() {bUsePreviewValueAsDefault=true, description=The opacity of the entire material., inputName=opacity, sortPriority=38, previewValue=<1 1 1 1>}
  n27 = FunctionInput(Preview=StaticBool()) {
   bUsePreviewValueAsDefault=true
   description=If true the surface is double-sided and represents an infinitely thin shell. Suitable for thin objects such as tree leaves or paper.
   inputName=thin_walled
   inputType=FunctionInput_StaticBool
   sortPriority=39
  }
  n28 = Constant3Vector() {constant=<rgba 0 0 1 0>}
  n29 = FunctionInput(Preview=@n28) {bUsePreviewValueAsDefault=true, description=Input geometric tangent, inputName=tangent, sortPriority=41, previewValue=<1 1 1 1>}
  n30 = MaterialFunctionCall!(
   base=@n1
   base_color=@n2
   diffuse_roughness=@n3
   metalness=@n4
   specular=@n5
   specular_color=@n6
   specular_roughness=@n7
   specular_IOR=@n8
   specular_anisotropy=@n9
   specular_rotation=@n10
   subsurface=@n11
   subsurface_color=@n13
   subsurface_radius=@n14
   subsurface_scale=@n15
   sheen=@n16
   sheen_color=@n17
   sheen_roughness=@n18
   coat=@n19
   coat_color=FunctionInput(Preview=Constant3Vector()) {bUsePreviewValueAsDefault=true, inputName=coat_color, sortPriority=26}
   coat_roughness=@n20
   coat_anisotropy=@n21
   coat_rotation=@n22
   coat_IOR=FunctionInput(Preview=Constant() {r=1.5}) {bUsePreviewValueAsDefault=true, inputName=coat_IOR, inputType=FunctionInput_Scalar, sortPriority=30}
   coat_normal=FunctionInput(Preview=Constant3Vector() {constant=<rgba 0 0 1 0>}) {bUsePreviewValueAsDefault=true, description=Input normal for clear-coat layer, inputName=coat_normal, sortPriority=31}
   thin_film_thickness=@n23
   thin_film_IOR=FunctionInput(Preview=Constant() {r=1.5}) {bUsePreviewValueAsDefault=true, inputName=thin_film_IOR, inputType=FunctionInput_Scalar, sortPriority=35}
   emission=@n24
   emission_color=@n25
   opacity=@n26
   thin_walled=@n27
   normal=FunctionInput(Preview=@n28) {bUsePreviewValueAsDefault=true, description=Input geometric normal, inputName=normal, sortPriority=40}
   tangent=@n29
  ) {materialFunction=&/MX_StandardSurface}
  n31 = FunctionInput(Preview=Constant()) {
   bUsePreviewValueAsDefault=true
   description=Transmission Weight
   inputName=transmission
   inputType=FunctionInput_Scalar
   sortPriority=10
   previewValue=<0 0 0 0>
  }
  n32 = FunctionInput(Preview=Constant3Vector() {constant=<rgba 1 1 1 0>}) {bUsePreviewValueAsDefault=true, description=Transmission Color, inputName=transmission_color, sortPriority=11, previewValue=<1 1 1 1>}
  n33 = FunctionInput(Preview=Constant() {r=1000}) {
   bUsePreviewValueAsDefault=true
   description=Distance travelled inside the material
   inputName=transmission_depth
   inputType=FunctionInput_Scalar
   sortPriority=12
  }
  n34 = FunctionInput(Preview=Constant()) {bUsePreviewValueAsDefault=true, description=Scattering coefficient of the interior medium, inputName=transmission_scatter, sortPriority=13}
  n35 = FunctionInput(Preview=Constant()) {
   bUsePreviewValueAsDefault=true
   description=Not Implemented
   inputName=transmission_scatter_anisotropy
   inputType=FunctionInput_Scalar
   sortPriority=14
  }
  n36 = FunctionInput(Preview=Constant()) {
   bUsePreviewValueAsDefault=true
   description=Not Implemented
   inputName=transmission_dispersion
   inputType=FunctionInput_Scalar
   sortPriority=15
  }
  n37 = FunctionInput(Preview=Constant() {r=-1}) {
   bUsePreviewValueAsDefault=true
   description="Blured refraction. Positive values blur refractions more and negative values blur refractions less [-1,1]"
   inputName=transmission_extra_roughness
   inputType=FunctionInput_Scalar
   sortPriority=16
  }
  n38 = MaterialFunctionCall!(
   Ior=@n8
   BaseColor=@n30.Base Color
   EmissiveColor=@n30.Base Color
   Specular=@n30.Base Color
   Normal=@n30.Base Color
   Opacity=@n30.Base Color
   transmission=@n31
   transmission_color=@n32
   transmission_depth=@n33
   transmission_scatter=@n34
   transmission_scatter_anisotropy=@n35
   transmission_dispersion=@n36
   transmission_extra_roughness=@n37
  ) {materialFunction=&/MX_Transmission}
  FunctionOutput(None=PathTracingQualitySwitch!(Normal=@n38.BaseColorRaster, PathTraced=@n30.Base Color)) {outputName=BaseColor}
  FunctionOutput(None=PathTracingQualitySwitch(Normal=Constant() {r=1}, PathTraced=@n8)) {outputName=Refraction, sortPriority=15}
  FunctionOutput(None=@n30.Specular) {outputName=Specular, sortPriority=2}
  FunctionOutput(None=@n30.SurfaceCoverage) {outputName=SurfaceCoverage, sortPriority=20}
  FunctionOutput(None=PathTracingQualitySwitch(Normal=@n38.OpacityRaster, PathTraced=@n30.Opacity)) {outputName=Opacity, sortPriority=6}
  FunctionOutput(None=@n38.TransmissionColor) {outputName=TransmissionColor, sortPriority=19}
  FunctionOutput(None=@n30.Metallic) {outputName=Metallic, sortPriority=1}
  FunctionOutput(None=PathTracingQualitySwitch(Normal=@n38.EmissiveColorRaster, PathTraced=@n30.EmissiveColor)) {outputName=EmissiveColor, sortPriority=5}
  FunctionOutput(None=PathTracingQualitySwitch(Normal=@n30.Roughness, PathTraced=Max(A=@n30.Roughness, B=@n38.MicroRoughness))) {outputName=Roughness, sortPriority=3}
  FunctionOutput(None=@n30.Anisotropy) {outputName=Anisotropy, sortPriority=4}
  FunctionOutput(None=@n30.Tangent) {outputName=Tangent, sortPriority=8}
  FunctionOutput(None=@n30.Normal) {outputName=Normal, sortPriority=7}
 ]
}
