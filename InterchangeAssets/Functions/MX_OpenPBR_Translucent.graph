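# MaterialX library function that exposes OpenPBR translucent surface inputs as a StandardSurface fallback, since full OpenPBR support requires Substrate.
# It renames and forwards inputs to MX_TransmissionSurface, mapping subsurface_radius_scale to subsurface_radius and subsurface_radius to subsurface_scale.
# Abbe number, thin film weight, subsurface scatter anisotropy and coat darkening inputs are unconnected.
# The call's results pass straight out as material attributes, including Refraction and TransmissionColor.
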MaterialFunction /InterchangeAssets/Functions/MX_OpenPBR_Translucent
graph {
 properties {
  bExposeToLibrary=true
  description="OpenPBR Surface Shader is only fully supported with Substrate, otherwise it falls back to the StandardSurface shader."
  libraryCategoriesText [MaterialX]
 }
 nodes [
  FunctionInput() {
   bUsePreviewValueAsDefault=true
   inputName=transmission_dispersion_abbe_number
   inputType=FunctionInput_Scalar
   sortPriority=17
   previewValue=<0 0 0 0>
  }
  FunctionInput() {
   bUsePreviewValueAsDefault=true
   description=The thickness of the thin film layer on a surface. Use for materials such as multitone car paint or soap bubbles.
   inputName=thin_film_weight
   inputType=FunctionInput_Scalar
   sortPriority=32
   previewValue=<0 0 0 0>
  }
  FunctionInput() {
   bUsePreviewValueAsDefault=true
   description="Controls the phase-function of subsurface scattering, where zero scatters light evenly, positive values scatter forward, and negative values scatter backward."
   inputName=subsurface_scatter_anisotropy
   inputType=FunctionInput_Scalar
   sortPriority=22
   previewValue=<0 1 1 1>
  }
  FunctionInput() {bUsePreviewValueAsDefault=true, inputName=coat_darkening, inputType=FunctionInput_Scalar, sortPriority=31, previewValue=<1 1 1 1>}
  n1 = FunctionInput() {
   bUsePreviewValueAsDefault=true
   description=Multiplier on the intensity of the diffuse reflection.
   inputName=base_weight
   inputType=FunctionInput_Scalar
   previewValue=<1 1 1 1>
  }
  n2 = FunctionInput() {
   bUsePreviewValueAsDefault=true
   description=Color of the diffuse reflection.
   inputName=base_color
   sortPriority=1
   previewValue=<0.8 0.8 0.8 1>
  }
  n3 = FunctionInput() {
   bUsePreviewValueAsDefault=true
   description=Roughness of the diffuse reflection. Higher values cause the surface to appear flatter and darker.
   inputName=base_diffuse_roughness
   inputType=FunctionInput_Scalar
   sortPriority=2
   previewValue=<0 0 0 0>
  }
  n4 = FunctionInput() {
   bUsePreviewValueAsDefault=true
   description="Specifies how metallic the material appears. At its maximum, the surface behaves like a metal, using fully specular reflection and complex fresnel."
   inputName=base_metalness
   inputType=FunctionInput_Scalar
   sortPriority=3
   previewValue=<0 0 0 0>
  }
  n5 = FunctionInput() {
   bUsePreviewValueAsDefault=true
   description=Multiplier on the intensity of the specular reflection.
   inputName=specular_weight
   inputType=FunctionInput_Scalar
   sortPriority=4
   previewValue=<1 1 1 1>
  }
  n6 = FunctionInput() {
   bUsePreviewValueAsDefault=true
   description="Color of the specular reflection (controls the physical edge-tint for metals, and a non-physical overall tint for dielectrics)."
   inputName=specular_color
   sortPriority=5
   previewValue=<1 1 1 1>
  }
  n7 = FunctionInput() {
   bUsePreviewValueAsDefault=true
   description="The roughness of the specular reflection. Lower numbers produce sharper reflections, higher numbers produce blurrier reflections."
   inputName=specular_roughness
   inputType=FunctionInput_Scalar
   sortPriority=6
   previewValue=<0.2 0.2 0.2 0.2>
  }
  n8 = FunctionInput() {
   bUsePreviewValueAsDefault=true
   description=Index of refraction for specular reflection.
   inputName=specular_ior
   inputType=FunctionInput_Scalar
   sortPriority=7
   previewValue=<1.5 1.5 1.5 1.5>
  }
  n9 = FunctionInput() {
   bUsePreviewValueAsDefault=true
   description=The directional bias of reflected and transmitted light resulting in materials appearing rougher or glossier in certain directions.
   inputName=specular_roughness_anisotropy
   inputType=FunctionInput_Scalar
   sortPriority=9
   previewValue=<0 0 0 0>
  }
  n10 = FunctionInput() {
   bUsePreviewValueAsDefault=true
   description=The amount of emitted incandescent light.
   inputName=transmission_weight
   inputType=FunctionInput_Scalar
   sortPriority=11
   previewValue=<0 0 0 0>
  }
  n11 = FunctionInput() {
   bUsePreviewValueAsDefault=true
   description=The color of the emitted light.
   inputName=transmission_color
   sortPriority=12
   previewValue=<1 1 1 1>
  }
  n12 = FunctionInput() {
   bUsePreviewValueAsDefault=true
   description=Specifies the distance light travels inside the transparent base before it becomes exactly the transmission_color according to Beer's law.
   inputName=transmission_depth
   inputType=FunctionInput_Scalar
   sortPriority=13
  }
  n13 = FunctionInput() {
   bUsePreviewValueAsDefault=true
   description="Controls the color of light volumetrically scattered inside the transparent base. Suitable for materials with visually significant scattering such as honey, fruit juice, murky water, opalescent glass, or milky glass."
   inputName=transmission_scatter
   sortPriority=14
  }
  n14 = FunctionInput() {
   bUsePreviewValueAsDefault=true
   description="The amount of directional bias, or anisotropy, of the volumetric scattering in the transparent base."
   inputName=transmission_scatter_anisotropy
   inputType=FunctionInput_Scalar
   sortPriority=15
   previewValue=<0 0 0 0>
  }
  n15 = FunctionInput() {
   bUsePreviewValueAsDefault=true
   inputName=transmission_dispersion_scale
   inputType=FunctionInput_Scalar
   sortPriority=16
   previewValue=<0 0 0 0>
  }
  n16 = FunctionInput() {
   bUsePreviewValueAsDefault=true
   description=Mixture weight which dials the opaque dielectric base between diffuse reflection and subsurface scattering. A value of 1.0 indicates full subsurface scattering and a value 0 for diffuse reflection only.
   inputName=subsurface_weight
   inputType=FunctionInput_Scalar
   sortPriority=18
   previewValue=<0 0 0 0>
  }
  n17 = FunctionInput() {
   bUsePreviewValueAsDefault=true
   description=The overall reflected color of the subsurface scattering effect.
   inputName=subsurface_color
   sortPriority=19
   previewValue=<0.8 0.8 0.8 1>
  }
  n18 = FunctionInput() {
   bUsePreviewValueAsDefault=true
   description="RGB multiplier to subsurface_radius, giving the per-channel diffusion blur profile size."
   inputName=subsurface_radius_scale
   sortPriority=21
   previewValue=<1 0.5 0.25 1>
  }
  n19 = FunctionInput() {
   bUsePreviewValueAsDefault=true
   description=Length scale of the subsurface diffusion blur profile on the surface.
   inputName=subsurface_radius
   inputType=FunctionInput_Scalar
   sortPriority=20
   previewValue=<1 1 1 1>
  }
  n20 = FunctionInput() {
   bUsePreviewValueAsDefault=true
   description=The weight of a sheen layer that can be used to approximate microfibers or fabrics such as velvet and satin.
   inputName=fuzz_weight
   inputType=FunctionInput_Scalar
   sortPriority=23
   previewValue=<0 0 0 0>
  }
  n21 = FunctionInput() {bUsePreviewValueAsDefault=true, description=The color of the sheen layer., inputName=fuzz_color, sortPriority=24, previewValue=<1 1 1 1>}
  n22 = FunctionInput() {
   bUsePreviewValueAsDefault=true
   description=The roughness of the sheen layer.
   inputName=fuzz_roughness
   inputType=FunctionInput_Scalar
   sortPriority=25
   previewValue=<0.5 0.5 0.5 0.5>
  }
  n23 = FunctionInput() {
   bUsePreviewValueAsDefault=true
   description=The weight of a reflective clear-coat layer on top of the material. Use for materials such as car paint or an oily layer.
   inputName=coat_weight
   inputType=FunctionInput_Scalar
   sortPriority=26
   previewValue=<0 0 0 0>
  }
  n24 = FunctionInput() {bUsePreviewValueAsDefault=true, inputName=coat_color, sortPriority=27, previewValue=<1 1 1 1>}
  n25 = FunctionInput() {
   bUsePreviewValueAsDefault=true
   description="The roughness of the clear-coat reflections. The lower the value, the sharper the reflection."
   inputName=coat_roughness
   inputType=FunctionInput_Scalar
   sortPriority=28
   previewValue=<0 0 0 0>
  }
  n26 = FunctionInput() {bUsePreviewValueAsDefault=true, inputName=coat_ior, inputType=FunctionInput_Scalar, sortPriority=30, previewValue=<1 1 1 1>}
  n27 = FunctionInput() {
   bUsePreviewValueAsDefault=true
   description=Input normal for clear-coat layer
   inputName=geometry_coat_normal
   sortPriority=40
   previewValue=<0 0 1 1>
  }
  n28 = FunctionInput() {
   bUsePreviewValueAsDefault=true
   description=The thickness of the thin film layer on a surface. Use for materials such as multitone car paint or soap bubbles.
   inputName=thin_film_thickness
   inputType=FunctionInput_Scalar
   sortPriority=33
   previewValue=<0 0 0 0>
  }
  n29 = FunctionInput() {
   bUsePreviewValueAsDefault=true
   description=The thickness of the thin film layer on a surface. Use for materials such as multitone car paint or soap bubbles.
   inputName=thin_film_ior
   inputType=FunctionInput_Scalar
   sortPriority=34
   previewValue=<0 0 0 0>
  }
  n30 = FunctionInput() {
   bUsePreviewValueAsDefault=true
   description=The amount of emitted incandescent light.
   inputName=emission_luminance
   inputType=FunctionInput_Scalar
   sortPriority=35
   previewValue=<0 0 0 0>
  }
  n31 = FunctionInput() {
   bUsePreviewValueAsDefault=true
   description=The color of the emitted light.
   inputName=emission_color
   sortPriority=36
   previewValue=<1 1 1 1>
  }
  n32 = FunctionInput() {
   bUsePreviewValueAsDefault=true
   description=Input geometric normal
   inputName=geometry_opacity
   inputType=FunctionInput_Scalar
   sortPriority=37
   previewValue=<1 0 0 1>
  }
  n33 = FunctionInput(Preview=StaticBool()) {
   bUsePreviewValueAsDefault=true
   description=If true the surface is double-sided and represents an infinitesimally thin shell. Suitable for extremely geometrically thin objects such as leaves or paper.
   inputName=geometry_thin_walled
   inputType=FunctionInput_StaticBool
   sortPriority=38
  }
  n34 = FunctionInput() {bUsePreviewValueAsDefault=true, description=Input geometric normal, inputName=geometry_normal, sortPriority=39, previewValue=<0 0 1 1>}
  n35 = FunctionInput() {bUsePreviewValueAsDefault=true, description=Input geometric tangent, inputName=geometry_tangent, sortPriority=41, previewValue=<1 0 0 1>}
  n36 = MaterialFunctionCall!(
   base=@n1
   base_color=@n2
   diffuse_roughness=@n3
   metalness=@n4
   specular=@n5
   specular_color=@n6
   specular_roughness=@n7
   specular_IOR=@n8
   specular_anisotropy=@n9
   specular_rotation
   transmission=@n10
   transmission_color=@n11
   transmission_depth=@n12
   transmission_scatter=@n13
   transmission_scatter_anisotropy=@n14
   transmission_dispersion=@n15
   transmission_extra_roughness
   subsurface=@n16
   subsurface_color=@n17
   subsurface_radius=@n18
   subsurface_scale=@n19
   sheen=@n20
   sheen_color=@n21
   sheen_roughness=@n22
   coat=@n23
   coat_color=@n24
   coat_roughness=@n25
   coat_anisotropy=FunctionInput() {inputName=coat_roughness_anisotropy, inputType=FunctionInput_Scalar, sortPriority=29}
   coat_rotation
   coat_IOR=@n26
   coat_normal=@n27
   thin_film_thickness=@n28
   thin_film_IOR=@n29
   emission=@n30
   emission_color=@n31
   opacity=@n32
   thin_walled=@n33
   normal=@n34
   tangent=@n35
  ) {materialFunction=&/MX_TransmissionSurface}
  FunctionOutput(None=@n36.BaseColor) {outputName=BaseColor}
  FunctionOutput(None=@n36.Refraction) {outputName=Refraction, sortPriority=8}
  FunctionOutput(None=@n36.TransmissionColor) {outputName=TransmissionColor, sortPriority=8}
  FunctionOutput(None=@n36.Metallic) {outputName=Metallic, sortPriority=1}
  FunctionOutput(None=@n36.Specular) {outputName=Specular, sortPriority=2}
  FunctionOutput(None=@n36.Roughness) {outputName=Roughness, sortPriority=3}
  FunctionOutput(None=@n36.Anisotropy) {outputName=Anisotropy, sortPriority=4}
  FunctionOutput(None=@n36.EmissiveColor) {outputName=EmissiveColor, sortPriority=5}
  FunctionOutput(None=@n36.Opacity) {outputName=Opacity, sortPriority=6}
  FunctionOutput(None=@n36.Tangent) {outputName=Tangent, sortPriority=8}
  FunctionOutput(None=@n36.Normal) {outputName=Normal, sortPriority=7}
 ]
}
