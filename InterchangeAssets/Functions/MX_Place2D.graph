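# Material function in the MaterialX library that implements Place2D, a 2D transform of texture coordinates around a pivot.
# It flips V on the Coordinates input, subtracts the pivot, divides by scale, rotates about the origin, subtracts offset and adds the pivot back.
# The rotate input is multiplied by roughly 1/360 before feeding the Rotator.
# The result's V is flipped again; the operationorder input is unused, so the order is fixed.
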MaterialFunction /InterchangeAssets/Functions/MX_Place2D
graph {
 properties {bExposeToLibrary=true, libraryCategoriesText [MaterialX]}
 nodes [
  FunctionInput() {
   bUsePreviewValueAsDefault=true
   description="The order in which to perform the transform operations.    \"0\" or \"SRT\" performs -pivot, scale, rotate, translate, +pivot as per the original    implementation matching the behavior of certain DCC packages, and \"1\" or \"TRS\" performs    -pivot, translate, rotate, scale, +pivot which does not introduce texture shear.    Default is 0 \"SRT\" for backward compatibility. (not currently used)"
   inputName=operationorder
   inputType=FunctionInput_Scalar
   sortPriority=5
  }
  n1 = FunctionInput() {bUsePreviewValueAsDefault=true, inputName=pivot, inputType=FunctionInput_Vector2, sortPriority=1}
  n2 = FunctionInput() {bUsePreviewValueAsDefault=true, inputName=scale, inputType=FunctionInput_Vector2, sortPriority=2, previewValue=<1 1 0 1>}
  n3 = Divide(A=Subtract(A=NamedRerouteUsage(), B=NamedRerouteDeclaration(None=@n1) {name=pivot, nodeColor=<rgba 0 0.6016 1 1>}), B=@n2)
  n4 = Multiply(A=FunctionInput() {bUsePreviewValueAsDefault=true, inputName=rotate, inputType=FunctionInput_Scalar, sortPriority=3}) {constB=0.0028}
  n5 = Add(
   A=Subtract(
    A=Rotator(Coordinate=@n3, Time=@n4) {centerX=0, centerY=0, speed=1}
    B=FunctionInput() {bUsePreviewValueAsDefault=true, inputName=offset, inputType=FunctionInput_Vector2, sortPriority=4}
   )
   B=NamedRerouteUsage()
  )
  n6 = ComponentMask(None=@n5) {g=true}
  FunctionOutput(None=AppendVector(A=ComponentMask(None=@n5) {r=true}, B=OneMinus(None=@n6)))
  n7 = FunctionInput(Preview=TextureCoordinate()) {bUsePreviewValueAsDefault=true, inputName=Coordinates, inputType=FunctionInput_Vector2}
  n8 = ComponentMask(None=@n7) {g=true}
  NamedRerouteDeclaration(None=AppendVector(A=ComponentMask(None=@n7) {r=true}, B=OneMinus(None=@n8))) {name=TexCoord, nodeColor=<rgba 0 0.2031 1 1>}
 ]
}
