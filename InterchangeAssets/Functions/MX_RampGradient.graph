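# MaterialX library material function that helps build a ramp gradient by handling a single control point between two interval positions and colors.
# It clamps x, offers linear remap, smoothstep and step paths through switches, and has a lerp for color mixing.
# A final If compares interval_num with num_intervals; a prev_color input suggests chaining of calls.
# Named reroutes connect the stages.
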MaterialFunction /InterchangeAssets/Functions/MX_RampGradient
graph {
 properties {bExposeToLibrary=true, description="A helper node that handles a single control point within a <ramp>.", libraryCategoriesText [MaterialX]}
 nodes [
  n1 = FunctionInput() {bUsePreviewValueAsDefault=true, inputName=interval_num, inputType=FunctionInput_Scalar, sortPriority=7, previewValue=<1 0 0 1>}
  n2 = FunctionInput() {bUsePreviewValueAsDefault=true, inputName=num_intervals, inputType=FunctionInput_Scalar, sortPriority=8, previewValue=<2 0 0 1>}
  n3 = NamedRerouteUsage()
  FunctionOutput(None=If(A=@n1, B=@n2, "A > B"=NamedRerouteUsage(), "A == B"=@n3, "A < B"=@n3))
  n4 = FunctionInput() {bUsePreviewValueAsDefault=true, inputName=x, inputType=FunctionInput_Scalar}
  NamedRerouteDeclaration(None=Clamp(None=NamedRerouteDeclaration(None=@n4) {name=x, nodeColor=<rgba 0 0.4141 1 1>}, Min=NamedRerouteUsage(), Max=NamedRerouteUsage())) {name=linear_clamp, nodeColor=<rgba 1 0.4688 0 1>}
  NamedRerouteDeclaration(None=FunctionInput() {bUsePreviewValueAsDefault=true, inputName=interval1, inputType=FunctionInput_Scalar, sortPriority=1}) {name=interval1, nodeColor=<rgba 0.1719 1 0 1>}
  NamedRerouteDeclaration(None=LinearInterpolate(A=NamedRerouteUsage(), B=NamedRerouteUsage(), Alpha=NamedRerouteUsage())) {name=mix3, nodeColor=<rgba 0.7813 0 1 1>}
  n5 = NamedRerouteUsage()
  NamedRerouteDeclaration(None=If(A=NamedRerouteUsage(), B=NamedRerouteUsage(), "A > B"=NamedRerouteUsage(), "A == B"=@n5, "A < B"=@n5)) {name=step_ifgreater, nodeColor=<rgba 1 0.9375 0 1>}
  MaterialExpressionNamedRerouteUsage_19 = NamedRerouteUsage()
  MaterialExpressionNamedRerouteUsage_20 = NamedRerouteUsage()
  n6 = Switch!(
   SwitchValue=NamedRerouteUsage()
   Default=@MaterialExpressionNamedRerouteUsage_19
   linear=@MaterialExpressionNamedRerouteUsage_19
   smooth=@MaterialExpressionNamedRerouteUsage_19
   step=@MaterialExpressionNamedRerouteUsage_20
  ) {inputs [
    {
     inputName=linear
     input {
      expression=&self:MaterialExpressionNamedRerouteUsage_19
      inputName=None
      mask=0
      maskA=0
      maskB=0
      maskG=0
      maskR=0
      outputIndex=0
     }
    }
    {
     inputName=smooth
     input {
      expression=&self:MaterialExpressionNamedRerouteUsage_19
      inputName=None
      mask=0
      maskA=0
      maskB=0
      maskG=0
      maskR=0
      outputIndex=0
     }
    }
    {
     inputName=step
     input {
      expression=&self:MaterialExpressionNamedRerouteUsage_20
      inputName=None
      mask=0
      maskA=0
      maskB=0
      maskG=0
      maskR=0
      outputIndex=0
     }
    }
   ]}
  NamedRerouteDeclaration(None=@n6) {name=interpolation_selector2, nodeColor=<rgba 0 1 0.9297 1>}
  NamedRerouteDeclaration(None=FunctionInput() {bUsePreviewValueAsDefault=true, inputName=prev_color, inputType=FunctionInput_Vector4, sortPriority=6}) {name=prev_color, nodeColor=<rgba 1 0.4453 0 1>}
  n7 = NamedRerouteUsage()
  NamedRerouteDeclaration(None=If(A=NamedRerouteUsage(), B=NamedRerouteUsage(), "A > B"=NamedRerouteUsage(), "A == B"=@n7, "A < B"=@n7)) {name=ifgreater3, nodeColor=<rgba 0.3594 1 0 1>}
  n8 = FunctionInput() {bUsePreviewValueAsDefault=true, inputName=interval2, inputType=FunctionInput_Scalar, sortPriority=2, previewValue=<1 1 1 1>}
  NamedRerouteDeclaration(None=@n8) {name=interval2, nodeColor=<rgba 1 0 0.4922 1>}
  NamedRerouteDeclaration(None=MaterialXRemap!(None=NamedRerouteUsage(), InputLow=NamedRerouteUsage(), InputHigh=NamedRerouteUsage(), TargetLow, TargetHigh)) {name=linear_remap, nodeColor=<rgba 0 0.8125 1 1>}
  NamedRerouteDeclaration(None=SmoothStep(Min=NamedRerouteUsage(), Max=NamedRerouteUsage(), Value=NamedRerouteUsage())) {name=smoothstep, nodeColor=<rgba 0.3359 1 0 1>}
  n9 = FunctionInput() {
   bUsePreviewValueAsDefault=true
   description="0 = linear, 1 = smooth, 2 = step"
   inputName=interpolation
   inputType=FunctionInput_Scalar
   sortPriority=5
   previewValue=<1 1 1 1>
  }
  NamedRerouteDeclaration(None=@n9) {name=interpolation, nodeColor=<rgba 1 0 0.6562 1>}
  MaterialExpressionNamedRerouteUsage_9 = NamedRerouteUsage()
  MaterialExpressionNamedRerouteUsage_10 = NamedRerouteUsage()
  n10 = Switch!(
   SwitchValue=NamedRerouteUsage()
   Default=@MaterialExpressionNamedRerouteUsage_9
   linear_remap=@MaterialExpressionNamedRerouteUsage_9
   smoothstep=@MaterialExpressionNamedRerouteUsage_10
  ) {inputs [
    {
     inputName=linear_remap
     input {
      expression=&self:MaterialExpressionNamedRerouteUsage_9
      inputName=None
      mask=0
      maskA=0
      maskB=0
      maskG=0
      maskR=0
      outputIndex=0
     }
    }
    {
     inputName=smoothstep
     input {
      expression=&self:MaterialExpressionNamedRerouteUsage_10
      inputName=None
      mask=0
      maskA=0
      maskB=0
      maskG=0
      maskR=0
      outputIndex=0
     }
    }
   ]}
  NamedRerouteDeclaration(None=@n10) {name=interpolation_selector, nodeColor=<rgba 0 0.9766 1 1>}
  NamedRerouteDeclaration(None=FunctionInput() {bUsePreviewValueAsDefault=true, inputName=color1, inputType=FunctionInput_Vector4, sortPriority=3}) {name=color1, nodeColor=<rgba 0 1 0.625 1>}
  n11 = FunctionInput() {bUsePreviewValueAsDefault=true, inputName=color2, inputType=FunctionInput_Vector4, sortPriority=4, previewValue=<1 1 1 1>}
  NamedRerouteDeclaration(None=@n11) {name=color2, nodeColor=<rgba 1 0.3047 0 1>}
 ]
}
