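# MX_Ramp is an Interchange material function that builds a color ramp with up to 10 interval/color control points.
# A type switch selects a ramp coordinate from the UVs: standard X, radial angle, circular distance or box maximum-axis distance.
# Nine MX_RampGradient calls, one per adjacent point pair, take a previous color, the clamped interval count and a linear, smooth or step interpolation mode.
# Named reroutes carry values between stages.
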MaterialFunction /InterchangeAssets/Functions/MX_Ramp
graph {
 properties {description=A ramp that supports up to 10 control points}
 nodes [
  FunctionOutput(None=NamedRerouteUsage())
  n1 = NamedRerouteUsage()
  n2 = NamedRerouteUsage()
  n3 = ComponentMask(None=@n2) {g=true}
  NamedRerouteDeclaration(None=If(A=ComponentMask(None=@n1) {r=true}, B=ComponentMask(None=@n1) {g=true}, "A > B"=ComponentMask(None=@n2) {r=true}, "A == B"=@n3, "A < B"=@n3)) {name=box_ifgreater, nodeColor=<rgba 0 1 0.5078 1>}
  n4 = FunctionInput() {
   bUsePreviewValueAsDefault=true
   description="0: standard, 1: radial, 2: circular, 3: box"
   inputName=type
   inputType=FunctionInput_Scalar
   sortPriority=1
  }
  MaterialExpressionNamedRerouteUsage_2 = NamedRerouteUsage()
  MaterialExpressionNamedRerouteUsage_3 = NamedRerouteUsage()
  MaterialExpressionNamedRerouteUsage_4 = NamedRerouteUsage()
  MaterialExpressionNamedRerouteUsage_7 = NamedRerouteUsage()
  n5 = Switch!(
   SwitchValue=@n4
   Default=@MaterialExpressionNamedRerouteUsage_2
   standard=@MaterialExpressionNamedRerouteUsage_2
   radial=@MaterialExpressionNamedRerouteUsage_3
   circular=@MaterialExpressionNamedRerouteUsage_4
   box=@MaterialExpressionNamedRerouteUsage_7
  ) {inputs [
    {
     inputName=standard
     input {
      expression=&self:MaterialExpressionNamedRerouteUsage_2
      inputName=None
      mask=0
      maskA=0
      maskB=0
      maskG=0
      maskR=0
      outputIndex=0
     }
    }
    {
     inputName=radial
     input {
      expression=&self:MaterialExpressionNamedRerouteUsage_3
      inputName=None
      mask=0
      maskA=0
      maskB=0
      maskG=0
      maskR=0
      outputIndex=0
     }
    }
    {
     inputName=circular
     input {
      expression=&self:MaterialExpressionNamedRerouteUsage_4
      inputName=None
      mask=0
      maskA=0
      maskB=0
      maskG=0
      maskR=0
      outputIndex=0
     }
    }
    {
     inputName=box
     input {
      expression=&self:MaterialExpressionNamedRerouteUsage_7
      inputName=None
      mask=0
      maskA=0
      maskB=0
      maskG=0
      maskR=0
      outputIndex=0
     }
    }
   ]}
  NamedRerouteDeclaration(None=@n5) {name=ramp_type_selector, nodeColor=<rgba 1 0.1875 0 1>}
  n6 = FunctionInput() {bUsePreviewValueAsDefault=true, inputName=num_intervals, inputType=FunctionInput_Scalar, sortPriority=3, previewValue=<2 0 0 1>}
  NamedRerouteDeclaration(None=Clamp(None=@n6) {maxDefault=10, minDefault=1}) {name=num_intervals, nodeColor=<rgba 0 0.1328 1 1>}
  n7 = FunctionInput() {
   bUsePreviewValueAsDefault=true
   description="0: linear, 1: smooth, 2: step"
   inputName=interpolation
   inputType=FunctionInput_Scalar
   sortPriority=2
  }
  NamedRerouteDeclaration(None=@n7) {name=interpolation, nodeColor=<rgba 0.4531 1 0 1>}
  n8 = FunctionInput() {bUsePreviewValueAsDefault=true, inputName=interval2, inputType=FunctionInput_Scalar, sortPriority=6, previewValue=<1 0 0 1>}
  n9 = FunctionInput() {bUsePreviewValueAsDefault=true, inputName=color1, inputType=FunctionInput_Vector4, sortPriority=5}
  n10 = FunctionInput() {bUsePreviewValueAsDefault=true, inputName=color2, inputType=FunctionInput_Vector4, sortPriority=7, previewValue=<1 1 1 1>}
  NamedRerouteDeclaration(None=MaterialFunctionCall!(
    x=NamedRerouteUsage()
    interval1=FunctionInput() {bUsePreviewValueAsDefault=true, inputName=interval1, inputType=FunctionInput_Scalar, sortPriority=4}
    interval2=NamedRerouteDeclaration(None=@n8) {name=interval2, nodeColor=<rgba 0 1 0.9063 1>}
    color1=@n9
    color2=NamedRerouteDeclaration(None=@n10) {name=color2, nodeColor=<rgba 1 0.5859 0 1>}
    interpolation=NamedRerouteUsage()
    prev_color=@n9
    interval_num=Constant() {r=1}
    num_intervals=NamedRerouteUsage()
   ) {materialFunction=&/MX_RampGradient}.Result) {name=ramp_gradient1, nodeColor=<rgba 1 0 0.7734 1>}
  n11 = FunctionInput() {bUsePreviewValueAsDefault=true, inputName=interval3, inputType=FunctionInput_Scalar, sortPriority=8, previewValue=<1 0 0 1>}
  n12 = FunctionInput() {bUsePreviewValueAsDefault=true, inputName=color3, inputType=FunctionInput_Vector4, sortPriority=9, previewValue=<1 1 1 1>}
  NamedRerouteDeclaration(None=MaterialFunctionCall(
    x=NamedRerouteUsage()
    interval1=NamedRerouteUsage()
    interval2=NamedRerouteDeclaration(None=@n11) {name=interval3, nodeColor=<rgba 0.2656 0 1 1>}
    color1=NamedRerouteUsage()
    color2=NamedRerouteDeclaration(None=@n12) {name=color3, nodeColor=<rgba 0.0547 1 0 1>}
    interpolation=NamedRerouteUsage()
    prev_color=NamedRerouteUsage()
    interval_num=Constant() {r=2}
    num_intervals=NamedRerouteUsage()
   ) {materialFunction=&MX_RampGradient}.Result) {name=ramp_gradient2, nodeColor=<rgba 1 0 0.375 1>}
  n13 = Multiply(A=NamedRerouteUsage()) {constB=1.414}
  NamedRerouteDeclaration(None=Length(None=NamedRerouteDeclaration(None=@n13) {name=circular_multiply, nodeColor=<rgba 1 0 0.6094 1>}) {is=class MaterialExpressionLength}) {name=circular_magnitude, nodeColor=<rgba 0 0.9297 1 1>}
  n14 = FunctionInput() {bUsePreviewValueAsDefault=true, inputName=interval4, inputType=FunctionInput_Scalar, sortPriority=10, previewValue=<1 0 0 1>}
  n15 = FunctionInput() {bUsePreviewValueAsDefault=true, inputName=color4, inputType=FunctionInput_Vector4, sortPriority=11, previewValue=<1 1 1 1>}
  NamedRerouteDeclaration(None=MaterialFunctionCall(
    x=NamedRerouteUsage()
    interval1=NamedRerouteUsage()
    interval2=NamedRerouteDeclaration(None=@n14) {name=interval4, nodeColor=<rgba 0.2656 0 1 1>}
    color1=NamedRerouteUsage()
    color2=NamedRerouteDeclaration(None=@n15) {name=color4, nodeColor=<rgba 0.0547 1 0 1>}
    interpolation=NamedRerouteUsage()
    prev_color=NamedRerouteUsage()
    interval_num=Constant() {r=3}
    num_intervals=NamedRerouteUsage()
   ) {materialFunction=&MX_RampGradient}.Result) {name=ramp_gradient3, nodeColor=<rgba 1 0 0.375 1>}
  n16 = FunctionInput() {bUsePreviewValueAsDefault=true, inputName=interval5, inputType=FunctionInput_Scalar, sortPriority=12, previewValue=<1 0 0 1>}
  n17 = FunctionInput() {bUsePreviewValueAsDefault=true, inputName=color5, inputType=FunctionInput_Vector4, sortPriority=13, previewValue=<1 1 1 1>}
  NamedRerouteDeclaration(None=MaterialFunctionCall(
    x=NamedRerouteUsage()
    interval1=NamedRerouteUsage()
    interval2=NamedRerouteDeclaration(None=@n16) {name=interval5, nodeColor=<rgba 0.2656 0 1 1>}
    color1=NamedRerouteUsage()
    color2=NamedRerouteDeclaration(None=@n17) {name=color5, nodeColor=<rgba 0.0547 1 0 1>}
    interpolation=NamedRerouteUsage()
    prev_color=NamedRerouteUsage()
    interval_num=Constant() {r=4}
    num_intervals=NamedRerouteUsage()
   ) {materialFunction=&MX_RampGradient}.Result) {name=ramp_gradient4, nodeColor=<rgba 1 0 0.375 1>}
  n18 = FunctionInput() {bUsePreviewValueAsDefault=true, inputName=interval6, inputType=FunctionInput_Scalar, sortPriority=14, previewValue=<1 0 0 1>}
  n19 = FunctionInput() {bUsePreviewValueAsDefault=true, inputName=color6, inputType=FunctionInput_Vector4, sortPriority=15, previewValue=<1 1 1 1>}
  NamedRerouteDeclaration(None=MaterialFunctionCall(
    x=NamedRerouteUsage()
    interval1=NamedRerouteUsage()
    interval2=NamedRerouteDeclaration(None=@n18) {name=interval6, nodeColor=<rgba 0.2656 0 1 1>}
    color1=NamedRerouteUsage()
    color2=NamedRerouteDeclaration(None=@n19) {name=color6, nodeColor=<rgba 0.0547 1 0 1>}
    interpolation=NamedRerouteUsage()
    prev_color=NamedRerouteUsage()
    interval_num=Constant() {r=5}
    num_intervals=NamedRerouteUsage()
   ) {materialFunction=&MX_RampGradient}.Result) {name=ramp_gradient5, nodeColor=<rgba 1 0 0.375 1>}
  n20 = FunctionInput() {bUsePreviewValueAsDefault=true, inputName=interval7, inputType=FunctionInput_Scalar, sortPriority=16, previewValue=<1 0 0 1>}
  n21 = FunctionInput() {bUsePreviewValueAsDefault=true, inputName=color7, inputType=FunctionInput_Vector4, sortPriority=17, previewValue=<1 1 1 1>}
  NamedRerouteDeclaration(None=MaterialFunctionCall(
    x=NamedRerouteUsage()
    interval1=NamedRerouteUsage()
    interval2=NamedRerouteDeclaration(None=@n20) {name=interval7, nodeColor=<rgba 0.2656 0 1 1>}
    color1=NamedRerouteUsage()
    color2=NamedRerouteDeclaration(None=@n21) {name=color7, nodeColor=<rgba 0.0547 1 0 1>}
    interpolation=NamedRerouteUsage()
    prev_color=NamedRerouteUsage()
    interval_num=Constant() {r=6}
    num_intervals=NamedRerouteUsage()
   ) {materialFunction=&MX_RampGradient}.Result) {name=ramp_gradient6, nodeColor=<rgba 1 0 0.375 1>}
  n22 = FunctionInput() {bUsePreviewValueAsDefault=true, inputName=interval8, inputType=FunctionInput_Scalar, sortPriority=18, previewValue=<1 0 0 1>}
  n23 = FunctionInput() {bUsePreviewValueAsDefault=true, inputName=color8, inputType=FunctionInput_Vector4, sortPriority=19, previewValue=<1 1 1 1>}
  NamedRerouteDeclaration(None=MaterialFunctionCall(
    x=NamedRerouteUsage()
    interval1=NamedRerouteUsage()
    interval2=NamedRerouteDeclaration(None=@n22) {name=interval8, nodeColor=<rgba 0.2656 0 1 1>}
    color1=NamedRerouteUsage()
    color2=NamedRerouteDeclaration(None=@n23) {name=color8, nodeColor=<rgba 0.0547 1 0 1>}
    interpolation=NamedRerouteUsage()
    prev_color=NamedRerouteUsage()
    interval_num=Constant() {r=7}
    num_intervals=NamedRerouteUsage()
   ) {materialFunction=&MX_RampGradient}.Result) {name=ramp_gradient7, nodeColor=<rgba 1 0 0.375 1>}
  n24 = FunctionInput() {bUsePreviewValueAsDefault=true, inputName=interval9, inputType=FunctionInput_Scalar, sortPriority=20, previewValue=<1 0 0 1>}
  n25 = FunctionInput() {bUsePreviewValueAsDefault=true, inputName=color9, inputType=FunctionInput_Vector4, sortPriority=21, previewValue=<1 1 1 1>}
  NamedRerouteDeclaration(None=MaterialFunctionCall(
    x=NamedRerouteUsage()
    interval1=NamedRerouteUsage()
    interval2=NamedRerouteDeclaration(None=@n24) {name=interval9, nodeColor=<rgba 0.2656 0 1 1>}
    color1=NamedRerouteUsage()
    color2=NamedRerouteDeclaration(None=@n25) {name=color9, nodeColor=<rgba 0.0547 1 0 1>}
    interpolation=NamedRerouteUsage()
    prev_color=NamedRerouteUsage()
    interval_num=Constant() {r=8}
    num_intervals=NamedRerouteUsage()
   ) {materialFunction=&MX_RampGradient}.Result) {name=ramp_gradient8, nodeColor=<rgba 1 0 0.375 1>}
  n26 = FunctionInput(Preview=TextureCoordinate()) {bUsePreviewValueAsDefault=true, inputName=Coordinates, inputType=FunctionInput_Vector2}
  n27 = Abs(None=NamedRerouteDeclaration(None=Subtract(A=@n26) {constB=0.5}) {name=subtract_half, nodeColor=<rgba 0.2891 1 0 1>})
  NamedRerouteDeclaration(None=Multiply(A=NamedRerouteDeclaration(None=@n27) {name=box_absval, nodeColor=<rgba 1 0.75 0 1>}) {constB=2}) {name=box_multiply, nodeColor=<rgba 0.4297 0 1 1>}
  n28 = FunctionInput() {bUsePreviewValueAsDefault=true, inputName=interval10, inputType=FunctionInput_Scalar, sortPriority=22, previewValue=<1 0 0 1>}
  n29 = FunctionInput() {bUsePreviewValueAsDefault=true, inputName=color10, inputType=FunctionInput_Vector4, sortPriority=23, previewValue=<1 1 1 1>}
  NamedRerouteDeclaration(None=MaterialFunctionCall(
    x=NamedRerouteUsage()
    interval1=NamedRerouteUsage()
    interval2=@n28
    color1=NamedRerouteUsage()
    color2=@n29
    interpolation=NamedRerouteUsage()
    prev_color=NamedRerouteUsage()
    interval_num=Constant() {r=9}
    num_intervals=NamedRerouteUsage()
   ) {materialFunction=&MX_RampGradient}.Result) {name=ramp_gradient9, nodeColor=<rgba 1 0 0.375 1>}
  n30 = NamedRerouteUsage()
  n31 = ComponentMask(None=@n30) {r=true}
  n32 = ComponentMask(None=@n30) {g=true}
  n33 = Divide(A=NamedRerouteDeclaration(None=Arctangent2!(Y=@n31, X=@n32)) {name=radial_atan2, nodeColor=<rgba 0 1 0.1094 1>}) {constB=6.2832}
  NamedRerouteDeclaration(None=Add(A=NamedRerouteDeclaration(None=@n33) {name=radial_divide, nodeColor=<rgba 0 0.5312 1 1>}) {constB=0.5}) {name=radial_add, nodeColor=<rgba 0.8516 1 0 1>}
  NamedRerouteDeclaration(None=ComponentMask(None=@n26) {r=true}) {name=texcoord_X, nodeColor=<rgba 0.8281 0 1 1>}
 ]
}
