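# MX_RandomFloat is a library-exposed MaterialX material function that produces a stable pseudo-random float between min and max from an input signal and seed.
# It scales Input by 4096 and packs it with the seed and a constant into a vector that drives a Vector Noise node as its position, giving 2D cell noise.
# The red channel is then remapped to the min/max range with clamping.
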MaterialFunction /InterchangeAssets/Functions/MX_RandomFloat
graph {
 properties {
  bExposeToLibrary=true
  description="Produces a stable randomized float value between 'min' and 'max', based on an 'input' signal and 'seed' value. Uses a 2d cellnoise function to produce the output"
  libraryCategoriesText [MaterialX]
 }
 nodes [
  n1 = FunctionInput() {description=Initial randomization seed, inputName=Input, inputType=FunctionInput_Scalar}
  n2 = Constant() {r=4096}
  n3 = FunctionInput(Preview=Constant()) {
   bUsePreviewValueAsDefault=true
   description=Additional seed
   inputName=seed
   inputType=FunctionInput_Scalar
   sortPriority=3
   previewValue=<1 0 0 1>
  }
  n4 = VectorNoise(World Position=MaterialXAppend3Vector!(A=Multiply(A=@n1, B=@n2), B=@n3, C=Constant()))
  n5 = FunctionInput(Preview=Constant()) {bUsePreviewValueAsDefault=true, description=The minimum output value, inputName=min, inputType=FunctionInput_Scalar, sortPriority=1}
  n6 = FunctionInput(Preview=Constant() {r=1}) {
   bUsePreviewValueAsDefault=true
   description=The maximum output value
   inputName=max
   inputType=FunctionInput_Scalar
   sortPriority=2
   previewValue=<1 0 0 1>
  }
  FunctionOutput(None=MaterialXRange!(None=ComponentMask(None=@n4) {r=true}, InputLow, InputHigh, TargetLow=@n5, TargetHigh=@n6, Gamma, Clamp) {bConstClamp=true})
 ]
}
