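# MaterialX library material function that outputs a random RGB color within given hue, saturation and brightness ranges, based on an input signal and seed.
# Three MX_RandomFloat calls share the input, and MaterialXRange nodes remap their results to the low/high bounds.
# The remapped values form an HSV vector that is converted to RGB.
# Named reroutes define separate hue, saturation and brightness seeds using constant offsets rounded up with Ceil.
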MaterialFunction /InterchangeAssets/Functions/MX_RandomColor
graph {
 properties {
  bExposeToLibrary=true
  description="Produces a randomized RGB color within a randomized hue, saturation and brightness range, based on an 'input' signal and 'seed' value."
  libraryCategoriesText [MaterialX]
 }
 nodes [
  n1 = FunctionInput() {description=" Initial randomization seed", inputName=Input, inputType=FunctionInput_Scalar}
  n2 = FunctionInput(Preview=Constant()) {bUsePreviewValueAsDefault=true, description=The minimum hue value, inputName=huelow, inputType=FunctionInput_Scalar, sortPriority=1}
  n3 = FunctionInput(Preview=Constant() {r=1}) {
   bUsePreviewValueAsDefault=true
   description=The maximum hue value
   inputName=huehigh
   inputType=FunctionInput_Scalar
   sortPriority=2
   previewValue=<1 0 0 1>
  }
  n4 = MaterialXRange!(
   None=MaterialFunctionCall!(In=@n1, min, max, seed=NamedRerouteUsage()) {materialFunction=&/MX_RandomFloat}.Result
   InputLow
   InputHigh
   TargetLow=@n2
   TargetHigh=@n3
   Gamma
   Clamp
  )
  n5 = FunctionInput(Preview=Constant() {r=0.825}) {
   bUsePreviewValueAsDefault=true
   description=The minimum saturation value
   inputName=saturationlow
   inputType=FunctionInput_Scalar
   sortPriority=4
   previewValue=<0.825 0 0 1>
  }
  n6 = FunctionInput(Preview=Constant() {r=1}) {
   bUsePreviewValueAsDefault=true
   description=The maximum saturation value
   inputName=saturationhigh
   inputType=FunctionInput_Scalar
   sortPriority=4
   previewValue=<1 0 0 1>
  }
  n7 = MaterialXRange(
   None=MaterialFunctionCall(In=@n1, seed=NamedRerouteUsage()) {materialFunction=&MX_RandomFloat}.Result
   TargetLow=@n5
   TargetHigh=@n6
  )
  n8 = Constant() {r=1}
  n9 = FunctionInput(Preview=@n8) {
   bUsePreviewValueAsDefault=true
   description=The minimum brightness value
   inputName=brightnesslow
   inputType=FunctionInput_Scalar
   sortPriority=5
  }
  n10 = FunctionInput(Preview=@n8) {
   bUsePreviewValueAsDefault=true
   description=The maximum brightness value
   inputName=brightnesshigh
   inputType=FunctionInput_Scalar
   sortPriority=6
   previewValue=<1 0 0 1>
  }
  n11 = MaterialXRange(
   None=MaterialFunctionCall(In=@n1, seed=NamedRerouteUsage()) {materialFunction=&MX_RandomFloat}.Result
   TargetLow=@n9
   TargetHigh=@n10
  )
  FunctionOutput(None=HsvToRgb!(None=MaterialXAppend3Vector!(A=@n4, B=@n7, C=@n11)))
  n12 = FunctionInput(Preview=Constant()) {bUsePreviewValueAsDefault=true, description=Additional randomization seed, inputName=seed, inputType=FunctionInput_Scalar, sortPriority=7}
  NamedRerouteDeclaration(None=@n12) {name=Seed, nodeColor=<rgba 0.1953 0 1 1>}
  n13 = NamedRerouteUsage()
  n14 = Constant() {r=413.3}
  NamedRerouteDeclaration(None=Ceil(None=Add(A=@n13, B=@n14))) {name=SeedHue, nodeColor=<rgba 0.125 1 0 1>}
  n15 = NamedRerouteUsage()
  n16 = Constant() {r=1522.4}
  NamedRerouteDeclaration(None=Ceil(None=Add(A=@n15, B=@n16))) {name=SeedSaturation, nodeColor=<rgba 1 0 0.4453 1>}
  n17 = NamedRerouteUsage()
  n18 = Constant() {r=1813.8}
  NamedRerouteDeclaration(None=Ceil(None=Add(A=@n17, B=@n18))) {name=SeedBrightness, nodeColor=<rgba 0 0.7656 1 1>}
 ]
}
