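# MaterialX library function that converts scalar roughness and anisotropy inputs into a two-component anisotropic roughness.
# Roughness is squared and saturated for a more linear look, and an aspect is derived as the square root of one minus clamped anisotropy.
# Positive anisotropy stretches roughness along the tangent by dividing one component by aspect, capped with Min, and multiplying the other.
# Otherwise, an If returns squared roughness in both components.
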MaterialFunction /InterchangeAssets/Functions/MX_Roughness_Anisotropy
graph {
 properties {
  bExposeToLibrary=true
  description="Calculates anisotropic surface roughness from a scalar roughness and anisotropy parameterization. An anisotropy value above 0.0 stretches the roughness in the direction of the surface's \"tangent\" vector. An anisotropy value of 0.0 gives isotropic roughness. The roughness value is squared to achieve a more linear roughness look over the input range [0,1]. Output type \"vector2\"."
  libraryCategoriesText [MaterialX]
 }
 nodes [
  n1 = FunctionInput() {
   bUsePreviewValueAsDefault=true
   description="Amount of anisotropy, range [0.0, 1.0]."
   inputName=anisotropy
   inputType=FunctionInput_Scalar
   sortPriority=1
  }
  n2 = NamedRerouteUsage()
  FunctionOutput(None=If(A=@n1, "A > B"=NamedRerouteUsage(), "A == B"=@n2, "A < B"=@n2))
  n3 = FunctionInput() {bUsePreviewValueAsDefault=true, description="Roughness value, range [0.0, 1.0].", inputName=roughness, inputType=FunctionInput_Scalar}
  NamedRerouteDeclaration(None=Saturate(None=Multiply(A=@n3, B=@n3))) {name=roughness_sqr, nodeColor=<rgba 0 1 0.3438 1>}
  NamedRerouteDeclaration(None=@n1) {name=anisotropy, nodeColor=<rgba 0.1016 1 0 1>}
  n4 = Clamp(None=NamedRerouteUsage()) {maxDefault=0.98}
  NamedRerouteDeclaration(None=SquareRoot(None=OneMinus(None=@n4))) {name=aspect, nodeColor=<rgba 0 0.6484 1 1>}
  n5 = Divide(A=NamedRerouteUsage(), B=NamedRerouteUsage())
  n6 = NamedRerouteUsage()
  n7 = NamedRerouteUsage()
  NamedRerouteDeclaration(None=AppendVector(A=Min(A=@n5), B=Multiply(A=@n6, B=@n7))) {name=result_anisotropy_pos, nodeColor=<rgba 1 0 0.6797 1>}
  n8 = NamedRerouteUsage()
  NamedRerouteDeclaration(None=AppendVector(A=@n8, B=@n8)) {name=result_anisotropy_neg, nodeColor=<rgba 1 0 0.8203 1>}
 ]
}
