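# Material function in the MaterialX library category that converts dual x/y roughness into a two-component anisotropic surface roughness.
# It splits the Vector2 input into named reroutes roughness_x and roughness_y, squares each for a more linear look, saturates both and appends them as the output.
# The If node feeding roughness_y returns the Y component on every branch, so it acts as a pass-through with no fallback to X.
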MaterialFunction /InterchangeAssets/Functions/MX_Roughness_Dual
graph {
 properties {
  bExposeToLibrary=true
  description="Calculates anisotropic surface roughness from a dual surface roughness parameterization. The roughness is squared to achieve a more linear roughness look over the input range [0,1]. Output type \"vector2\"."
  libraryCategoriesText [MaterialX]
 }
 nodes [
  n1 = FunctionInput() {
   bUsePreviewValueAsDefault=true
   description="Roughness in x and y directions, range [0.0, 1.0]."
   inputName=roughness
   inputType=FunctionInput_Vector2
  }
  n2 = MaterialFunctionCall(Float2=@n1) {materialFunction=&BreakOutFloat2Components}
  n3 = NamedRerouteDeclaration(None=@n2.R) {name=roughness_x, nodeColor=<rgba 0 1 0.1562 1>}
  n4 = Multiply(A=@n3, B=@n3)
  n5 = NamedRerouteDeclaration(None=If(A=@n2.G, "A > B"=@n2.G, "A == B"=@n2.G, "A < B"=@n2.G)) {name=roughness_y, nodeColor=<rgba 0 0.7891 1 1>}
  n6 = Multiply(A=@n5, B=@n5)
  FunctionOutput(None=AppendVector(A=Saturate(None=@n4), B=Saturate(None=@n6)))
 ]
}
