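# MaterialX library material function that stands in for the sheen BSDF node, described as a cloth-like back-scattering lobe for layering over a base BSDF.
# It declares weight, color, roughness and normal inputs, with the normal preview wired to a +Z constant vector.
# The single output is a MakeMaterialAttributes node with no pins connected, so the function is a placeholder: none of the inputs affect the result and no sheen shading is computed.
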MaterialFunction /InterchangeAssets/Functions/MX_SheenBSDF
graph {
 properties {
  description="Constructs a microfacet BSDF for the back-scattering properties of cloth-like materials. This node may be layered vertically over a base BSDF using a <layer> node. All energy that is not reflected will be transmitted to the base layer."
  libraryCategoriesText [MaterialX]
 }
 nodes [
  FunctionInput() {
   bUsePreviewValueAsDefault=true
   description="Weight for this BSDF’s contribution"
   inputName=weight
   inputType=FunctionInput_Scalar
   previewValue=<1 0 0 1>
  }
  FunctionInput() {bUsePreviewValueAsDefault=true, description=Sheen reflectivity., inputName=color, sortPriority=1, previewValue=<0.18 0.18 0.18 0.18>}
  FunctionInput() {
   bUsePreviewValueAsDefault=true
   description=Surface roughness
   inputName=roughness
   inputType=FunctionInput_Scalar
   sortPriority=2
   previewValue=<0.3 0.3 0.3 0.3>
  }
  FunctionInput(Preview=Constant3Vector() {constant=<rgba 0 0 1 0>}) {description=Normal vector of the surface., inputName=normal, sortPriority=3, previewValue=<0.18 0.18 0.18 0.18>}
  FunctionOutput(None=MakeMaterialAttributes!(
    BaseColor
    Metallic
    Specular
    Roughness
    Anisotropy
    EmissiveColor
    Opacity
    OpacityMask
    Normal
    Tangent
    WorldPositionOffset
    SubsurfaceColor
    ClearCoat
    ClearCoatRoughness
    AmbientOcclusion
    Refraction
    CustomizedUVs_0
    CustomizedUVs_1
    CustomizedUVs_2
    CustomizedUVs_3
    CustomizedUVs_4
    CustomizedUVs_5
    CustomizedUVs_6
    CustomizedUVs_7
    PixelDepthOffset
    ShadingModel
    Displacement
   )) {outputName=Output}
 ]
}
