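# MX_SrgbToRec709 is a library material function in the MaterialX category that converts an sRGB color input to linear Rec709.
# It computes both a linear segment and a clamped, offset power segment of the transfer curve for the whole color.
# Per-channel If nodes compare against a threshold, are appended into a mask, and drive a lerp that selects the segment branchlessly, with named reroutes linking the stages.
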MaterialFunction /InterchangeAssets/Functions/MX_SrgbToRec709
graph {
 properties {
  bExposeToLibrary=true
  description=Color Transform from srgb to Linear Rec709
  userExposedCaption=sRGBToRec709
  libraryCategoriesText [MaterialX]
 }
 nodes [
  FunctionOutput(None=LinearInterpolate(A=NamedRerouteUsage(), B=NamedRerouteUsage(), Alpha=NamedRerouteUsage()))
  NamedRerouteDeclaration(None=Constant() {r=0.0404}) {name=Threshold, nodeColor=<rgba 0.4297 0 1 1>}
  n1 = FunctionInput() {bUsePreviewValueAsDefault=true, inputName=Input}
  n2 = MaterialFunctionCall(Float3=@n1) {materialFunction=&BreakOutFloat3Components}
  NamedRerouteDeclaration(None=@n2.R) {name=R, nodeColor=<rgba 0 0.8359 1 1>}
  NamedRerouteDeclaration(None=Divide(A=@n1) {constB=12.92}) {name=LinSeg, nodeColor=<rgba 1 0 0.7734 1>}
  n3 = Max(A=Add(A=@n1) {constB=0.055}) {constB=0}
  NamedRerouteDeclaration(None=Power(Base=Divide(A=@n3) {constB=1.055}) {constExponent=2.4}) {name=PowSeg, nodeColor=<rgba 1 0.5859 0 1>}
  NamedRerouteDeclaration(None=@n2.G) {name=G, nodeColor=<rgba 0 1 0 1>}
  NamedRerouteDeclaration(None=@n2.B) {name=B, nodeColor=<rgba 0 0 1 1>}
  n4 = Constant()
  NamedRerouteDeclaration(None=If(A=NamedRerouteUsage(), B=NamedRerouteUsage(), "A > B"=Constant() {r=1}, "A == B"=@n4, "A < B"=@n4)) {name=IsAboveR, nodeColor=<rgba 0 0.4375 1 1>}
  n5 = Constant()
  NamedRerouteDeclaration(None=If(A=NamedRerouteUsage(), B=NamedRerouteUsage(), "A > B"=Constant() {r=1}, "A == B"=@n5, "A < B"=@n5)) {name=IsAboveG, nodeColor=<rgba 0 1 0 1>}
  n6 = Constant()
  NamedRerouteDeclaration(None=If(A=NamedRerouteUsage(), B=NamedRerouteUsage(), "A > B"=Constant() {r=1}, "A == B"=@n6, "A < B"=@n6)) {name=IsAboveB, nodeColor=<rgba 0 0 1 1>}
  NamedRerouteDeclaration(None=MaterialXAppend3Vector!(A=NamedRerouteUsage(), B=NamedRerouteUsage(), C=NamedRerouteUsage())) {name=IsAbove, nodeColor=<rgba 0.4531 0.5658 0.1575 1>}
 ]
}
