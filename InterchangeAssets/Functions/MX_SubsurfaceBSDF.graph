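# Material function for the MaterialX library that stands in for the MaterialX subsurface scattering BSDF, apparently a placeholder.
# It declares weight, color, radius, anisotropy and normal inputs with descriptions and preview defaults, the normal previewing as +Z.
# None of these inputs are wired: the output is a MakeMaterialAttributes node with nothing connected, so it returns default attributes and computes no scattering.
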MaterialFunction /InterchangeAssets/Functions/MX_SubsurfaceBSDF
graph {
 properties {
  description="Constructs a subsurface scattering BSDF for subsurface scattering within a homogeneous medium. The parameterization is chosen to match random walk Monte Carlo methods as well as approximate empirical methods. Note that this category of subsurface scattering can be defined more rigorously as a BSDF vertically layered over an <anisotropic_vdf>, and we expect these two descriptions of the scattering-surface distribution function to be unified in future versions of MaterialX."
  libraryCategoriesText [MaterialX]
 }
 nodes [
  FunctionInput() {
   bUsePreviewValueAsDefault=true
   description="Weight for this BSDF’s contribution"
   inputName=weight
   inputType=FunctionInput_Scalar
   previewValue=<1 0 0 1>
  }
  FunctionInput() {
   bUsePreviewValueAsDefault=true
   description="Diffuse reflectivity (albedo)."
   inputName=color
   sortPriority=1
   previewValue=<0.18 0.18 0.18 0.18>
  }
  FunctionInput() {
   bUsePreviewValueAsDefault=true
   description=Sets the average distance that light might propagate below the surface before scattering back out. This is also known as the mean free path of the material. The radius can be set for each color component separately.
   inputName=radius
   sortPriority=2
   previewValue=<1 1 1 1>
  }
  FunctionInput(Preview=Constant3Vector() {constant=<rgba 0 0 1 0>}) {description=Normal vector of the surface., inputName=normal, sortPriority=4, previewValue=<0.18 0.18 0.18 0.18>}
  FunctionInput() {
   bUsePreviewValueAsDefault=true
   description="Anisotropy factor, controlling the scattering direction, range [-1.0, 1.0]. Negative values give backwards scattering, positive values give forward scattering, and a value of zero gives uniform scattering."
   inputName=anisotropy
   inputType=FunctionInput_Scalar
   sortPriority=3
   previewValue=<0 0 0 0>
  }
  FunctionOutput(None=MakeMaterialAttributes!(
    BaseColor
    Metallic
    Specular
    Roughness
    Anisotropy
    EmissiveColor
    Opacity
    OpacityMask
    Normal
    Tangent
    WorldPositionOffset
    SubsurfaceColor
    ClearCoat
    ClearCoatRoughness
    AmbientOcclusion
    Refraction
    CustomizedUVs_0
    CustomizedUVs_1
    CustomizedUVs_2
    CustomizedUVs_3
    CustomizedUVs_4
    CustomizedUVs_5
    CustomizedUVs_6
    CustomizedUVs_7
    PixelDepthOffset
    ShadingModel
    Displacement
   )) {outputName=Output}
 ]
}
