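# Material layer blend function, exposed to the library, that overrides the Opacity of an input material attributes set.
# It reads the incoming Opacity and lerps it toward a NewOpacity scalar using a Mask scalar as alpha.
# The result is written back into the attributes and output as Blended Material, with only Opacity changed.
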MaterialFunction /Engine/Functions/MaterialLayerFunctions/MatLayerBlend_OverrideOpacity
graph {
 properties {bExposeToLibrary=true, description=Replaced the Opacity, libraryCategoriesText [Misc, MaterialLayerBlend]}
 nodes [
  n1 = GetMaterialAttributes(None=FunctionInput() {bUsePreviewValueAsDefault=true, inputName=Material, inputType=FunctionInput_MaterialAttributes}) {attributeGetTypes [#1]}
  n2 = FunctionInput() {bUsePreviewValueAsDefault=true, inputName=NewOpacity, inputType=FunctionInput_Scalar, sortPriority=1, previewValue=<1 1 1 1>}
  n3 = FunctionInput() {bUsePreviewValueAsDefault=true, inputName=Mask, inputType=FunctionInput_Scalar, sortPriority=2, previewValue=<1 1 1 1>}
  FunctionOutput(None=SetMaterialAttributes!(MaterialAttributes=@n1.MaterialAttributes, Opacity=LinearInterpolate(A=@n1.Opacity, B=@n2, Alpha=@n3)) {attributeSetTypes [#1]}) {outputName=Blended Material, sortPriority=-20}
 ]
}
