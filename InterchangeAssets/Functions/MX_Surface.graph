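# MX_Surface is a library-exposed MaterialX material function that builds a surface shader, output as Material Attributes, from scattering and emission inputs.
# It takes the bsdf input's attributes and replaces their Emissive Color with the one extracted from the edf input.
# The result passes through MatLayerBlend_OverrideOpacity to apply the scalar cutout opacity input, which defaults to fully opaque, and exits as Surface.
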MaterialFunction /InterchangeAssets/Functions/MX_Surface
graph {
 properties {
  bExposeToLibrary=true
  description="Constructs a surface shader describing light scattering and emission for closed”thick” objects. If the BSDF is opaque only the front side will reflect light and back sides will beblack if visible. If the BSDF is transmissive the surface will act as an interface to a solid volumemade from this material. Output type \"MaterialAttributes\"."
  libraryCategoriesText [MaterialX]
 }
 nodes [
  n1 = GetMaterialAttributes(None=FunctionInput() {description=Distribution function for surface scattering., inputName=bsdf, inputType=FunctionInput_MaterialAttributes}) {attributeGetTypes [#1]}
  n2 = FunctionInput() {description=Distribution function for surface emission., inputName=edf, inputType=FunctionInput_MaterialAttributes, sortPriority=1}
  n3 = GetMaterialAttributes(None=@n2) {attributeGetTypes [#1]}
  n4 = FunctionInput() {
   bUsePreviewValueAsDefault=true
   description=Surface cutout opacity
   inputName=opacity
   inputType=FunctionInput_Scalar
   sortPriority=2
   previewValue=<1 1 1 1>
  }
  FunctionOutput(None=MaterialFunctionCall!(
    Material=SetMaterialAttributes!(MaterialAttributes=@n1.MaterialAttributes, EmissiveColor=@n3.EmissiveColor) {attributeSetTypes [#1]}
    NewOpacity=@n4
    Mask
   ) {materialFunction=/Engine/Functions/MaterialLayerFunctions/MatLayerBlend_OverrideOpacity}.Blended Material) {outputName=Surface}
 ]
}
