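# MX_ThinFilmBSDF is a placeholder material function for the MaterialX thin-film BSDF node in InterchangeAssets, marked deprecated since MaterialX 1.39.3.
# It declares scalar thickness and ior inputs with preview values as defaults, but neither feeds into anything.
# The single output is a MakeMaterialAttributes node with no pins connected, so it returns default material attributes and computes no thin-film effect.
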MaterialFunction /InterchangeAssets/Functions/MX_ThinFilmBSDF
graph {
 properties {description=DEPRECATED since MaterialX 1.39.3, libraryCategoriesText [MaterialX]}
 nodes [
  FunctionInput() {bUsePreviewValueAsDefault=true, inputName=thickness, inputType=FunctionInput_Scalar, previewValue=<550 550 550 550>}
  FunctionInput() {bUsePreviewValueAsDefault=true, inputName=ior, inputType=FunctionInput_Scalar, previewValue=<1.5 1.5 1.5 1.5>}
  FunctionOutput(None=MakeMaterialAttributes!(
    BaseColor
    Metallic
    Specular
    Roughness
    Anisotropy
    EmissiveColor
    Opacity
    OpacityMask
    Normal
    Tangent
    WorldPositionOffset
    SubsurfaceColor
    ClearCoat
    ClearCoatRoughness
    AmbientOcclusion
    Refraction
    CustomizedUVs_0
    CustomizedUVs_1
    CustomizedUVs_2
    CustomizedUVs_3
    CustomizedUVs_4
    CustomizedUVs_5
    CustomizedUVs_6
    CustomizedUVs_7
    PixelDepthOffset
    ShadingModel
    Displacement
   )) {outputName=Output}
 ]
}
