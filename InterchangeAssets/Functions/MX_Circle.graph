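# MX_Circle is a MaterialX-category material function in the Interchange assets that returns a hard-edged circle mask in 2D coordinate space.
# It subtracts the center from the input coordinates and takes the dot product of that offset with itself to get squared distance.
# An If node compares this to the squared radius, avoiding a square root, and outputs 1 on or inside the circle and 0 outside.
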MaterialFunction /InterchangeAssets/Functions/MX_Circle
graph {
 properties {
  bExposeToLibrary=true
  description="Returns 1 if texcoord is inside a circle defined by center and radius; otherwise returns 0."
  libraryCategoriesText [MaterialX]
 }
 nodes [
  n1 = FunctionInput(Preview=TextureCoordinate()) {
   bUsePreviewValueAsDefault=true
   description=The input 2d space. Default is the first texture coordinates.
   inputName=Coordinates
   inputType=FunctionInput_Vector2
  }
  n2 = FunctionInput() {
   bUsePreviewValueAsDefault=true
   description=The center coordinate of the circle
   inputName=center
   inputType=FunctionInput_Vector2
   sortPriority=1
  }
  n3 = Subtract(A=@n1, B=@n2)
  n4 = FunctionInput() {
   bUsePreviewValueAsDefault=true
   description=The radius of the circle
   inputName=radius
   inputType=FunctionInput_Scalar
   sortPriority=2
   previewValue=<0.5 0 0 1>
  }
  n5 = Constant() {r=1}
  FunctionOutput(None=If(A=DotProduct(A=@n3, B=@n3), B=Multiply(A=@n4, B=@n4), "A > B"=Constant(), "A == B"=@n5, "A < B"=@n5))
 ]
}
