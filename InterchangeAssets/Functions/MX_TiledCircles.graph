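# MX_TiledCircles is a MaterialX library material function that draws a black and white grid of circles with adjustable tiling, offset and diameter, output as float3.
# Tiled, offset UVs are wrapped with MaterialXMod and fed to MX_Circle.
# The staggered variant shifts alternate rows by half, uses 0.866 row spacing for an equilateral triangle layout, and Max-combines three circles.
# An If on the staggered input selects regular or staggered output.
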MaterialFunction /InterchangeAssets/Functions/MX_TiledCircles
graph {
 properties {
  bExposeToLibrary=true
  description="Creates a black and white pattern of circles with a defined tiling and size (diameter). Pattern can be regular or staggered."
  libraryCategoriesText [MaterialX]
 }
 nodes [
  n1 = FunctionInput(Preview=StaticBool()) {
   bUsePreviewValueAsDefault=true
   description="If true, every other row will be offset 50%, and the spacing of the tiling will be adjusted in the V direction to center the circles on the vertices of an equilateral triangle grid."
   inputName=staggered
   inputType=FunctionInput_StaticBool
   sortPriority=4
   previewValue=<0.5 0 0 1>
  }
  n2 = NamedRerouteUsage()
  n3 = If(A=@n1, B=StaticBool() {value=true}, "A > B"=@n2, "A == B"=NamedRerouteUsage(), "A < B"=@n2)
  FunctionOutput(None=MaterialFunctionCall(X=@n3, Y=@n3, Z=@n3) {materialFunction=&MakeFloat3}.Result)
  n4 = FunctionInput(Preview=TextureCoordinate()) {
   bUsePreviewValueAsDefault=true
   description=The input 2d space. Default is the first texture coordinates.
   inputName=Coordinates
   inputType=FunctionInput_Vector2
  }
  n5 = FunctionInput() {
   bUsePreviewValueAsDefault=true
   description="Tiling factor, with higher values producing a denser grid."
   inputName=uvtiling
   inputType=FunctionInput_Vector2
   sortPriority=1
   previewValue=<1 1 0 1>
  }
  n6 = Multiply(A=@n4, B=@n5)
  n7 = FunctionInput() {bUsePreviewValueAsDefault=true, description=UV Offset, inputName=uvoffset, inputType=FunctionInput_Vector2, sortPriority=2}
  n8 = MaterialXMod!(A=NamedRerouteDeclaration(None=Subtract(A=@n6, B=@n7)) {name=texcoord_bias, nodeColor=<rgba 1 0 0.5156 1>}, B)
  n9 = Constant() {r=2}
  NamedRerouteDeclaration(None=Subtract(A=Multiply(A=@n8, B=@n9), B=Constant() {r=1})) {name=recenter, nodeColor=<rgba 0 0.8359 1 1>}
  NamedRerouteDeclaration(None=AppendVector(A=NamedRerouteUsage(), B=NamedRerouteUsage())) {name=coord_circ3, nodeColor=<rgba 0 0.1328 1 1>}
  n10 = FunctionInput() {
   bUsePreviewValueAsDefault=true
   description="The diameter of the circles in the tiled pattern; if size is 1.0, the edges of adjacent circles in the tiling will exactly touch."
   inputName=size
   inputType=FunctionInput_Scalar
   sortPriority=3
   previewValue=<0.5 0 0 1>
  }
  NamedRerouteDeclaration(None=Multiply(A=NamedRerouteDeclaration(None=@n10) {name=size, nodeColor=<rgba 0.875 0 1 1>}, B=Constant() {r=0.5})) {name=scale_half, nodeColor=<rgba 0 1 0.3438 1>}
  n11 = NamedRerouteDeclaration(None=MaterialFunctionCall!(Coordinates=NamedRerouteUsage(), center, radius=NamedRerouteUsage()) {materialFunction=&/MX_Circle}.Result) {name=circle_stagg1, nodeColor=<rgba 1 0 0.0703 1>}
  n12 = NamedRerouteDeclaration(None=MaterialFunctionCall(Coordinates=NamedRerouteUsage(), radius=NamedRerouteUsage()) {materialFunction=&MX_Circle}.Result) {name=circle_stagg2, nodeColor=<rgba 0.542 1 0.8171 1>}
  n13 = MaterialFunctionCall(Coordinates=NamedRerouteUsage(), radius=NamedRerouteUsage()) {materialFunction=&MX_Circle}
  NamedRerouteDeclaration(None=Max(A=Max(A=@n11, B=@n12), B=NamedRerouteDeclaration(None=@n13.Result) {name=circle_stagg3, nodeColor=<rgba 1 0 0.0703 1>})) {name=circle_staggered, nodeColor=<rgba 1 0.1875 0 1>}
  NamedRerouteDeclaration(None=MaterialFunctionCall(Coordinates=NamedRerouteUsage(), radius=NamedRerouteUsage()) {materialFunction=&MX_Circle}.Result) {name=circle_regular, nodeColor=<rgba 1 0.3047 0 1>}
  NamedRerouteDeclaration(None=AppendVector(A=NamedRerouteUsage(), B=NamedRerouteUsage())) {name=coord_circ1, nodeColor=<rgba 0.7578 0 1 1>}
  n14 = NamedRerouteUsage()
  n15 = Constant()
  n16 = If(
   A=MaterialXMod(A=ComponentMask(None=@n14) {g=true}, B=Constant() {r=1.732})
   B=Constant() {r=0.866}
   "A > B"=Constant() {r=0.5}
   "A == B"=@n15
   "A < B"=@n15
  )
  n17 = ComponentMask(None=NamedRerouteUsage()) {r=true}
  n18 = NamedRerouteDeclaration(None=MaterialXMod(A=Add(A=@n16, B=@n17))) {name=mod_X_1, nodeColor=<rgba 0 0.0391 1 1>}
  NamedRerouteDeclaration(None=OneMinus(None=@n18)) {name=coord_adj_1, nodeColor=<rgba 0 1 0.8359 1>}
  NamedRerouteDeclaration(None=Subtract(A=@n18, B=Constant() {r=0.5})) {name=coord_adj_2, nodeColor=<rgba 1 0.5156 0 1>}
  n19 = Constant() {r=0.866}
  n20 = MaterialXMod(A=ComponentMask(None=NamedRerouteUsage()) {g=true}, B=@n19)
  NamedRerouteDeclaration(None=Subtract(A=@n19, B=NamedRerouteDeclaration(None=@n20) {name=mod_Y_1, nodeColor=<rgba 0.3594 1 0 1>})) {name=coord_adj_3, nodeColor=<rgba 1 0 0.0469 1>}
  NamedRerouteDeclaration(None=AppendVector(A=NamedRerouteUsage(), B=NamedRerouteUsage())) {name=coord_circ2, nodeColor=<rgba 0 0.9297 1 1>}
 ]
}
