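# Material function providing MaterialX-style unified 3D noise, selectable between Perlin, Cell, Worley and Fractal, with scaled object-space position as default input.
# Position is scaled by frequency and offset, with a jitter-driven rotation branch for varying results.
# VectorNoise nodes and a MaterialXFractal call feed a Switch through named reroutes, then MaterialXRange remaps to outmin/outmax with optional clamping.
# The Worley style input is converted to an integer but left unconnected.
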MaterialFunction /InterchangeAssets/Functions/MX_UnifiedNoise3d
graph {nodes [
  n1 = FunctionInput() {
   bUsePreviewValueAsDefault=true
   description=Sets the style of cell used when Noise Type is set to Worley.
   inputName=style
   inputType=FunctionInput_Scalar
   sortPriority=11
  }
  FloatToUInt!(None=@n1)
  n2 = FunctionInput() {
   bUsePreviewValueAsDefault=true
   description="Menu to select the type of noise: Perlin, Cell, Worley, or Fractal. Default is Perlin."
   inputName=type
   inputType=FunctionInput_Scalar
   sortPriority=10
  }
  MaterialExpressionNamedRerouteUsage_6 = NamedRerouteUsage()
  MaterialExpressionNamedRerouteUsage_7 = NamedRerouteUsage()
  MaterialExpressionNamedRerouteUsage_8 = NamedRerouteUsage()
  MaterialExpressionNamedRerouteUsage_9 = NamedRerouteUsage()
  n3 = Switch!(
   SwitchValue=@n2
   Default=@MaterialExpressionNamedRerouteUsage_6
   PerlinNoise3D=@MaterialExpressionNamedRerouteUsage_6
   CellNoise3D=@MaterialExpressionNamedRerouteUsage_7
   WorleyNoise3D=@MaterialExpressionNamedRerouteUsage_8
   Fractal3D=@MaterialExpressionNamedRerouteUsage_9
  ) {inputs [
    {
     inputName=PerlinNoise3D
     input {
      expression=&self:MaterialExpressionNamedRerouteUsage_6
      inputName=None
      mask=0
      maskA=0
      maskB=0
      maskG=0
      maskR=0
      outputIndex=0
     }
    }
    {
     inputName=CellNoise3D
     input {
      expression=&self:MaterialExpressionNamedRerouteUsage_7
      inputName=None
      mask=0
      maskA=0
      maskB=0
      maskG=0
      maskR=0
      outputIndex=0
     }
    }
    {
     inputName=WorleyNoise3D
     input {
      expression=&self:MaterialExpressionNamedRerouteUsage_8
      inputName=None
      mask=0
      maskA=0
      maskB=0
      maskG=0
      maskR=0
      outputIndex=0
     }
    }
    {
     inputName=Fractal3D
     input {
      expression=&self:MaterialExpressionNamedRerouteUsage_9
      inputName=None
      mask=0
      maskA=0
      maskB=0
      maskG=0
      maskR=0
      outputIndex=0
     }
    }
   ]}
  n4 = FunctionInput() {
   bUsePreviewValueAsDefault=true
   description=The lowest values fit to the noise.
   inputName=outmin
   inputType=FunctionInput_Scalar
   sortPriority=4
  }
  n5 = FunctionInput() {
   bUsePreviewValueAsDefault=true
   description=The highest values fit to the noise.
   inputName=outmax
   inputType=FunctionInput_Scalar
   sortPriority=5
   previewValue=<1 0 0 1>
  }
  n6 = FunctionInput(Preview=StaticBool() {value=true}) {
   bUsePreviewValueAsDefault=true
   description=Clamp the output to the min and max output values.
   inputName=clampoutput
   inputType=FunctionInput_StaticBool
   sortPriority=6
   previewValue=<1 0 0 1>
  }
  FunctionOutput(None=MaterialXRange!(None=@n3, InputLow, InputHigh, TargetLow=@n4, TargetHigh=@n5, Gamma, Clamp=@n6) {bConstClamp=true})
  n7 = LocalPosition()
  n8 = Constant() {r=0.01}
  n9 = NamedRerouteDeclaration(None=FunctionInput(Preview=Multiply(A=@n7.XYZ, B=@n8)) {bUsePreviewValueAsDefault=true, description=The input 3d space. Default is position in object-space., inputName=position}) {name=Position, nodeColor=<rgba 0.2656 1 0 1>}
  n10 = FunctionInput() {
   bUsePreviewValueAsDefault=true
   description="Adjusts the noise frequency, with higher values producing smaller noise shapes."
   inputName=freq
   sortPriority=1
   previewValue=<1 1 1 1>
  }
  NamedRerouteDeclaration(None=Add(
    A=Multiply(A=@n9, B=@n10)
    B=FunctionInput() {bUsePreviewValueAsDefault=true, description=Shift the noise in 3d space., inputName=offset, sortPriority=2}
   )) {name=ApplyOffset, nodeColor=<rgba 0.2188 0 1 1>}
  n11 = FunctionInput() {
   bUsePreviewValueAsDefault=true
   description="Adjust uniformity of Worley noise; for other noise types jitters the results."
   inputName=jitter
   inputType=FunctionInput_Scalar
   sortPriority=3
   previewValue=<1 0 0 1>
  }
  n12 = Constant() {r=1}
  NamedRerouteDeclaration(None=Multiply(A=Subtract(A=@n11, B=@n12), B=Constant() {r=90000})) {name=CellJitterMult, nodeColor=<rgba 0.1016 1 0 1>}
  n13 = Constant3Vector() {constant=<rgba 0.1 1 0 1>}
  n14 = NamedRerouteUsage()
  n15 = Constant3Vector() {constant=<rgba 0 0 0 1>}
  n16 = NamedRerouteUsage()
  NamedRerouteDeclaration(None=Add(A=RotateAboutAxis(NormalizedRotationAxis=@n13, RotationAngle=@n14, PivotPoint=@n15, Position=@n16) {period=360}, B=@n16)) {name=ApplyCellJitter, nodeColor=<rgba 1 0.9375 0 1>}
  n17 = FunctionInput() {
   bUsePreviewValueAsDefault=true
   description=The number of octaves of Fractal noise to be generated.
   inputName=octaves
   inputType=FunctionInput_Scalar
   sortPriority=7
   previewValue=<3 0 0 1>
  }
  n18 = FunctionInput() {
   bUsePreviewValueAsDefault=true
   description=The exponential scale between successive octaves of Fractal noise.
   inputName=lacunarity
   inputType=FunctionInput_Scalar
   sortPriority=8
   previewValue=<2 0 0 1>
  }
  n19 = FunctionInput() {
   bUsePreviewValueAsDefault=true
   description=The rate at which noise amplitude is diminished for each octave of Fractal noise.
   inputName=diminish
   inputType=FunctionInput_Scalar
   sortPriority=9
   previewValue=<0.5 0 0 1>
  }
  MaterialExpressionMaterialXFractal3D_0 = MaterialXFractal!(Position=NamedRerouteUsage(), Amplitude, Octaves=@n17, Lacunarity=@n18, Diminish=@n19)
  NamedRerouteDeclaration(None=ComponentMask(None=@MaterialExpressionMaterialXFractal3D_0.R) {g=true}) {name=Fractal, nodeColor=<rgba 0.2656 1 0 1>}
  n20 = NamedRerouteUsage()
  NamedRerouteDeclaration(None=ComponentMask(None=VectorNoise(World Position=@n20) {noiseFunction=VNF_VoronoiALU, quality=3}) {a=true}) {name=WorleyNoise, nodeColor=<rgba 1 0 0.5859 1>}
  n21 = NamedRerouteUsage()
  NamedRerouteDeclaration(None=ComponentMask(None=VectorNoise(World Position=@n21)) {r=true}) {name=CellNoise, nodeColor=<rgba 0 0.5078 1 1>}
  n22 = Constant() {r=0.5}
  n23 = Multiply(A=VectorNoise(World Position=NamedRerouteUsage()) {noiseFunction=VNF_VectorALU}, B=@n22) {constB=0.5}
  NamedRerouteDeclaration(None=ComponentMask(None=Add(A=@n23, B=@n22) {constB=0.5}) {r=true}) {name=PerlinNoise, nodeColor=<rgba 0.5937 1 0 1>}
 ]}
