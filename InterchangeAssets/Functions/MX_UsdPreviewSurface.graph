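# MaterialX library material function that exposes USD Preview Surface parameters as Unreal surface outputs such as BaseColor, Specular, Clearcoat, Refraction and Displacement.
# Inputs default to their preview values and mostly pass straight through, with ior feeding Refraction.
# Opacity becomes binary: saturated opacity below opacityThreshold outputs 1, otherwise 0, the inverse of the threshold input's description.
# The useSpecularWorkflow switch is unconnected, so both workflows' inputs always pass through.
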MaterialFunction /InterchangeAssets/Functions/MX_UsdPreviewSurface
graph {
 properties {bExposeToLibrary=true, libraryCategoriesText [MaterialX]}
 nodes [
  FunctionInput() {
   bUsePreviewValueAsDefault=true
   description="This node can fundamentally operate in two modes : Specular workflow where you provide a texture/value to the specularColor input. Or, Metallic workflow where you provide a texture/value to the metallic input. Depending on the 0 or 1 value of this parameter, the following parameters are conditionally enabled: -useSpecularWorkflow = 1: (Specular workflow) -useSpecularWorkflow = 0: (Metalness workflow)"
   inputName=useSpecularWorkflow
   inputType=FunctionInput_Bool
   sortPriority=13
   previewValue=<0 0 0 0>
  }
  n1 = FunctionInput() {
   bUsePreviewValueAsDefault=true
   description="Parameter used as diffuseColor when using the specular workflow, when using metallic workflow this is interpreted as albedo."
   inputName=diffuseColor
   previewValue=<0.18 0.18 0.18 0.18>
  }
  FunctionOutput(None=@n1) {outputName=BaseColor}
  n2 = FunctionInput() {
   bUsePreviewValueAsDefault=true
   description="Used only in the metalness workflow.     1 for metallic surfaces and 0 for non-metallic. - If metallic is 1, then both F0 (reflectivity at 0 degree incidence) and edge F90 reflectivity will simply be the Albedo.               - If metallic is 0, then Albedo is ignored in the calculation of F0 and F90; F0 is derived from ior via ( (1-ior)/(1+ior) )^2 and F90 is white.                                In between, we interpolate."
   inputName=metallic
   inputType=FunctionInput_Scalar
   sortPriority=3
  }
  FunctionOutput(None=@n2) {outputName=Metallic, sortPriority=1}
  n3 = FunctionInput() {
   bUsePreviewValueAsDefault=true
   description=Roughness for the second specular lobe.
   inputName=clearcoatRoughness
   inputType=FunctionInput_Scalar
   sortPriority=6
   previewValue=<0.01 0.01 0.01 0.01>
  }
  FunctionOutput(None=@n3) {outputName=ClearcoatRoughness, sortPriority=8}
  n4 = FunctionInput() {
   bUsePreviewValueAsDefault=true
   description=Displacement in the direction of the normal.
   inputName=displacement
   inputType=FunctionInput_Scalar
   sortPriority=11
   previewValue=<0 0 0 0>
  }
  FunctionOutput(None=@n4) {outputName=Displacement, sortPriority=9}
  n5 = FunctionInput() {
   bUsePreviewValueAsDefault=true
   description=Used only in the specular workflow. Specular color to be used. This is the color at 0 incidence. Edge color is assumed white. Transition between the two colors according to Schlick fresnel approximation.
   inputName=specularColor
   sortPriority=2
  }
  FunctionOutput(None=@n5) {outputName=Specular, sortPriority=2}
  n6 = FunctionInput() {
   bUsePreviewValueAsDefault=true
   description="Roughness for the specular lobe. The value ranges from 0 to 1, which goes from a perfectly specular surface at 0.0 to maximum roughness of the specular lobe. This value is usually squared before use with a GGX or Beckmann lobe."
   inputName=roughness
   inputType=FunctionInput_Scalar
   sortPriority=4
   previewValue=<0.5 0.5 0.5 0.5>
  }
  FunctionOutput(None=@n6) {outputName=Roughness, sortPriority=3}
  n7 = FunctionInput() {
   bUsePreviewValueAsDefault=true
   description=Opacity of the material.
   inputName=opacity
   inputType=FunctionInput_Scalar
   sortPriority=7
   previewValue=<1 1 1 1>
  }
  n8 = FunctionInput() {
   bUsePreviewValueAsDefault=true
   description="Threshold used to determine opacity values that will beconsidered fully transparent. A value of 0.0 indicates that no maskingis applied to the opacity input, while a value greater than 0.0 indicatesthat rendering of the surface is limited to the areas where the opacityis greater or equal to that value. Note that when opacityThreshold isgreater than zero, the opacity values less than the opacityThreshold willnot be rendered, and the opacity values greater than or equal to theopacityThreshold will be fully visible."
   inputName=opacityThreshold
   inputType=FunctionInput_Scalar
   sortPriority=8
   previewValue=<0 0 0 0>
  }
  n9 = Constant()
  FunctionOutput(None=If(A=Saturate(None=@n7), B=@n8, "A > B"=@n9, "A == B"=@n9, "A < B"=Constant() {r=1})) {outputName=Opacity, sortPriority=5}
  n10 = FunctionInput() {
   bUsePreviewValueAsDefault=true
   description="Expects normal in tangent space [(-1,-1,-1), (1,1,1)]This means your texture reader implementation should provide data to this node that is properly scaled and readyto be consumed as a tangent space normal."
   inputName=normal
   sortPriority=10
   previewValue=<0 0 1 1>
  }
  FunctionOutput(None=@n10) {outputName=Normal, sortPriority=6}
  FunctionOutput(None=FunctionInput() {bUsePreviewValueAsDefault=true, description=Emissive component., inputName=emissiveColor, sortPriority=1}) {outputName=EmissiveColor, sortPriority=4}
  n11 = FunctionInput() {
   bUsePreviewValueAsDefault=true
   description="Occlusion signal. This provides extra information about theocclusion of different parts of the mesh that this material is appliedto. Occlusion only makes sense as a surface-varying signal, andpathtracers will likely choose to ignore it.  An occlusion value of 0.0means the surface point is fully occluded by other parts of the surface,and a value of 1.0 means the surface point is completely unoccluded byother parts of the surface."
   inputName=occlusion
   inputType=FunctionInput_Scalar
   sortPriority=12
   previewValue=<1 1 1 0>
  }
  FunctionOutput(None=@n11) {outputName=Occlusion, sortPriority=9}
  n12 = FunctionInput() {
   bUsePreviewValueAsDefault=true
   description="Index of Refraction to be used for translucent objects and objects with specular components, including the clearcoat if clearcoat > 0."
   inputName=ior
   inputType=FunctionInput_Scalar
   sortPriority=9
   previewValue=<1.5 1.5 1.5 1.5>
  }
  FunctionOutput(None=@n12) {outputName=Refraction, sortPriority=10}
  n13 = FunctionInput() {
   bUsePreviewValueAsDefault=true
   description=Second specular lobe amount. The color is white.
   inputName=clearcoat
   inputType=FunctionInput_Scalar
   sortPriority=5
  }
  FunctionOutput(None=@n13) {outputName=Clearcoat, sortPriority=7}
 ]
}
