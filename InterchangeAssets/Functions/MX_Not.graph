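# MX_Not is a material function in the MaterialX library category, under InterchangeAssets, that performs a logical NOT on a scalar.
# The scalar Input, which falls back to its preview value when unconnected, feeds a Custom HLSL node that returns !A to the single output.
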MaterialFunction /InterchangeAssets/Functions/MX_Not
graph {
 properties {bExposeToLibrary=true, libraryCategoriesText [MaterialX]}
 nodes [
  MaterialExpressionFunctionInput_0 = FunctionInput() {bUsePreviewValueAsDefault=true, inputName=Input, inputType=FunctionInput_Scalar}
  n1 = Custom!(A=@MaterialExpressionFunctionInput_0) {
   code="return !A;"
   showCode=true
   inputs [{inputName=A, input {expression=&self:MaterialExpressionFunctionInput_0}}]
  }
  FunctionOutput(None=@n1)
 ]
}
