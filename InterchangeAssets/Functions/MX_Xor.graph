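# MX_Xor is a material function in InterchangeAssets that computes a logical exclusive-or of two scalar inputs, A and B.
# It nests other MX functions: MX_Not inverts each input, two MX_And calls form (not A and B) and (A and not B), and MX_Or merges them into the output.
# Both inputs use their preview values as defaults when unconnected.
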MaterialFunction /InterchangeAssets/Functions/MX_Xor
graph {nodes [
  n1 = FunctionInput() {bUsePreviewValueAsDefault=true, inputName=A, inputType=FunctionInput_Scalar}
  n2 = MaterialFunctionCall!(Input=@n1) {materialFunction=&/MX_Not}
  n3 = FunctionInput() {bUsePreviewValueAsDefault=true, inputName=B, inputType=FunctionInput_Scalar, sortPriority=1}
  n4 = MaterialFunctionCall(Input=@n3) {materialFunction=&MX_Not}
  FunctionOutput(None=MaterialFunctionCall!(
    in1=MaterialFunctionCall!(in1=@n2.Result, in2=@n3) {materialFunction=&/MX_And}.Result
    in2=MaterialFunctionCall(in1=@n1, in2=@n4.Result) {materialFunction=&MX_And}.Result
   ) {materialFunction=&/MX_Or}.Result)
 ]}
