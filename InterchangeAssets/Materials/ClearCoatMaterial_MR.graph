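# Clear coat parent material in InterchangeAssets, exposing metallic-roughness PBR inputs as parameters, with skeletal mesh and mesh deformer usage enabled.
# PBRSurfaceFunction_MR builds the base surface from value, map and map-weight parameters, then its result is broken into individual attributes.
# MapOrScalarFunction resolves clear coat amount and roughness from a map or scalar.
# MapOrLinearColorFunction feeds the separate clear coat normal output, using a flat +Z constant as the non-map value.
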Material /InterchangeAssets/Materials/ClearCoatMaterial_MR
graph {
 properties {
  bUsedWithMeshDeformer=true
  bUsedWithSkeletalMesh=true
  pixelDepthOffsetMode=PDOM_Legacy
  shadingModel=MSM_ClearCoat
  usedShadingModels=Clear Coat
 }
 nodes [
  BreakMaterialAttributes!(Attr=MaterialFunctionCall!(
    BaseColor=VectorParameter() {parameterName=BaseColor}.RGB
    BaseColorMap=TextureObjectParameter() {parameterName=BaseColorMap}
    BaseColorMapWeight=ScalarParameter() {parameterName=BaseColorMapWeight}
    Metallic=ScalarParameter() {parameterName=Metallic}
    MetallicMap=TextureObjectParameter() {parameterName=MetallicMap, samplerType=SAMPLERTYPE_LinearColor, texture=/Engine/EngineMaterials/BaseFlattenLinearColor}
    MetallicMapWeight=ScalarParameter() {parameterName=MetallicMapWeight}
    Specular=ScalarParameter() {defaultValue=0.5, parameterName=Specular}
    SpecularMap=TextureObjectParameter() {parameterName=SpecularMap, samplerType=SAMPLERTYPE_LinearColor, texture=&BaseFlattenLinearColor}
    SpecularMapWeight=ScalarParameter() {parameterName=SpecularMapWeight}
    Roughness=ScalarParameter() {defaultValue=0.5, parameterName=Roughness}
    RoughnessMap=TextureObjectParameter() {parameterName=RoughnessMap, samplerType=SAMPLERTYPE_LinearColor, texture=&BaseFlattenLinearColor}
    RoughnessMapWeight=ScalarParameter() {parameterName=RoughnessMapWeight}
    EmissiveColor=VectorParameter() {parameterName=EmissiveColor}.RGB
    EmissiveColorMap=TextureObjectParameter() {parameterName=EmissiveColorMap}
    EmissiveColorMapWeight=ScalarParameter() {parameterName=EmissiveColorMapWeight}
    NormalMap=TextureObjectParameter() {parameterName=NormalMap, samplerType=SAMPLERTYPE_Normal, texture=/Engine/EngineMaterials/DefaultNormal}
    NormalMapWeight=ScalarParameter() {parameterName=NormalMapWeight}
    AmbientOcclusionMap=TextureObjectParameter() {parameterName=AmbientOcclusionMap, samplerType=SAMPLERTYPE_LinearColor, texture=&BaseFlattenLinearColor}
    AmbientOcclusionMapWeight=ScalarParameter() {parameterName=AmbientOcclusionMapWeight}
   ) {materialFunction=&/PBRSurfaceFunction_MR}.Result)
  ClearCoatNormalCustomOutput!(None=MaterialFunctionCall!(
    Map=TextureObjectParameter() {parameterName=ClearCoatNormalMap, samplerType=SAMPLERTYPE_Normal, texture=&DefaultNormal}
    MapWeight=ScalarParameter() {parameterName=ClearCoatNormalMapWeight}
    Color=Constant3Vector() {constant=<rgba 0 0 1 0>}
   ) {materialFunction=&/MapOrLinearColorFunction}.Result)
  MaterialFunctionCall!(
   Map=TextureObjectParameter() {parameterName=ClearCoatMap, samplerType=SAMPLERTYPE_LinearColor, texture=&BaseFlattenLinearColor}
   MapWeight=ScalarParameter() {parameterName=ClearCoatMapWeight}
   Scalar=ScalarParameter() {parameterName=ClearCoat}
  ) {materialFunction=&/MapOrScalarFunction}
  n1 = TextureObjectParameter() {parameterName=ClearCoatRoughnessMap, samplerType=SAMPLERTYPE_LinearColor, texture=&BaseFlattenLinearColor}
  MaterialFunctionCall(
   Map=@n1
   MapWeight=ScalarParameter() {parameterName=ClearCoatRoughnessMapWeight}
   Scalar=ScalarParameter() {parameterName=ClearCoatRoughness}
  ) {materialFunction=&MapOrScalarFunction}
 ]
}
