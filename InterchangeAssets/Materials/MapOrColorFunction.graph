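# Material function from the Interchange assets that yields a value blended between a color and a sampled texture map.
# The Map texture input is sampled and its RGB feeds a linear interpolation against the Color input.
# The MapWeight scalar drives the blend, so zero gives the plain color and one gives the map.
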MaterialFunction /InterchangeAssets/Materials/MapOrColorFunction
graph {nodes [
  n1 = FunctionInput() {inputName=Map, inputType=FunctionInput_Texture2D}
  FunctionOutput(None=LinearInterpolate(
    A=FunctionInput() {inputName=Color}
    B=TextureSample(Tex=@n1).RGB
    Alpha=FunctionInput() {inputName=MapWeight, inputType=FunctionInput_Scalar}
   ))
 ]}
