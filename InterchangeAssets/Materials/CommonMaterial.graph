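# CommonMaterial is a material function in InterchangeAssets that builds the shared base color, emissive, normal and ambient occlusion channels as one Material Attributes output.
# MapOrColor, MapOrLinearColor and MapOrScalar helper functions combine each map and weight with a fallback color or scalar.
# The static switch bHasBump treats NormalMap as a height map through NormalFromHeightmap, reusing its weight as intensity, or as a normal map with a flat fallback.
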MaterialFunction /InterchangeAssets/Materials/CommonMaterial
graph {nodes [
  n1 = FunctionInput(Preview=TextureObject()) {inputName=BaseColorMap, inputType=FunctionInput_Texture2D, sortPriority=1}
  n2 = FunctionInput(Preview=Constant()) {inputName=BaseColorMapWeight, inputType=FunctionInput_Scalar, sortPriority=2}
  n3 = FunctionInput(Preview=Constant3Vector()) {inputName=BaseColor}
  n4 = FunctionInput(Preview=TextureObject()) {inputName=EmissiveColorMap, inputType=FunctionInput_Texture2D, sortPriority=4}
  n5 = FunctionInput(Preview=Constant()) {inputName=EmissiveColorMapWeight, inputType=FunctionInput_Scalar, sortPriority=5}
  n6 = FunctionInput() {inputName=EmissiveColor, sortPriority=3}
  n7 = FunctionInput(Preview=TextureObject() {samplerType=SAMPLERTYPE_Normal, texture=/Engine/EngineMaterials/DefaultNormal}) {inputName=NormalMap, inputType=FunctionInput_Texture2D, sortPriority=6}
  n8 = FunctionInput(Preview=Constant()) {inputName=NormalMapWeight, inputType=FunctionInput_Scalar, sortPriority=7}
  n9 = MaterialFunctionCall!(Height Map=@n7, Normal Map Intensity=@n8, Height Map UV Offset, Coordinates, Height Map Channel Selector) {materialFunction=/Engine/Functions/Engine_MaterialFunctions03/Procedurals/NormalFromHeightmap}
  n10 = MaterialFunctionCall!(Map=@n7, MapWeight=@n8, Color=Constant3Vector() {constant=<rgba 0 0 1 0>}) {materialFunction=&/MapOrLinearColorFunction}
  n11 = FunctionInput(Preview=TextureObject() {samplerType=SAMPLERTYPE_LinearColor, texture=/Engine/EngineMaterials/BaseFlattenLinearColor}) {inputName=AmbientOcclusionMap, inputType=FunctionInput_Texture2D, sortPriority=8}
  n12 = FunctionInput(Preview=Constant()) {inputName=AmbientOcclusionMapWeight, inputType=FunctionInput_Scalar, sortPriority=9}
  n13 = Constant()
  FunctionOutput(None=MakeMaterialAttributes!(
    BaseColor=MaterialFunctionCall!(Map=@n1, MapWeight=@n2, Color=@n3) {materialFunction=&/MapOrColorFunction}.Result
    Metallic
    Specular
    Roughness
    Anisotropy
    EmissiveColor=MaterialFunctionCall(Map=@n4, MapWeight=@n5, Color=@n6) {materialFunction=&MapOrColorFunction}.Result
    Opacity
    OpacityMask
    Normal=StaticSwitchParameter(True=@n9.Result, False=@n10.Result) {parameterName=bHasBump}
    Tangent
    WorldPositionOffset
    SubsurfaceColor
    ClearCoat
    ClearCoatRoughness
    AmbientOcclusion=MaterialFunctionCall!(Map=@n11, MapWeight=@n12, Scalar=@n13) {materialFunction=&/MapOrScalarFunction}.Result
    Refraction
    CustomizedUVs_0
    CustomizedUVs_1
    CustomizedUVs_2
    CustomizedUVs_3
    CustomizedUVs_4
    CustomizedUVs_5
    CustomizedUVs_6
    CustomizedUVs_7
    PixelDepthOffset
    ShadingModel
    Displacement
   ))
 ]}
