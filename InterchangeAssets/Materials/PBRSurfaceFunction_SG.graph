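# Material function in InterchangeAssets that builds PBR material attributes from specular-glossiness inputs, each map paired with a weight.
# Diffuse and specular map-or-value blends pass through ConvertFromDiffSpec to drive base color, metallic and specular.
# Glossiness blends with its map, whose weight is multiplied by a large constant, then inverts into roughness.
# Emissive, normal and ambient occlusion come from a shared CommonMaterial function, broken apart and merged into the output.
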MaterialFunction /InterchangeAssets/Materials/PBRSurfaceFunction_SG
graph {nodes [
  n1 = TextureObject()
  n2 = Constant()
  n3 = Constant3Vector()
  n4 = TextureObject() {samplerType=SAMPLERTYPE_LinearColor, texture=/Engine/EngineMaterials/BaseFlattenLinearColor}
  n5 = Constant()
  n6 = Constant() {r=0.1}
  n7 = MaterialFunctionCall(
   DiffuseColor=MaterialFunctionCall!(
    Map=FunctionInput(Preview=@n1) {inputName=DiffuseColorMap, inputType=FunctionInput_Texture2D, sortPriority=4}
    MapWeight=FunctionInput(Preview=@n2) {inputName=DiffuseColorMapWeight, inputType=FunctionInput_Scalar, sortPriority=5}
    Scalar=FunctionInput(Preview=@n3) {inputName=DiffuseColor, sortPriority=3}
   ) {materialFunction=&/MapOrScalarFunction}.Result
   SpecularColor=MaterialFunctionCall!(
    Map=FunctionInput(Preview=@n4) {inputName=SpecularMap, inputType=FunctionInput_Texture2D, sortPriority=7}
    MapWeight=FunctionInput(Preview=@n5) {inputName=SpecularMapWeight, inputType=FunctionInput_Scalar, sortPriority=8}
    Color=FunctionInput(Preview=@n6) {inputName=SpecularColor, sortPriority=6, previewValue=<0.5 0.5 0.5 1>}
   ) {materialFunction=&/MapOrColorFunction}.Result
  ) {materialFunction=&ConvertFromDiffSpec}
  n8 = Constant()
  n9 = FunctionInput(Preview=Constant() {r=0.1}) {inputName=Glossiness, inputType=FunctionInput_Scalar, sortPriority=9, previewValue=<0.5 0.5 0.5 1>}
  n10 = MaterialFunctionCall(
   Map=FunctionInput(Preview=TextureObject() {samplerType=SAMPLERTYPE_Grayscale, texture=&BaseFlattenLinearColor}) {inputName=GlossinessMap, inputType=FunctionInput_Texture2D, sortPriority=10}
   MapWeight=Multiply(A=FunctionInput(Preview=@n8) {inputName=GlossinessMapWeight, inputType=FunctionInput_Scalar, sortPriority=11}) {constB=1000}
   Scalar=@n9
  ) {materialFunction=&MapOrScalarFunction}
  n11 = TextureObject()
  n12 = Constant()
  n13 = TextureObject() {samplerType=SAMPLERTYPE_Normal, texture=/Engine/EngineMaterials/DefaultNormal}
  n14 = Constant()
  n15 = TextureObject() {samplerType=SAMPLERTYPE_LinearColor, texture=&BaseFlattenLinearColor}
  n16 = Constant()
  n17 = BreakMaterialAttributes!(Attr=MaterialFunctionCall!(
    BaseColor
    BaseColorMap
    BaseColorMapWeight
    EmissiveColor=FunctionInput() {inputName=EmissiveColor, sortPriority=12}
    EmissiveColorMap=FunctionInput(Preview=@n11) {inputName=EmissiveColorMap, inputType=FunctionInput_Texture2D, sortPriority=13}
    EmissiveColorMapWeight=FunctionInput(Preview=@n12) {inputName=EmissiveColorMapWeight, inputType=FunctionInput_Scalar, sortPriority=14}
    NormalMap=FunctionInput(Preview=@n13) {inputName=NormalMap, inputType=FunctionInput_Texture2D, sortPriority=15}
    NormalMapWeight=FunctionInput(Preview=@n14) {inputName=NormalMapWeight, inputType=FunctionInput_Scalar, sortPriority=16}
    AmbientOcclusionMap=FunctionInput(Preview=@n15) {inputName=AmbientOcclusionMap, inputType=FunctionInput_Texture2D, sortPriority=17}
    AmbientOcclusionMapWeight=FunctionInput(Preview=@n16) {inputName=AmbientOcclusionMapWeight, inputType=FunctionInput_Scalar, sortPriority=18}
   ) {materialFunction=&/CommonMaterial}.Result)
  FunctionOutput(None=MakeMaterialAttributes!(
    BaseColor=@n7.BaseColor
    Metallic=@n7.BaseColor
    Specular=@n7.BaseColor
    Roughness=OneMinus(None=@n10.Result)
    Anisotropy
    EmissiveColor=@n17.EmissiveColor
    Opacity
    OpacityMask
    Normal=@n17.EmissiveColor
    Tangent
    WorldPositionOffset
    SubsurfaceColor
    ClearCoat
    ClearCoatRoughness
    AmbientOcclusion=@n17.EmissiveColor
    Refraction
    CustomizedUVs_0
    CustomizedUVs_1
    CustomizedUVs_2
    CustomizedUVs_3
    CustomizedUVs_4
    CustomizedUVs_5
    CustomizedUVs_6
    CustomizedUVs_7
    PixelDepthOffset
    ShadingModel
    Displacement
   ))
 ]}
