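# Interchange parent material for clear coat surfaces in the specular-glossiness workflow, enabled for skeletal meshes and mesh deformers.
# PBRSurfaceFunction_SG turns diffuse, specular, glossiness, emissive, normal and occlusion parameters into material attributes, which are then broken out.
# Clear coat amount and roughness each use MapOrScalarFunction to combine a map, weight and scalar.
# A separate coat normal goes through MapOrLinearColorFunction, defaulting to flat, into ClearCoatNormalCustomOutput.
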Material /InterchangeAssets/Materials/ClearCoatMaterial_SG
graph {
 properties {
  bUsedWithMeshDeformer=true
  bUsedWithSkeletalMesh=true
  pixelDepthOffsetMode=PDOM_Legacy
  shadingModel=MSM_ClearCoat
  usedShadingModels=Clear Coat
 }
 nodes [
  BreakMaterialAttributes!(Attr=MaterialFunctionCall!(
    DiffuseColor=VectorParameter() {parameterName=DiffuseColor, defaultValue=<rgba 0 0 0 1>}.RGB
    DiffuseColorMap=TextureObjectParameter() {parameterName=DiffuseColorMap}
    DiffuseColorMapWeight=ScalarParameter() {parameterName=DiffuseColorMapWeight}
    SpecularColor=VectorParameter() {parameterName=SpecularColor, defaultValue=<rgba 0.03 0.03 0.03 1>}.RGB
    SpecularMap=TextureObjectParameter() {parameterName=SpecularColorMap}
    SpecularMapWeight=ScalarParameter() {parameterName=SpecularColorMapWeight}
    Glossiness=ScalarParameter() {defaultValue=0.5, parameterName=Glossiness}
    GlossinessMap=TextureObjectParameter() {parameterName=GlossinessMap, samplerType=SAMPLERTYPE_LinearColor, texture=/Engine/EngineMaterials/BaseFlattenLinearColor}
    GlossinessMapWeight=ScalarParameter() {parameterName=GlossinessMapWeight}
    EmissiveColor=VectorParameter() {parameterName=EmissiveColor}.RGB
    EmissiveColorMap=TextureObjectParameter() {parameterName=EmissiveColorMap}
    EmissiveColorMapWeight=ScalarParameter() {parameterName=EmissiveColorMapWeight}
    NormalMap=TextureObjectParameter() {parameterName=NormalMap, samplerType=SAMPLERTYPE_Normal, texture=/Engine/EngineMaterials/DefaultNormal}
    NormalMapWeight=ScalarParameter() {parameterName=NormalMapWeight}
    AmbientOcclusionMap=TextureObjectParameter() {parameterName=AmbientOcclusionMap, samplerType=SAMPLERTYPE_LinearColor, texture=&BaseFlattenLinearColor}
    AmbientOcclusionMapWeight=ScalarParameter() {parameterName=AmbientOcclusionMapWeight}
   ) {materialFunction=&/PBRSurfaceFunction_SG}.Result)
  ClearCoatNormalCustomOutput!(None=MaterialFunctionCall!(
    Map=TextureObjectParameter() {parameterName=ClearCoatNormalMap, samplerType=SAMPLERTYPE_Normal, texture=&DefaultNormal}
    MapWeight=ScalarParameter() {parameterName=ClearCoatNormalMapWeight}
    Color=Constant3Vector() {constant=<rgba 0 0 1 0>}
   ) {materialFunction=&/MapOrLinearColorFunction}.Result)
  MaterialFunctionCall!(
   Map=TextureObjectParameter() {parameterName=ClearCoatMap, samplerType=SAMPLERTYPE_LinearColor, texture=&BaseFlattenLinearColor}
   MapWeight=ScalarParameter() {parameterName=ClearCoatMapWeight}
   Scalar=ScalarParameter() {parameterName=ClearCoat}
  ) {materialFunction=&/MapOrScalarFunction}
  n1 = TextureObjectParameter() {parameterName=ClearCoatRoughnessMap, samplerType=SAMPLERTYPE_LinearColor, texture=&BaseFlattenLinearColor}
  MaterialFunctionCall(
   Map=@n1
   MapWeight=ScalarParameter() {parameterName=ClearCoatRoughnessMapWeight}
   Scalar=ScalarParameter() {parameterName=ClearCoatRoughness}
  ) {materialFunction=&MapOrScalarFunction}
 ]
}
