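# DecalMaterial is a translucent deferred decal material in InterchangeAssets, exposing DecalTexture and NormalTexture parameters with engine default textures.
# One Multiply squares the DecalTexture color by multiplying its RGB with itself.
# Another scales the NormalTexture RGB by the DecalTexture alpha.
# Two loose constants accompany these nodes, and no links to material outputs are shown.
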Material /InterchangeAssets/Materials/DecalMaterial
graph {
 properties {
  bUsedWithMeshDeformer=true
  bUsedWithSkeletalMesh=true
  bUsedWithSplineMeshes=true
  bUsedWithStaticLighting=true
  blendMode=BLEND_Translucent
  materialDomain=MD_DeferredDecal
  opacityMaskClipValue=0.33
  pixelDepthOffsetMode=PDOM_Legacy
  shadingModel=MSM_Unlit
  usedShadingModels=Unlit
 }
 nodes [
  Constant()
  Constant() {r=0.5}
  n1 = TextureSampleParameter2D() {parameterName=DecalTexture, texture=/Engine/EngineMaterials/DefaultDiffuse}
  Multiply(A=@n1.RGB, B=@n1.RGB)
  Multiply(
   A=TextureSampleParameter2D() {parameterName=NormalTexture, samplerType=SAMPLERTYPE_Normal, texture=/Engine/EngineMaterials/DefaultNormal}.RGB
   B=@n1.A
  )
 ]
}
