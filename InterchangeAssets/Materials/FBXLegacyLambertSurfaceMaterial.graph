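# Parameterized Default Lit surface material in Interchange assets, named for legacy FBX Lambert surfaces and usable on skeletal, Nanite and instanced meshes.
# Diffuse, emissive, normal, ambient occlusion, opacity and opacity mask each lerp from a base value to a UV-tiled texture via a weight parameter.
# A shared OpacitySourceChannel parameter is described as choosing alpha or red for both opacity maps, though both inputs of those lerps show RGB.
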Material /InterchangeAssets/Materials/FBXLegacyLambertSurfaceMaterial
graph {
 properties {
  bUsedWithInstancedStaticMeshes=true
  bUsedWithMeshDeformer=true
  bUsedWithNanite=true
  bUsedWithSkeletalMesh=true
  pixelDepthOffsetMode=PDOM_Legacy
  usedShadingModels=Default Lit
 }
 nodes [
  n1 = TextureCoordinate()
  n2 = AppendVector(
   A=ScalarParameter() {defaultValue=1, parameterName=DiffuseColorMapTilingU}
   B=ScalarParameter() {defaultValue=1, parameterName=DiffuseColorMapTilingV}
  )
  LinearInterpolate(
   A=VectorParameter() {parameterName=DiffuseColor, defaultValue=<rgba 1 1 1 0>}.RGB
   B=TextureSampleParameter2D(UVs=Multiply(A=@n1, B=@n2)) {parameterName=DiffuseColorMap}.RGB
   Alpha=ScalarParameter() {parameterName=DiffuseColorMapWeight}
  )
  n3 = TextureCoordinate()
  n4 = AppendVector(
   A=ScalarParameter() {defaultValue=1, parameterName=EmissiveColorMapTilingU}
   B=ScalarParameter() {defaultValue=1, parameterName=EmissiveColorMapTilingV}
  )
  LinearInterpolate(
   A=VectorParameter() {parameterName=EmissiveColor}.RGB
   B=TextureSampleParameter2D(UVs=Multiply(A=@n3, B=@n4)) {parameterName=EmissiveColorMap}.RGB
   Alpha=ScalarParameter() {parameterName=EmissiveColorMapWeight}
  )
  n5 = TextureCoordinate()
  n6 = AppendVector(A=ScalarParameter() {defaultValue=1, parameterName=NormalMapTilingU}, B=ScalarParameter() {defaultValue=1, parameterName=NormalMapTilingV})
  LinearInterpolate(
   A=Constant3Vector() {constant=<rgba 0 0 1 0>}
   B=TextureSampleParameter2D(UVs=Multiply(A=@n5, B=@n6)) {parameterName=NormalMap, samplerType=SAMPLERTYPE_Normal, texture=/Engine/EngineMaterials/DefaultNormal}.RGB
   Alpha=ScalarParameter() {parameterName=NormalMapWeight}
  )
  n7 = TextureCoordinate()
  n8 = AppendVector(
   A=ScalarParameter() {defaultValue=1, parameterName=AmbientOcclusionMapTilingU}
   B=ScalarParameter() {defaultValue=1, parameterName=AmbientOcclusionMapTilingV}
  )
  LinearInterpolate(
   A=ScalarParameter() {defaultValue=1, parameterName=AmbientOcclusion}
   B=TextureSampleParameter2D(UVs=Multiply(A=@n7, B=@n8)) {parameterName=AmbientOcclusionMap, samplerType=SAMPLERTYPE_LinearColor, texture=/Engine/EngineMaterials/BaseFlattenLinearColor}.RGB
   Alpha=ScalarParameter() {parameterName=AmbientOcclusionMapWeight}
  )
  n9 = ScalarParameter() {defaultValue=1, parameterName=OpacityMapTilingU}
  n10 = ScalarParameter() {defaultValue=1, parameterName=OpacityMapTilingV}
  n11 = TextureSampleParameter2D(UVs=Multiply(A=TextureCoordinate(), B=AppendVector(A=@n9, B=@n10))) {parameterName=OpacityMap, samplerType=SAMPLERTYPE_LinearColor, texture=&BaseFlattenLinearColor}
  n12 = ScalarParameter() {desc="Set to 1 to have Alpha be the source of opacity/mask, 0 to use Red channel of the texture", parameterName=OpacitySourceChannel}
  LinearInterpolate(
   A=ScalarParameter() {defaultValue=1, parameterName=Opacity}
   B=LinearInterpolate(A=@n11.RGB, B=@n11.RGB, Alpha=@n12)
   Alpha=ScalarParameter() {parameterName=OpacityMapWeight}
  )
  n13 = ScalarParameter() {defaultValue=1, parameterName=OpacityMaskMapTilingU}
  n14 = ScalarParameter() {defaultValue=1, parameterName=OpacityMaskMapTilingV}
  n15 = TextureSampleParameter2D(UVs=Multiply(A=TextureCoordinate(), B=AppendVector(A=@n13, B=@n14))) {parameterName=OpacityMaskMap, samplerType=SAMPLERTYPE_LinearColor, texture=&BaseFlattenLinearColor}
  LinearInterpolate(A=Constant(), B=LinearInterpolate(A=@n15.RGB, B=@n15.RGB, Alpha=@n12), Alpha=ScalarParameter() {parameterName=OpacityMaskMapWeight})
 ]
}
