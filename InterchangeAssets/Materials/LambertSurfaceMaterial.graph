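# Lambert surface material in InterchangeAssets, a parameterized Default Lit base material usable with skeletal meshes, Nanite and mesh deformers.
# It exposes diffuse color, emissive color, normal and ambient occlusion maps, each with a weight, plus color parameters for diffuse and emissive.
# All parameters feed the shared CommonMaterial function, whose result passes through a BreakMaterialAttributes node.
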Material /InterchangeAssets/Materials/LambertSurfaceMaterial
graph {
 properties {
  bUsedWithMeshDeformer=true
  bUsedWithNanite=true
  bUsedWithSkeletalMesh=true
  pixelDepthOffsetMode=PDOM_Legacy
  usedShadingModels=Default Lit
 }
 nodes [BreakMaterialAttributes!(Attr=MaterialFunctionCall!(
    BaseColor=VectorParameter() {parameterName=DiffuseColor}.RGB
    BaseColorMap=TextureObjectParameter() {parameterName=DiffuseColorMap}
    BaseColorMapWeight=ScalarParameter() {parameterName=DiffuseColorMapWeight}
    EmissiveColor=VectorParameter() {parameterName=EmissiveColor}.RGB
    EmissiveColorMap=TextureObjectParameter() {parameterName=EmissiveColorMap}
    EmissiveColorMapWeight=ScalarParameter() {parameterName=EmissiveColorMapWeight}
    NormalMap=TextureObjectParameter() {parameterName=NormalMap, samplerType=SAMPLERTYPE_Normal, texture=/Engine/EngineMaterials/DefaultNormal}
    NormalMapWeight=ScalarParameter() {parameterName=NormalMapWeight}
    AmbientOcclusionMap=TextureObjectParameter() {parameterName=AmbientOcclusionMap, samplerType=SAMPLERTYPE_LinearColor, texture=/Engine/EngineMaterials/BaseFlattenLinearColor}
    AmbientOcclusionMapWeight=ScalarParameter() {parameterName=AmbientOcclusionMapWeight}
   ) {materialFunction=&/CommonMaterial}.Result)]
}
