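# Lambert surface material in InterchangeAssets, a Default Lit base exposing diffuse, emissive, normal and ambient occlusion colors or maps as parameters.
# Each map has OffsetScale and Rotation parameters feeding its own glTF MF_TransformUVs call.
# CommonMaterialWithUVTransform combines these UVs with colors, textures and map weights, then its result is broken into material attributes.
# Usage flags cover skeletal meshes, Nanite and mesh deformers.
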Material /InterchangeAssets/Materials/LambertSurfaceMaterialWithUVTransform
graph {
 properties {
  bUsedWithMeshDeformer=true
  bUsedWithNanite=true
  bUsedWithSkeletalMesh=true
  pixelDepthOffsetMode=PDOM_Legacy
  usedShadingModels=Default Lit
 }
 nodes [
  n1 = VectorParameter() {parameterName=EmissiveColorMap_OffsetScale, defaultValue=<rgba 0 0 1 1>}
  n2 = ScalarParameter() {parameterName=EmissiveColorMap_Rotation}
  n3 = VectorParameter() {parameterName=DiffuseColorMap_OffsetScale, defaultValue=<rgba 0 0 1 1>}
  n4 = ScalarParameter() {parameterName=DiffuseColorMap_Rotation}
  n5 = VectorParameter() {parameterName=AmbientOcclusionMap_OffsetScale, defaultValue=<rgba 0 0 1 1>}
  n6 = ScalarParameter() {parameterName=AmbientOcclusionMap_Rotation}
  n7 = VectorParameter() {parameterName=NormalMap_OffsetScale, defaultValue=<rgba 0 0 1 1>}
  n8 = ScalarParameter() {parameterName=NormalMap_Rotation}
  n9 = TextureObjectParameter() {parameterName=DiffuseColorMap}
  BreakMaterialAttributes!(Attr=MaterialFunctionCall!(
    EmissiveColor UVs=MaterialFunctionCall!(Offset_X=@n1.R, Offset_Y=@n1.R, Scale_X=@n1.R, Scale_Y=@n1.R, Rotation=@n2, TexCoord, TilingMethod) {materialFunction=/InterchangeAssets/gltf/MaterialFunctions/MF_TransformUVs}.UVs
    BaseColor UVs=MaterialFunctionCall(Offset_X=@n3.R, Offset_Y=@n3.R, Scale_X=@n3.R, Scale_Y=@n3.R, Rotation=@n4) {materialFunction=&MF_TransformUVs}.UVs
    AmbientOcclusion UVs=MaterialFunctionCall(Offset_X=@n5.R, Offset_Y=@n5.R, Scale_X=@n5.R, Scale_Y=@n5.R, Rotation=@n6) {materialFunction=&MF_TransformUVs}.UVs
    Normal UVs=MaterialFunctionCall(Offset_X=@n7.R, Offset_Y=@n7.R, Scale_X=@n7.R, Scale_Y=@n7.R, Rotation=@n8) {materialFunction=&MF_TransformUVs}.UVs
    BaseColor=VectorParameter() {parameterName=DiffuseColor}.RGB
    BaseColorMap=@n9
    BaseColorMapWeight=ScalarParameter() {parameterName=DiffuseColorMapWeight}
    EmissiveColor=VectorParameter() {parameterName=EmissiveColor}.RGB
    EmissiveColorMap=TextureObjectParameter() {parameterName=EmissiveColorMap}
    EmissiveColorMapWeight=ScalarParameter() {parameterName=EmissiveColorMapWeight}
    NormalMap=TextureObjectParameter() {parameterName=NormalMap, samplerType=SAMPLERTYPE_Normal, texture=/Engine/EngineMaterials/DefaultNormal}
    NormalMapWeight=ScalarParameter() {parameterName=NormalMapWeight}
    AmbientOcclusionMap=TextureObjectParameter() {parameterName=AmbientOcclusionMap, samplerType=SAMPLERTYPE_LinearColor, texture=/Engine/EngineMaterials/BaseFlattenLinearColor}
    AmbientOcclusionMapWeight=ScalarParameter() {parameterName=AmbientOcclusionMapWeight}
   ) {materialFunction=&/CommonMaterialWithUVTransform}.Result)
  TextureSample(Tex=@n9)
 ]
}
