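# Material function that computes a view-angle-dependent iridescence color, apparently thin-film interference, from the New_LUT lookup texture.
# The absolute dot of pixel normal and reflection vector drives a remap of Thickness, with shift scaled by n2, into the LUT's U coordinate.
# Inputs n2, n3 and k select regions of the texture, which acts as a packed atlas.
# Two samples at neighboring k slices are linearly blended by k's fractional position to fake continuous interpolation.
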MaterialFunction /InterchangeAssets/Functions/MF_Iridescence
graph {
 properties {bExposeToLibrary=true}
 nodes [
  n1 = PixelNormalWS()
  n2 = ReflectionVectorWS()
  n3 = OneMinus(None=Abs(None=DotProduct(A=@n1, B=@n2)))
  n4 = Constant() {r=0.5}
  n5 = Subtract(A=Clamp(None=FunctionInput() {inputName=n3, inputType=FunctionInput_Scalar, sortPriority=1, previewValue=<1.6 0 0 1>}) {maxDefault=5, minDefault=1})
  n6 = Saturate(None=Fmod(A=Multiply(A=@n4, B=@n5), B=Constant() {r=1}))
  n7 = Constant() {r=4}
  n8 = Multiply(A=Truncate!(None=Multiply(A=@n6, B=@n7)) {is=class MaterialExpressionTruncate}, B=Constant() {r=0.25})
  n9 = Saturate(None=Add(A=FunctionInput() {inputName=Thickness, inputType=FunctionInput_Scalar, previewValue=<1.3 0 0 1>}, B=@n8))
  n10 = FunctionInput() {inputName=n2, inputType=FunctionInput_Scalar, sortPriority=1, previewValue=<1.5 0 0 1>}
  n11 = Multiply(A=Subtract(A=Clamp(None=@n10) {maxDefault=4.99, minDefault=1.01}), B=Constant() {r=0.25})
  n12 = LinearInterpolate(B=Multiply(A=@n3, B=Constant() {r=0.25}), Alpha=@n11)
  n13 = MaterialFunctionCall(Input=@n3, Input Low=Constant() {r=1}, Input High=Constant(), Target Low=Min(A=@n9, B=Add(A=@n8, B=@n12)), Target High=@n9) {materialFunction=&RemapValueRange}
  n14 = Multiply(A=@n11) {constB=0.05}
  n15 = Constant() {r=0.501}
  n16 = If(A=Multiply(A=@n5, B=Constant() {r=0.25}), B=Constant() {r=0.5}, "A > B"=@n15, "A == B"=@n15, "A < B"=Constant() {r=0.001})
  n17 = Multiply(A=Clamp(None=FunctionInput() {inputName=k, inputType=FunctionInput_Scalar, sortPriority=2}) {maxDefault=5}) {constB=0.2}
  n18 = Multiply(A=LinearInterpolate(Alpha=@n17) {constB=0.4999}) {constB=10}
  n19 = Add(A=@n16, B=Divide(A=Floor(None=@n18)) {constB=10})
  n20 = AppendVector(A=@n13.Result, B=Add(A=@n14, B=Saturate(None=@n19)))
  n21 = TextureObject() {samplerType=SAMPLERTYPE_LinearColor, texture=/InterchangeAssets/Utilities/New_LUT}
  n22 = Multiply(A=LinearInterpolate(Alpha=Add(A=@n17) {constB=0.1}) {constB=0.4999}) {constB=10}
  n23 = Add(A=@n16, B=Divide(A=Floor(None=@n22)) {constB=10})
  n24 = AppendVector(A=@n13.Result, B=Add(A=@n14, B=Saturate(None=@n23)))
  n25 = Multiply(A=@n17) {constB=10}
  FunctionOutput(None=LinearInterpolate(
    A=TextureSample(UVs=@n20, Tex=@n21) {samplerType=SAMPLERTYPE_LinearColor}.RGB
    B=TextureSample(UVs=@n24, Tex=@n21) {samplerType=SAMPLERTYPE_LinearColor}.RGB
    Alpha=Frac(None=@n25)
   )) {desc=https://half4.xyz/wp-content/uploads/2020/10/image-2.png}
 ]
}
