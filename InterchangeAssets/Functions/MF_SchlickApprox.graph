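# MF_SchlickApprox is a library-exposed material function that converts an index of refraction into a normalized reflectance value.
# It computes Schlick's normal-incidence reflectance F0 as the square of (1 - IOR) / (1 + IOR), with a glass-like default IOR.
# An inverse lerp remaps F0 from the 0 to 0.08 range and saturates it, matching Unreal's Specular input scale.
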MaterialFunction /InterchangeAssets/Functions/MF_SchlickApprox
graph {
 properties {bExposeToLibrary=true}
 nodes [
  n1 = Constant()
  n2 = Constant() {r=0.08}
  n3 = FunctionInput() {bUsePreviewValueAsDefault=true, inputName=IOR, inputType=FunctionInput_Scalar, previewValue=<1.5 0 0 1>}
  n4 = Divide(A=OneMinus(None=@n3), B=Add(A=@n3, B=Constant() {r=1}))
  n5 = Multiply(A=@n4, B=@n4)
  FunctionOutput(None=Saturate(None=InverseLinearInterpolate(A=@n1, B=@n2, Value=@n5)))
 ]
}
