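# Layered PBR surface material in InterchangeAssets, exposing base, specular, coat, sheen, subsurface, emission and anisotropy parameters.
# Weights select the shading model at the pixel: Cloth overrides Clear Coat, then Subsurface, then Default Lit.
# Base color blends an Oren-Nayar boost by Metallic, using camera position as a hacky light vector, then multiplies by thin-film iridescence.
# Schlick-from-IOR scales specular, coat supplies a second normal, and anisotropy rotates the tangent.
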Material /InterchangeAssets/Materials/OrenNayerMaterial
graph {
 properties {
  bUsedWithMeshDeformer=true
  bUsedWithSkeletalMesh=true
  pixelDepthOffsetMode=PDOM_Legacy
  shadingModel=MSM_FromMaterialExpression
  usedShadingModels=Default Lit | Subsurface | Clear Coat | Cloth
 }
 nodes [
  n1 = ScalarParameter() {parameterName=CoatWeight}
  n2 = Constant()
  n3 = Constant3Vector() {constant=<rgba 0 0 1 0>}
  n4 = StaticSwitchParameter(
   True=TextureSampleParameter2D() {parameterName=NormalMap, samplerType=SAMPLERTYPE_Normal, texture=/Engine/EngineMaterials/FlatNormal}.RGB
   False=@n3
  ) {parameterName=UseNormalMap}
  ClearCoatNormalCustomOutput!(None=If(A=@n1, B=@n2, "A > B"=@n4, "A == B"=@n3, "A < B"=@n3))
  n5 = ScalarParameter() {parameterName=AnisotropyRotation}
  n6 = Cosine(None=@n5)
  n7 = VectorParameter() {parameterName=Tangent, defaultValue=<rgba 1 0 0 1>}
  n8 = TextureSampleParameter2D() {parameterName=TangentTexture_WS, samplerType=SAMPLERTYPE_Normal, texture=&FlatNormal}
  n9 = If(A=Max(A=Max(A=@n7.R, B=@n7.R), B=@n7.B), B=Constant(), "A > B"=@n7.RGB, "A == B"=@n8.RGB, "A < B"=@n8.RGB)
  n10 = Sine(None=@n5)
  n11 = CrossProduct(A=Constant3Vector() {constant=<rgba 0 0 1 0>}, B=@n9)
  n12 = Constant3Vector() {constant=<rgba 1 0 0 0>}
  If(
   A=ScalarParameter() {parameterName=SpecularAnisotropy}
   B=Constant()
   "A > B"=Add(A=Multiply(A=@n6, B=@n9), B=Multiply(A=@n10, B=@n11))
   "A == B"=@n12
   "A < B"=@n12
  )
  If(
   A=@n1
   B=@n2
   "A > B"=StaticSwitchParameter(
    True=TextureSampleParameter2D() {parameterName=CoatNormal, samplerType=SAMPLERTYPE_Normal, texture=&FlatNormal}.RGB
    False=@n3
   ) {parameterName=UseCoatNormal}
   "A == B"=@n4
   "A < B"=@n4
  )
  n13 = ScalarParameter() {parameterName=SheenWeight}
  n14 = ScalarParameter() {parameterName=CoatWeight}
  If(A=@n13, B=Constant(), "A > B"=@n13, "A == B"=@n14, "A < B"=@n14)
  n15 = ShadingModel()
  n16 = If(
   A=ScalarParameter() {parameterName=SubsurfaceWeight}
   B=Constant()
   "A > B"=ShadingModel() {shadingModel=MSM_Subsurface}
   "A == B"=@n15
   "A < B"=@n15
  )
  n17 = If(
   A=ScalarParameter() {parameterName=CoatWeight}
   B=Constant()
   "A > B"=ShadingModel() {shadingModel=MSM_ClearCoat}
   "A == B"=@n16
   "A < B"=@n16
  )
  If(A=ScalarParameter() {parameterName=SheenWeight}, B=Constant(), "A > B"=ShadingModel() {shadingModel=MSM_Cloth}, "A == B"=@n17, "A < B"=@n17)
  n18 = ScalarParameter() {parameterName=SubsurfaceWeight}
  n19 = Multiply(A=VectorParameter() {parameterName=SubsurfaceColor, defaultValue=<rgba 1 1 1 0>}.RGB, B=@n18)
  If(A=@n13, B=Constant(), "A > B"=VectorParameter() {parameterName=FuzzColor, defaultValue=<rgba 1 1 1 0>}.RGB, "A == B"=@n19, "A < B"=@n19)
  LinearInterpolate(
   A=ScalarParameter() {defaultValue=0.2, parameterName=SpecularRoughness}
   B=ScalarParameter() {defaultValue=0.3, parameterName=SheenRoughness}
   Alpha=ScalarParameter() {parameterName=SheenWeight}
  )
  Multiply(A=VectorParameter() {parameterName=EmissionColor, defaultValue=<rgba 1 1 1 0>}.RGB, B=ScalarParameter() {parameterName=EmissionWeight})
  Multiply(
   A=@n18
   B=Multiply(
    A=TextureSampleParameter2D() {parameterName=SubsurfaceScale, texture=/Engine/EngineResources/WhiteSquareTexture}.RGB
    B=ScalarParameter() {defaultValue=1, parameterName=SSS_Scale}
   )
  )
  n20 = ScalarParameter() {parameterName=Thickness}
  n21 = Constant()
  n22 = Multiply(A=Clamp(None=@n20) {maxDefault=6.38, minDefault=0.02}, B=Constant() {r=0.0391})
  n23 = VectorParameter() {parameterName=BaseColor, defaultValue=<rgba 0.18 0.18 0.18 1>}
  n24 = MaterialFunctionCall!(
   Light=Normalize(VectorInput=CameraPositionWS())
   View=CameraVectorWS()
   Roughness=ScalarParameter() {defaultValue=0.25, desc=DiffuseRough, parameterName=DiffuseRough}
  ) {desc=Mega Hacky - needs proper light vector, materialFunction=/InterchangeAssets/Functions/MF_OrenNayerView}
  n25 = Saturate(None=Add(A=Multiply(A=@n23.RGB, B=@n24.Result), B=@n23.RGB))
  n26 = ScalarParameter() {parameterName=Metallic}
  Multiply(
   A=MaterialFunctionCall!(Thickness=If(A=@n20, B=@n21, "A > B"=@n22, "A == B"=@n21, "A < B"=@n21), n2=Constant() {r=1.6}, n3=Constant() {r=1.1}, k=Constant() {r=1.5}) {materialFunction=/InterchangeAssets/Functions/MF_Iridescence}.Result
   B=Multiply(A=ScalarParameter() {defaultValue=1, parameterName=BaseWeight}, B=LinearInterpolate(A=@n25, B=@n23.RGB, Alpha=@n26))
  )
  n27 = ScalarParameter() {defaultValue=1, parameterName=SpecularWeight}
  Multiply(
   A=@n27
   B=MaterialFunctionCall!(IOR=ScalarParameter() {defaultValue=1.5, parameterName=IOR}) {materialFunction=/InterchangeAssets/Functions/MF_SchlickApprox}.Result
  )
  Multiply(A=@n26, B=@n27)
  ScalarParameter() {defaultValue=0.1, parameterName=CoatRoughness}
 ]
}
