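# Default Lit PBR metallic-roughness surface material in InterchangeAssets, enabled for skeletal meshes and mesh deformers.
# It exposes value, texture and map-weight parameters for base color, metallic, specular, roughness, emissive and ambient occlusion, plus a weighted normal map.
# All feed PBRSurfaceFunction_MR, which does the shading setup.
# A DebugFloat3Values node reads the result's BaseColor; no connection to material outputs is shown.
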Material /InterchangeAssets/Materials/PBRSurfaceMaterial_MR
graph {
 properties {bUsedWithMeshDeformer=true, bUsedWithSkeletalMesh=true, pixelDepthOffsetMode=PDOM_Legacy, usedShadingModels=Default Lit}
 nodes [
  n1 = VectorParameter() {parameterName=BaseColor}
  n2 = TextureObjectParameter() {parameterName=BaseColorMap}
  n3 = ScalarParameter() {parameterName=BaseColorMapWeight}
  n4 = ScalarParameter() {parameterName=Metallic}
  n5 = TextureObjectParameter() {parameterName=MetallicMap, samplerType=SAMPLERTYPE_LinearColor, texture=/Engine/EngineMaterials/BaseFlattenLinearColor}
  n6 = ScalarParameter() {parameterName=MetallicMapWeight}
  n7 = ScalarParameter() {defaultValue=0.5, parameterName=Specular}
  n8 = TextureObjectParameter() {parameterName=SpecularMap, samplerType=SAMPLERTYPE_LinearColor, texture=&BaseFlattenLinearColor}
  n9 = ScalarParameter() {parameterName=SpecularMapWeight}
  n10 = ScalarParameter() {defaultValue=0.5, parameterName=Roughness}
  n11 = TextureObjectParameter() {parameterName=RoughnessMap, samplerType=SAMPLERTYPE_LinearColor, texture=&BaseFlattenLinearColor}
  n12 = ScalarParameter() {parameterName=RoughnessMapWeight}
  n13 = VectorParameter() {parameterName=EmissiveColor}
  n14 = TextureObjectParameter() {parameterName=EmissiveColorMap}
  n15 = ScalarParameter() {parameterName=EmissiveColorMapWeight}
  n16 = TextureObjectParameter() {parameterName=NormalMap, samplerType=SAMPLERTYPE_Normal, texture=/Engine/EngineMaterials/DefaultNormal}
  n17 = ScalarParameter() {parameterName=NormalMapWeight}
  n18 = TextureObjectParameter() {parameterName=AmbientOcclusionMap, samplerType=SAMPLERTYPE_LinearColor, texture=&BaseFlattenLinearColor}
  n19 = ScalarParameter() {parameterName=AmbientOcclusionMapWeight}
  MaterialFunctionCall!(
   Vector3=BreakMaterialAttributes!(Attr=MaterialFunctionCall!(
     BaseColor=@n1.RGB
     BaseColorMap=@n2
     BaseColorMapWeight=@n3
     Metallic=@n4
     MetallicMap=@n5
     MetallicMapWeight=@n6
     Specular=@n7
     SpecularMap=@n8
     SpecularMapWeight=@n9
     Roughness=@n10
     RoughnessMap=@n11
     RoughnessMapWeight=@n12
     EmissiveColor=@n13.RGB
     EmissiveColorMap=@n14
     EmissiveColorMapWeight=@n15
     NormalMap=@n16
     NormalMapWeight=@n17
     AmbientOcclusionMap=@n18
     AmbientOcclusionMapWeight=@n19
    ) {materialFunction=&/PBRSurfaceFunction_MR}.Result).BaseColor
   MaximumNumberOfDigits
   UVs
   DebugTextLocation RG_UpperRight BA_LowerLeft
   Component Spacing
  ) {materialFunction=/Engine/Functions/Engine_MaterialFunctions02/Utility/DebugFloat3Values}
 ]
}
