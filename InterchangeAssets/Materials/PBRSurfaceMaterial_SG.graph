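# Default Lit parent material for Interchange imports of specular-glossiness PBR surfaces, usable on skeletal meshes and with mesh deformers.
# Exposes diffuse, specular, glossiness and emissive as constant, texture and weight parameters, plus weighted normal and ambient occlusion maps.
# PBRSurfaceFunction_SG does all the shading logic from these parameters, and BreakMaterialAttributes splits its Result into individual material attributes.
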Material /InterchangeAssets/Materials/PBRSurfaceMaterial_SG
graph {
 properties {bUsedWithMeshDeformer=true, bUsedWithSkeletalMesh=true, pixelDepthOffsetMode=PDOM_Legacy, usedShadingModels=Default Lit}
 nodes [BreakMaterialAttributes!(Attr=MaterialFunctionCall!(
    DiffuseColor=VectorParameter() {parameterName=DiffuseColor, defaultValue=<rgba 0 0 0 1>}.RGB
    DiffuseColorMap=TextureObjectParameter() {parameterName=DiffuseColorMap}
    DiffuseColorMapWeight=ScalarParameter() {parameterName=DiffuseColorMapWeight}
    SpecularColor=VectorParameter() {parameterName=SpecularColor, defaultValue=<rgba 0.03 0.03 0.03 1>}.RGB
    SpecularMap=TextureObjectParameter() {parameterName=SpecularColorMap}
    SpecularMapWeight=ScalarParameter() {parameterName=SpecularColorMapWeight}
    Glossiness=ScalarParameter() {defaultValue=0.5, parameterName=Glossiness}
    GlossinessMap=TextureObjectParameter() {parameterName=GlossinessMap, samplerType=SAMPLERTYPE_LinearColor, texture=/Engine/EngineMaterials/BaseFlattenLinearColor}
    GlossinessMapWeight=ScalarParameter() {parameterName=GlossinessMapWeight}
    EmissiveColor=VectorParameter() {parameterName=EmissiveColor}.RGB
    EmissiveColorMap=TextureObjectParameter() {parameterName=EmissiveColorMap}
    EmissiveColorMapWeight=ScalarParameter() {parameterName=EmissiveColorMapWeight}
    NormalMap=TextureObjectParameter() {parameterName=NormalMap, samplerType=SAMPLERTYPE_Normal, texture=/Engine/EngineMaterials/DefaultNormal}
    NormalMapWeight=ScalarParameter() {parameterName=NormalMapWeight}
    AmbientOcclusionMap=TextureObjectParameter() {parameterName=AmbientOcclusionMap, samplerType=SAMPLERTYPE_LinearColor, texture=&BaseFlattenLinearColor}
    AmbientOcclusionMapWeight=ScalarParameter() {parameterName=AmbientOcclusionMapWeight}
   ) {materialFunction=&/PBRSurfaceFunction_SG}.Result)]
}
