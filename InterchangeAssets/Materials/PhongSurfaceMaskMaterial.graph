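# Interchange masked, Default Lit parent-style material that represents imported Phong surfaces, usable on skeletal meshes and mesh deformers.
# MapOr functions pair textures with weights and fallback values for diffuse, specular, normal, emissive, occlusion, opacity mask and shininess.
# ConvertFromDiffSpec translates diffuse/specular into PBR inputs, while shininess is divided by 100 and inverted, as is the opacity mask.
# A tint-driven thin translucent output node is also present despite the masked blend mode.
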Material /InterchangeAssets/Materials/PhongSurfaceMaskMaterial
graph {
 properties {
  bUsedWithMeshDeformer=true
  bUsedWithSkeletalMesh=true
  blendMode=BLEND_Masked
  pixelDepthOffsetMode=PDOM_Legacy
  usedShadingModels=Default Lit
 }
 nodes [
  MaterialFunctionCall(
   DiffuseColor=MaterialFunctionCall!(
    Map=TextureObjectParameter() {parameterName=DiffuseColorMap}
    MapWeight=ScalarParameter() {parameterName=DiffuseColorMapWeight}
    Color=VectorParameter() {parameterName=DiffuseColor}.RGB
   ) {materialFunction=&/MapOrColorFunction}.Result
   SpecularColor=MaterialFunctionCall(
    Map=TextureObjectParameter() {parameterName=SpecularColorMap}
    MapWeight=ScalarParameter() {parameterName=SpecularColorMapWeight}
    Color=VectorParameter() {parameterName=SpecularColor}.RGB
   ) {materialFunction=&MapOrColorFunction}.Result
  ) {materialFunction=&ConvertFromDiffSpec}
  MaterialFunctionCall!(
   Map=TextureObjectParameter() {parameterName=NormalMap, samplerType=SAMPLERTYPE_Normal, texture=/Engine/EngineMaterials/DefaultNormal}
   MapWeight=ScalarParameter() {parameterName=NormalMapWeight}
   Color=Constant3Vector() {constant=<rgba 0 0 1 0>}
  ) {materialFunction=&/MapOrLinearColorFunction}
  MaterialFunctionCall(
   Map=TextureObjectParameter() {parameterName=EmissiveColorMap}
   MapWeight=ScalarParameter() {parameterName=EmissiveColorMapWeight}
   Color=VectorParameter() {parameterName=EmissiveColor}.RGB
  ) {materialFunction=&MapOrColorFunction}
  MaterialFunctionCall!(
   Map=TextureObjectParameter() {parameterName=AmbientOcclusionMap, samplerType=SAMPLERTYPE_LinearColor, texture=/Engine/EngineMaterials/BaseFlattenLinearColor}
   MapWeight=ScalarParameter() {parameterName=AmbientOcclusionMapWeight}
   Scalar=Constant()
  ) {materialFunction=&/MapOrScalarFunction}
  OneMinus(None=MaterialFunctionCall(
    Map=TextureObjectParameter() {parameterName=OpacityMaskMap}
    MapWeight=ScalarParameter() {parameterName=OpacityMaskMapWeight}
    Scalar=ScalarParameter() {defaultValue=1, parameterName=OpacityMask, sliderMax=1}
   ) {materialFunction=&MapOrScalarFunction}.Result)
  n1 = TextureObjectParameter() {parameterName=ShininessMap, samplerType=SAMPLERTYPE_LinearColor, texture=&BaseFlattenLinearColor}
  n2 = ScalarParameter() {parameterName=ShininessMapWeight}
  n3 = ScalarParameter() {parameterName=Shininess}
  OneMinus(None=Divide(A=MaterialFunctionCall(Map=@n1, MapWeight=@n2, Scalar=@n3) {materialFunction=&MapOrScalarFunction}.Result) {constB=100})
  ThinTranslucentMaterialOutput!(TransmittanceColor=VectorParameter() {parameterName=TintColor, defaultValue=<rgba 1 1 1 0>}.RGB, SurfaceCoverage)
 ]
}
