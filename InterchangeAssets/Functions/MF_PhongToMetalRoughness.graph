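# Material function in InterchangeAssets that converts Phong shading inputs into metallic-roughness PBR outputs.
# DiffuseColor passes straight through to BaseColor, while AmbientColor and SpecularColor become Metallic and Specular by averaging channels through a dot product with a "one" parameter, divided by three.
# Roughness comes from clamped Shininess as the fourth root of 2/(Shininess+2), clamped again on output.
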MaterialFunction /InterchangeAssets/Functions/MF_PhongToMetalRoughness
graph {nodes [
  FunctionOutput(None=FunctionInput() {bUsePreviewValueAsDefault=true, inputName=DiffuseColor}) {outputName=BaseColor}
  n1 = FunctionInput() {bUsePreviewValueAsDefault=true, inputName=AmbientColor, sortPriority=-1}
  n2 = ScalarParameter() {defaultValue=1, parameterName=one}
  FunctionOutput(None=Divide(A=DotProduct(A=@n1, B=@n2)) {constB=3}) {outputName=Metallic, sortPriority=1}
  n3 = FunctionInput() {bUsePreviewValueAsDefault=true, inputName=SpecularColor, sortPriority=1, previewValue=<0.0667 0.0667 0.0667 1>}
  FunctionOutput(None=Divide(A=DotProduct(A=@n3, B=@n2)) {constB=3}) {outputName=Specular, sortPriority=2}
  n4 = FunctionInput() {bUsePreviewValueAsDefault=true, inputName=Shininess, inputType=FunctionInput_Scalar, sortPriority=2, previewValue=<20 0 0 1>}
  n5 = Divide(B=Add(A=Clamp(None=@n4) {maxDefault=1000, minDefault=2}) {constB=2}) {constA=2}
  FunctionOutput(None=Clamp(None=Power(Base=@n5) {constExponent=0.25})) {outputName=Roughness, sortPriority=4}
 ]}
