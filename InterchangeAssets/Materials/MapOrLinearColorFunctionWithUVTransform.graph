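# MapOrLinearColorFunctionWithUVTransform is an Interchange material function that yields either a color, a texture map's color, or a mix of both.
# It samples the Map input with a linear color sampler and lerps from Color to the sampled RGB by MapWeight.
# The UVs input falls back to the default texture coordinate; despite the name, the function itself applies no UV transform.
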MaterialFunction /InterchangeAssets/Materials/MapOrLinearColorFunctionWithUVTransform
graph {nodes [
  n1 = FunctionInput(Preview=TextureCoordinate()) {bUsePreviewValueAsDefault=true, inputName=UVs, inputType=FunctionInput_Vector2}
  n2 = FunctionInput() {inputName=Map, inputType=FunctionInput_Texture2D}
  FunctionOutput(None=LinearInterpolate(
    A=FunctionInput() {inputName=Color}
    B=TextureSample(UVs=@n1, Tex=@n2) {samplerType=SAMPLERTYPE_LinearColor}.RGB
    Alpha=FunctionInput() {inputName=MapWeight, inputType=FunctionInput_Scalar}
   ))
 ]}
