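# Masked Default Lit Interchange material for imported Phong surfaces, usable on skeletal meshes and mesh deformers.
# Diffuse, specular, shininess, opacity mask, emissive, normal and occlusion each combine a weighted texture with a fallback value, using per-map UV offset, scale and rotation.
# MF_PhongToMetalRoughness converts the Phong inputs to PBR, with diffuse doubling as ambient.
# The opacity mask result is inverted, and a TintColor thin-translucent output is also present.
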Material /InterchangeAssets/Materials/PhongSurfaceMaskMaterialWithUVTransform
graph {
 properties {
  bUsedWithMeshDeformer=true
  bUsedWithSkeletalMesh=true
  blendMode=BLEND_Masked
  pixelDepthOffsetMode=PDOM_Legacy
  usedShadingModels=Default Lit
 }
 nodes [
  n1 = VectorParameter() {parameterName=DiffuseColorMap_OffsetScale, defaultValue=<rgba 0 0 1 1>}
  n2 = MaterialFunctionCall!(
   Map=TextureObjectParameter() {parameterName=DiffuseColorMap}
   MapWeight=ScalarParameter() {parameterName=DiffuseColorMapWeight}
   UVs=MaterialFunctionCall!(
    Offset_X=@n1.R
    Offset_Y=@n1.R
    Scale_X=@n1.R
    Scale_Y=@n1.R
    Rotation=ScalarParameter() {parameterName=DiffuseColorMap_Rotation}
    TexCoord
    TilingMethod
   ) {materialFunction=/InterchangeAssets/gltf/MaterialFunctions/MF_TransformUVs}.UVs
   Color=VectorParameter() {parameterName=DiffuseColor}.RGB
  ) {materialFunction=&/MapOrColorFunctionWithUVTransform}
  n3 = VectorParameter() {parameterName=SpecularColorMap_OffsetScale, defaultValue=<rgba 0 0 1 1>}
  n4 = ScalarParameter() {parameterName=SpecularColorMap_Rotation}
  n5 = VectorParameter() {parameterName=ShininessMap_OffsetScale, defaultValue=<rgba 0 0 1 1>}
  n6 = ScalarParameter() {parameterName=ShininessMap_Rotation}
  MaterialFunctionCall!(
   AmbientColor=@n2.Result
   DiffuseColor=@n2.Result
   SpecularColor=MaterialFunctionCall(
    Map=TextureObjectParameter() {parameterName=SpecularColorMap}
    MapWeight=ScalarParameter() {parameterName=SpecularColorMapWeight}
    UVs=MaterialFunctionCall(Offset_X=@n3.R, Offset_Y=@n3.R, Scale_X=@n3.R, Scale_Y=@n3.R, Rotation=@n4) {materialFunction=&MF_TransformUVs}.UVs
    Color=VectorParameter() {parameterName=SpecularColor}.RGB
   ) {materialFunction=&MapOrColorFunctionWithUVTransform}.Result
   Shininess=MaterialFunctionCall!(
    Map=TextureObjectParameter() {parameterName=ShininessMap, samplerType=SAMPLERTYPE_LinearColor, texture=/Engine/EngineMaterials/BaseFlattenLinearColor}
    MapWeight=ScalarParameter() {parameterName=ShininessMapWeight}
    UVs=MaterialFunctionCall(Offset_X=@n5.R, Offset_Y=@n5.R, Scale_X=@n5.R, Scale_Y=@n5.R, Rotation=@n6) {materialFunction=&MF_TransformUVs}.UVs
    Scalar=ScalarParameter() {parameterName=Shininess}
   ) {materialFunction=&/MapOrScalarFunctionWithUVTransform}.Result
  ) {materialFunction=/InterchangeAssets/Functions/MF_PhongToMetalRoughness}
  n7 = VectorParameter() {parameterName=OpacityMap_OffsetScale, defaultValue=<rgba 0 0 1 1>}
  n8 = ScalarParameter() {parameterName=OpacityMap_Rotation}
  OneMinus(None=MaterialFunctionCall(
    Map=TextureObjectParameter() {parameterName=OpacityMaskMap}
    MapWeight=ScalarParameter() {parameterName=OpacityMaskMapWeight}
    UVs=MaterialFunctionCall(Offset_X=@n7.R, Offset_Y=@n7.R, Scale_X=@n7.R, Scale_Y=@n7.R, Rotation=@n8) {materialFunction=&MF_TransformUVs}.UVs
    Scalar=ScalarParameter() {defaultValue=1, parameterName=OpacityMask, sliderMax=1}
   ) {materialFunction=&MapOrScalarFunctionWithUVTransform}.Result)
  n9 = VectorParameter() {parameterName=EmissiveColorMap_OffsetScale, defaultValue=<rgba 0 0 1 1>}
  n10 = ScalarParameter() {parameterName=EmissiveColorMap_Rotation}
  Reroute(None=MaterialFunctionCall(
    Map=TextureObjectParameter() {parameterName=EmissiveColorMap}
    MapWeight=ScalarParameter() {parameterName=EmissiveColorMapWeight}
    UVs=MaterialFunctionCall(Offset_X=@n9.R, Offset_Y=@n9.R, Scale_X=@n9.R, Scale_Y=@n9.R, Rotation=@n10) {materialFunction=&MF_TransformUVs}.UVs
    Color=VectorParameter() {parameterName=EmissiveColor}.RGB
   ) {materialFunction=&MapOrColorFunctionWithUVTransform}.Result)
  n11 = VectorParameter() {parameterName=NormalMap_OffsetScale, defaultValue=<rgba 0 0 1 1>}
  n12 = ScalarParameter() {parameterName=NormalMap_Rotation}
  Reroute(None=MaterialFunctionCall!(
    Map=TextureObjectParameter() {parameterName=NormalMap, samplerType=SAMPLERTYPE_Normal, texture=/Engine/EngineMaterials/DefaultNormal}
    MapWeight=ScalarParameter() {parameterName=NormalMapWeight}
    UVs=MaterialFunctionCall(Offset_X=@n11.R, Offset_Y=@n11.R, Scale_X=@n11.R, Scale_Y=@n11.R, Rotation=@n12) {materialFunction=&MF_TransformUVs}.UVs
    Color=Constant3Vector() {constant=<rgba 0 0 1 0>}
   ) {materialFunction=&/MapOrLinearColorFunctionWithUVTransform}.Result)
  n13 = VectorParameter() {parameterName=AmbientOcclusionMap_OffsetScale, defaultValue=<rgba 0 0 1 1>}
  n14 = ScalarParameter() {parameterName=AmbientOcclusionMap_Rotation}
  Reroute(None=MaterialFunctionCall(
    Map=TextureObjectParameter() {parameterName=AmbientOcclusionMap, samplerType=SAMPLERTYPE_LinearColor, texture=&BaseFlattenLinearColor}
    MapWeight=ScalarParameter() {parameterName=AmbientOcclusionMapWeight}
    UVs=MaterialFunctionCall(Offset_X=@n13.R, Offset_Y=@n13.R, Scale_X=@n13.R, Scale_Y=@n13.R, Rotation=@n14) {materialFunction=&MF_TransformUVs}.UVs
    Scalar=Constant()
   ) {materialFunction=&MapOrScalarFunctionWithUVTransform}.Result)
  ThinTranslucentMaterialOutput!(TransmittanceColor=VectorParameter() {parameterName=TintColor, defaultValue=<rgba 1 1 1 0>}.RGB, SurfaceCoverage)
 ]
}
