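# Parent material for Interchange assets that use a specular-glossiness PBR model with sheen, rendered with the Cloth shading model.
# PBRSurfaceFunction_SG combines diffuse, specular, glossiness, emissive, normal and ambient occlusion parameters, and its result is split by BreakMaterialAttributes.
# MapOrColorFunction and MapOrScalarFunction supply sheen color and sheen roughness.
# Each map has a weight parameter, and usage is enabled for skeletal meshes and mesh deformers.
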Material /InterchangeAssets/Materials/SheenMaterial_SG
graph {
 properties {bUsedWithMeshDeformer=true, bUsedWithSkeletalMesh=true, pixelDepthOffsetMode=PDOM_Legacy, shadingModel=MSM_Cloth, usedShadingModels=Cloth}
 nodes [
  BreakMaterialAttributes!(Attr=MaterialFunctionCall!(
    DiffuseColor=VectorParameter() {parameterName=DiffuseColor, defaultValue=<rgba 0 0 0 1>}.RGB
    DiffuseColorMap=TextureObjectParameter() {parameterName=DiffuseColorMap}
    DiffuseColorMapWeight=ScalarParameter() {parameterName=DiffuseColorMapWeight}
    SpecularColor=VectorParameter() {parameterName=SpecularColor, defaultValue=<rgba 0.03 0.03 0.03 1>}.RGB
    SpecularMap=TextureObjectParameter() {parameterName=SpecularColorMap}
    SpecularMapWeight=ScalarParameter() {parameterName=SpecularColorMapWeight}
    Glossiness=ScalarParameter() {defaultValue=0.5, parameterName=Glossiness}
    GlossinessMap=TextureObjectParameter() {parameterName=GlossinessMap, samplerType=SAMPLERTYPE_LinearColor, texture=/Engine/EngineMaterials/BaseFlattenLinearColor}
    GlossinessMapWeight=ScalarParameter() {parameterName=GlossinessMapWeight}
    EmissiveColor=VectorParameter() {parameterName=EmissiveColor}.RGB
    EmissiveColorMap=TextureObjectParameter() {parameterName=EmissiveColorMap}
    EmissiveColorMapWeight=ScalarParameter() {parameterName=EmissiveColorMapWeight}
    NormalMap=TextureObjectParameter() {parameterName=NormalMap, samplerType=SAMPLERTYPE_Normal, texture=/Engine/EngineMaterials/DefaultNormal}
    NormalMapWeight=ScalarParameter() {parameterName=NormalMapWeight}
    AmbientOcclusionMap=TextureObjectParameter() {parameterName=AmbientOcclusionMap, samplerType=SAMPLERTYPE_LinearColor, texture=&BaseFlattenLinearColor}
    AmbientOcclusionMapWeight=ScalarParameter() {parameterName=AmbientOcclusionMapWeight}
   ) {materialFunction=&/PBRSurfaceFunction_SG}.Result)
  MaterialFunctionCall!(
   Map=TextureObjectParameter() {parameterName=SheenColorMap}
   MapWeight=ScalarParameter() {parameterName=SheenColorMapWeight}
   Color=VectorParameter() {parameterName=SheenColor}.RGB
  ) {materialFunction=&/MapOrColorFunction}
  MaterialFunctionCall!(
   Map=TextureObjectParameter() {parameterName=SheenRoughnessMap, samplerType=SAMPLERTYPE_LinearColor, texture=&BaseFlattenLinearColor}
   MapWeight=ScalarParameter() {parameterName=SheenRoughnessMapWeight}
   Scalar=ScalarParameter() {parameterName=SheenRoughness}
  ) {materialFunction=&/MapOrScalarFunction}
 ]
}
